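# Unlit, two-sided translucent particle material for the WaywardRebel skin effect, drawing a star or light beam on Niagara meshes.
# Use_StarOrBeam blends beam opacity, from panning light rays in a centre gradient, with the default star, a tiled dot mask cut by a star texture and a radial ramp driven by dynamic Mask_Add.
# Particle alpha, view-angle fade and depth fade shape opacity, while boosted particle colour feeds emissive.
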Material /BRCosmetics/Characters/Player/Female/Medium/Bodies/F_MED_WaywardRebel/FX/M_WaywardRebel_Skin_Star
graph {
 properties {
  bUsedWithNiagaraMeshParticles=true
  blendMode=BLEND_Translucent
  pixelDepthOffsetMode=PDOM_Legacy
  shadingModel=MSM_Unlit
  twoSided=true
  usedShadingModels=Unlit
 }
 nodes [
  n1 = ParticleColor()
  n2 = TextureSample(UVs=Panner(Coordinate=TextureCoordinate() {uTiling=2, vTiling=0.01}) {speedX=0.2}) {texture=/Game/Effects/Fort_Effects/Textures/Outlander/T_InTheZone_LightRays_01}
  n3 = TextureSample(UVs=Panner(Coordinate=TextureCoordinate() {uTiling=1.5, vTiling=0.01}) {speedX=-0.4}) {texture=&T_InTheZone_LightRays_01}
  n4 = Power(Base=Multiply(A=@n2.R, B=@n3.G)) {constExponent=1.5}
  n5 = ComponentMask(None=Add(A=TextureCoordinate(), B=ScalarParameter() {defaultValue=-0.03, parameterName=Slide})) {g=true}
  n6 = Power(Base=@n5, Exp=ScalarParameter() {defaultValue=2.5, parameterName=Center Exponent})
  n7 = ScalarParameter() {defaultValue=5, parameterName=Center Brightness}
  n8 = Multiply(A=Multiply(A=@n5, B=Multiply(A=@n6, B=@n7)), B=Power(Base=OneMinus(None=@n5)))
  n9 = Multiply(A=@n8) {constB=3}
  n10 = Multiply(A=@n1.A, B=Add(A=Multiply(A=@n4, B=@n8), B=Power(Base=@n9) {constExponent=4}))
  n11 = TextureCoordinate()
  n12 = TextureSample(UVs=TextureCoordinate()) {texture=/BRCosmetics/Accessories/FORT_Backpacks/Backpack_F_MED_WaywardRebel/FX/T_WaywardRebel_Backpack_StarRGB}
  n13 = Add(
   A=TextureSample(UVs=Multiply(A=@n11) {constB=6}) {texture=/Game/Effects/Fort_Effects/Textures/Mask/T_Sphere_A}.R
   B=Multiply(A=@n12.R) {constB=0}
  )
  n14 = Add(
   A=MaterialFunctionCall() {materialFunction=&VectorToRadialValue}.Radial Coordinates
   B=DynamicParameter() {paramNames [Mask_Add,Param2,Param3,Param4]}.Mask_Add
  ) {constB=-1.5}
  n15 = Multiply(
   A=@n1.A
   B=Multiply(
    A=Multiply(A=@n13, B=@n12.G)
    B=TextureSample(UVs=@n14) {samplerSource=SSM_Clamp_WorldGroupSettings, texture=/Game/Effects/Fort_Effects/Textures/Ramps/T_Power_Source_Sphere_Ramp}.R
   )
  )
  n16 = ScalarParameter() {defaultValue=1, parameterName=Use_StarOrBeam}
  n17 = ScalarParameter() {parameterName=MeshFade}
  DepthFade(
   Opacity=Multiply(
    A=LinearInterpolate(A=@n10, B=@n15, Alpha=@n16)
    B=MaterialFunctionCall("FadeSensitivity (optional)"=@n17) {materialFunction=&FadeBasedOnViewAngle}.Result
   )
   FadeDistance=ScalarParameter() {parameterName=DepthFade}
  )
  MaterialFunctionCall(Emissive=Multiply(A=@n1.RGB, B=ScalarParameter() {defaultValue=1, parameterName=Emissive}) {constB=100}) {materialFunction=&MF_FX_EmissiveAdjustments}
  n18 = Panner(Coordinate=TextureCoordinate() {uTiling=4, vTiling=0}) {speedX=0.2}
  Multiply(A=Multiply(A=TextureSample(UVs=@n18) {texture=&T_NoisePack_01}.R, B=@n2.R), B=@n3.G)
 ]
}
