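# Translucent unlit, two-sided Niagara trail material for the Wayward Rebel character skin FX.
# World position offset combines UV-driven sine waves, a Niagara-driven bend and switchable SplineThicken width.
# Emissive tints particle color with one of five indexed curve-atlas gradients or a scrolling rainbow texture.
# Opacity blends panning trail textures or a screen-space dot/mosaic pattern, then applies UV masks and camera, view-angle and depth fades, with scalar lerps toggling most features.
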Material /BRCosmetics/Characters/Player/Female/Medium/Bodies/F_MED_WaywardRebel/FX/M_WaywardRebel_Skin_Trail
graph {
 properties {
  bUsedWithNiagaraMeshParticles=true
  bUsedWithNiagaraRibbons=true
  bUsedWithNiagaraSprites=true
  blendMode=BLEND_Translucent
  pixelDepthOffsetMode=PDOM_Legacy
  shadingModel=MSM_Unlit
  twoSided=true
  usedShadingModels=Unlit
 }
 nodes [
  n1 = VectorParameter() {parameterName=WPO_B, defaultValue=<rgba 0 0 3 1>}
  n2 = Add(
   A=TextureCoordinate() {uTiling=0.85}
   B=AppendVector(A=DynamicParameter() {defaultValue=<rgba -0.8 1 -60 60>, paramNames [Noise_TexspeedU, random, Clamp_min, Clamp_max]}.random, B=Constant())
  )
  n3 = Add(A=ComponentMask(None=@n2) {r=true}, B=Multiply(A=Time()) {constB=-1})
  n4 = VectorParameter() {parameterName=WPO_R, defaultValue=<rgba 3 0 0 1>}
  n5 = Add(A=ComponentMask(None=@n2) {r=true}, B=Multiply(A=Time()) {constB=-1})
  n6 = Add(
   A=Multiply(A=Transform(None=@n1.RGB) {transformSourceType=TRANSFORMSOURCE_Local}, B=Sine(None=@n3))
   B=Multiply(A=Transform(None=@n4.RGB) {transformSourceType=TRANSFORMSOURCE_Local}, B=Sine(None=@n5))
  )
  n7 = ScalarParameter() {defaultValue=2, parameterName=WPO_Intensity}
  n8 = ComponentMask(None=TextureCoordinate()) {r=true}
  n9 = Saturate(None=Multiply(A=@n8) {constB=5})
  n10 = OneMinus(None=@n8)
  n11 = Multiply(A=@n9, B=Saturate(None=Multiply(A=@n10) {constB=5}))
  n12 = ScalarParameter() {parameterName=Use_WPO_MaskTwoSide}
  n13 = DynamicParameter() {desc=Parameters set by Niagara, parameterIndex=1, paramNames [X,Y,Z,WPOIntensity]}
  n14 = DynamicParameter() {defaultValue=<rgba -0.8 1 -60 60>, paramNames [Noise_TexspeedU, random, Clamp_min, Clamp_max]}
  n15 = Clamp(
   None=Subtract(
    A=MaterialFunctionCall(X=@n13.X, Y=@n13.X, Z=@n13.X) {materialFunction=&MakeFloat3}.Result
    B=ParticlePositionWS()
   )
   Min=@n14.Clamp_min
   Max=@n14.Clamp_min
  ) {maxDefault=1000, minDefault=-1000}
  n16 = TextureCoordinate()
  n17 = Multiply(A=Power(Base=ComponentMask(None=@n16) {r=true}, Exp=ScalarParameter() {defaultValue=3, parameterName=WPOBendPower}), B=@n13.WPOIntensity)
  n18 = Add(
   A=Multiply(A=Multiply(A=@n6, B=@n7), B=LinearInterpolate(A=@n9, B=@n11, Alpha=@n12))
   B=LinearInterpolate(A=Constant(), B=LinearInterpolate(B=@n15, Alpha=@n17), Alpha=ScalarParameter() {defaultValue=1, parameterName=Use interpolate WPO})
  ) {constB=0}
  n19 = ScalarParameter() {defaultValue=3, parameterName=Base}
  n20 = DynamicParameter() {parameterIndex=2, defaultValue=<rgba 1 0 1 1>, paramNames [TrailTex_001speedV, Mask_AddYOffset, SplineWidth, Use_Dyn_ChangeColorCurve002]}
  n21 = ScalarParameter() {parameterName=Use_DynSplineWidth}
  n22 = LinearInterpolate(A=@n19, B=Multiply(A=@n19, B=@n20.SplineWidth), Alpha=@n21)
  n23 = ScalarParameter() {defaultValue=3, parameterName=Tip}
  n24 = LinearInterpolate(A=@n23, B=Multiply(A=@n23, B=@n20.SplineWidth), Alpha=@n21)
  n25 = StaticBool()
  LinearInterpolate(
   A=Add(
    A=@n18
    B=MaterialFunctionCall(WidthBase=@n22, WidthTip=@n24, Expand U or V UV Channel=@n25) {materialFunction=&SplineThicken}.WorldPositionOffset
   )
   B=@n18
   Alpha=ScalarParameter() {defaultValue=1, parameterName=Use_SplineThicken}
  )
  n26 = ParticleColor()
  n27 = ComponentMask(None=TextureCoordinate()) {r=true}
  n28 = CurveAtlasRowParameter(CurveTime=@n27) {
   atlas=/Game/Accessories/FORT_Backpacks/Backpack_F_MED_WaywardRebel/FX/CAtlas_WaywardRebel_ColorCurve
   curve=/Game/Characters/Player/Female/Medium/Bodies/F_MED_WaywardRebel/FX/CV_WaywardRebel_Skin_ColorCurve01
   defaultValue=1
   parameterName=Color001
  }
  n29 = CurveAtlasRowParameter(CurveTime=@n27) {
   atlas=&CAtlas_WaywardRebel_ColorCurve
   curve=/Game/Characters/Player/Female/Medium/Bodies/F_MED_WaywardRebel/FX/CV_Character_LocalTalent_Flare_ColorCurve03
   defaultValue=12
   parameterName=Color002
  }
  n30 = ScalarParameter() {parameterName=ColorIndex}
  n31 = DynamicParameter() {parameterIndex=2, defaultValue=<rgba 1 0 1 1>, paramNames [TrailTex_001speedV, Mask_AddYOffset, SplineWidth, Use_ChangeColorIndex]}
  n32 = ScalarParameter() {parameterName=Use_ChangeColorIndex}
  n33 = Clamp(None=Round(None=LinearInterpolate(A=@n30, B=@n31.Use_ChangeColorIndex, Alpha=@n32))) {maxDefault=4}
  n34 = Constant() {r=1}
  n35 = Constant()
  n36 = LinearInterpolate(A=@n28, B=@n29, Alpha=If(A=@n33, "A > B"=@n34, "A == B"=@n34, "A < B"=@n35) {constB=1})
  n37 = CurveAtlasRowParameter(CurveTime=@n27) {
   atlas=&CAtlas_WaywardRebel_ColorCurve
   curve=/Game/Characters/Player/Female/Medium/Bodies/F_MED_WaywardRebel/FX/CV_Character_LocalTalent_Flare_ColorCurve04
   defaultValue=13
   parameterName=Color003
  }
  n38 = If(A=@n33, "A > B"=@n34, "A == B"=@n34, "A < B"=@n35) {constB=2}
  n39 = CurveAtlasRowParameter(CurveTime=@n27) {
   atlas=&CAtlas_WaywardRebel_ColorCurve
   curve=/Game/Characters/Player/Female/Medium/Bodies/F_MED_WaywardRebel/FX/CV_Character_LocalTalent_Flare_ColorCurve05
   defaultValue=14
   parameterName=Color004
  }
  n40 = CurveAtlasRowParameter(CurveTime=@n27) {
   atlas=&CAtlas_WaywardRebel_ColorCurve
   curve=/Game/Characters/Player/Female/Medium/Bodies/F_MED_WaywardRebel/FX/CV_Character_LocalTalent_Flare_ColorCurve06
   defaultValue=15
   parameterName=Color005
  }
  n41 = LinearInterpolate(
   A=LinearInterpolate(A=LinearInterpolate(A=@n36, B=@n37, Alpha=@n38), B=@n39, Alpha=If(A=@n33, "A > B"=@n34, "A == B"=@n34, "A < B"=@n35) {constB=3})
   B=@n40
   Alpha=If(A=@n33, "A > B"=@n34, "A == B"=@n34, "A < B"=@n35) {constB=4}
  )
  n42 = Multiply(
   A=Add(A=TextureCoordinate(), B=ParticleRandom())
   B=AppendVector(
    A=ScalarParameter() {defaultValue=10, group=Color, parameterName=RainbowColorTex_UTiling}
    B=ScalarParameter() {defaultValue=10, group=Color, parameterName=RainbowColorTex_VTiling}
   )
  ) {constB=0.9}
  n43 = Constant()
  n44 = ScalarParameter() {group=Color, parameterName=RainbowColorTex_SpeedX}
  n45 = ScalarParameter() {group=Color, parameterName=RainbowColorTex_speedY}
  n46 = TextureSampleParameter2D(UVs=Panner(
    Coordinate=MaterialFunctionCall(UVs=@n42, "Rotation Angle (0-1)"=@n43) {materialFunction=&CustomRotator}.Rotated Values
    Speed=AppendVector(A=@n44, B=@n45)
   )) {parameterName=RainbowColorTex, texture=/Game/Gadgets/Assets/VinderTech_GliderChute/Glider_Peach_Male/FX/T_Glider_Peach_rainbow_gradient}
  n47 = Multiply(A=Desaturation(None=@n46.RGB, Fraction=Constant()))
  n48 = ScalarParameter() {parameterName=Use_Rainbowcolor_or_ColorCurve}
  n49 = DynamicParameter() {parameterIndex=3, paramNames [RainbowColor,Param2,Param3,Param4]}
  n50 = LinearInterpolate(
   A=@n26.RGB
   B=LinearInterpolate(A=Multiply(A=@n41, B=@n26.RGB), B=Multiply(A=@n47, B=@n26.RGB), Alpha=Multiply(A=@n48, B=@n49.RainbowColor))
   Alpha=ScalarParameter() {defaultValue=1, parameterName=Use_ColorCurve}
  )
  n51 = ScalarParameter() {parameterName=Desaturation}
  MaterialFunctionCall(Emissive=Multiply(A=Desaturation(None=@n50, Fraction=@n51), B=ScalarParameter() {defaultValue=1, parameterName=Emissive})) {materialFunction=&MF_FX_EmissiveAdjustments}
  ScalarParameter() {defaultValue=60, parameterName=MosaicUV}
  n52 = ScalarParameter() {defaultValue=0.6, parameterName=TrailTex_001U}
  n53 = ScalarParameter() {defaultValue=1, parameterName=TrailTex_001V}
  n54 = Add(
   A=Multiply(A=TextureCoordinate() {uTiling=1.5, vTiling=0.8}, B=AppendVector(A=@n52, B=@n53))
   B=AppendVector(A=ScalarParameter() {parameterName=TrailTex_MoveU}, B=ScalarParameter() {defaultValue=0.1, parameterName=TrailTex_MoveV})
  )
  n55 = TextureSampleParameter2D(UVs=Panner(Coordinate=TextureCoordinate() {uTiling=0.8, vTiling=0.4}) {speedX=-0.25}) {parameterName=NoiseTex, samplerType=SAMPLERTYPE_LinearGrayscale, texture=/Game/Global/Textures/Default/Noise/T_SwirlNoise_N_height}
  n56 = Add(A=@n54, B=Multiply(A=@n55.R, B=ScalarParameter() {defaultValue=0.1, parameterName=Noise_Intensity}) {constB=0.25})
  n57 = ScalarParameter() {defaultValue=1, parameterName=Use_Noise}
  n58 = ScalarParameter() {defaultValue=-1.25, parameterName=TrailTex_001SpeedU}
  n59 = ScalarParameter() {parameterName=TrailTex_001speedV}
  n60 = DynamicParameter() {parameterIndex=2, defaultValue=<rgba 1 0 1 1>, paramNames [TrailTex_001speedV, Mask_AddYOffset, SplineWidth, Use_Dyn_ChangeColorCurve002]}
  n61 = LinearInterpolate(A=@n59, B=Multiply(A=@n59, B=@n60.TrailTex_001speedV), Alpha=ScalarParameter() {parameterName=Use_Dyn_TrailTex_001speedV})
  n62 = TextureSampleParameter2D(UVs=Panner(Coordinate=LinearInterpolate(A=@n56, B=@n54, Alpha=@n57), Speed=AppendVector(A=@n58, B=@n61)) {speedX=-1.25}) {parameterName=TrailTex_001, texture=&/T_WaywardRebel_Skin_TrailRGB}
  n63 = StaticComponentMaskParameter(None=@n62.RGB) {defaultR=true, parameterName=ChannelMask}
  n64 = TextureCoordinate()
  n65 = VectorParameter() {parameterName=Trail002_Tex, defaultValue=<rgba 1 1 -0.8 0>}
  n66 = AppendVector(A=@n65.R, B=@n65.R)
  n67 = TextureSampleParameter2D(UVs=Panner(Coordinate=Multiply(A=@n64, B=@n66), Speed=AppendVector(A=@n65.B, B=@n65.B)) {speedX=-0.8}) {parameterName=Trail002_Tex, samplerType=SAMPLERTYPE_Grayscale, texture=&T_Smoke_Trail_Soft}
  n68 = Multiply(A=@n63, B=@n67.R)
  n69 = Power(Base=@n63, Exp=ScalarParameter() {defaultValue=1, parameterName=Tex_Power})
  n70 = ScalarParameter() {defaultValue=1, parameterName=Use_Trail002}
  n71 = MaterialFunctionCall(World Position=WorldPosition().XYZ) {materialFunction=&TransformToClipSpace}
  n72 = AppendVector(
   A=Multiply(
    A=ComponentMask(None=@n71.Buffer UVs for Scene Textures) {r=true}
    B=ScalarParameter() {defaultValue=140, parameterName=PointTexture_TilingU}
   ) {constB=100}
   B=Multiply(A=ComponentMask(None=@n71.Buffer UVs for Scene Textures) {g=true}, B=ScalarParameter() {defaultValue=70, parameterName=PointTexture_TilingV}) {constB=50}
  )
  n73 = Multiply(
   A=TextureSampleParameter2D(UVs=Multiply(A=@n72) {constB=2}) {parameterName=PointTexture, texture=/Game/Effects/Fort_Effects/Textures/Mask/T_HalfSphere_D}.R
   B=ScalarParameter() {defaultValue=1, parameterName=PointTexture_Mul}
  )
  n74 = AppendVector(
   A=ScalarParameter() {defaultValue=200, parameterName=PointMosaicGlow_U}
   B=ScalarParameter() {defaultValue=100, parameterName=PointMosaicGlow_V}
  )
  n75 = Divide(A=Floor(None=Multiply(A=@n71.Buffer UVs for Scene Textures, B=@n74)), B=@n74)
  n76 = Time()
  n77 = AppendVector(A=Constant(), B=ScalarParameter() {defaultValue=1.5, parameterName=PointMosaicGlow_SpeedY})
  n78 = TextureSampleParameter2D(UVs=Add(A=Multiply(A=@n75) {constB=3}, B=Multiply(A=@n76, B=@n77))) {parameterName=MosaicTex, samplerType=SAMPLERTYPE_LinearColor, texture=/Game/Global/Textures/Default/Noise/T_Wrap_SnowStreak_Noise}
  n79 = Constant()
  n80 = Multiply(A=Power(Base=Desaturation(None=@n78.RGB, Fraction=@n79)) {constExponent=3}) {constB=3}
  n81 = LinearInterpolate(A=@n73, B=Multiply(A=LinearInterpolate(Alpha=@n80) {constA=0.2}, B=@n73), Alpha=ScalarParameter() {parameterName=Use_PointMosaicGlow})
  n82 = Multiply(
   A=@n26.A
   B=LinearInterpolate(
    A=Multiply(A=@n81, B=@n63)
    B=LinearInterpolate(A=@n68, B=@n69, Alpha=@n70)
    Alpha=ScalarParameter() {defaultValue=1, parameterName=Use_Point}
   )
  )
  n83 = Power(Base=ComponentMask(None=TextureCoordinate()) {r=true}, Exp=ScalarParameter() {defaultValue=1, parameterName=LinearMaskLeft_Power})
  n84 = Multiply(A=Power(Base=OneMinus(None=@n83), Exp=ScalarParameter() {defaultValue=2, parameterName=LinearMaskRight_Power}), B=@n83)
  n85 = ComponentMask(None=TextureCoordinate() {uTiling=0.5, vTiling=0.5}) {g=true}
  n86 = Multiply(A=Power(Base=Sine(None=@n85)), B=ScalarParameter() {defaultValue=1, parameterName=Mask_Mul})
  n87 = LinearInterpolate(
   A=Multiply(A=@n82, B=Multiply(A=@n84, B=@n86))
   B=Multiply(A=@n82, B=@n84)
   Alpha=ScalarParameter() {defaultValue=1, parameterName=Use_Mask02}
  )
  n88 = VectorParameter() {parameterName=MaskTexture, defaultValue=<rgba 1 1 0 0>}
  n89 = Multiply(A=TextureCoordinate(), B=AppendVector(A=@n88.R, B=@n88.R))
  n90 = DynamicParameter() {parameterIndex=2, defaultValue=<rgba 1 0 1 1>, paramNames [TrailTex_001speedV, Mask_AddYOffset, SplineWidth, Use_Dyn_ChangeColorCurve002]}
  n91 = AppendVector(A=Constant(), B=@n90.Mask_AddYOffset)
  n92 = TextureSampleParameter2D(UVs=Panner(Coordinate=Add(A=@n89, B=@n91), Speed=AppendVector(A=@n88.B, B=@n88.B))) {parameterName=MaskTexture, texture=/Game/Characters/Player/Male/Small/T3_Items/Shuriken/Textures/T_Shuriken_Trail}
  n93 = Multiply(
   A=LinearInterpolate(A=@n87, B=Multiply(A=@n87, B=@n92.R), Alpha=ScalarParameter() {parameterName=Use_MaskTexture})
   B=ScalarParameter() {defaultValue=1, parameterName=Alpha}
  )
  n94 = MaterialFunctionCall(Distance=ScalarParameter() {defaultValue=150, parameterName=Near camera fade}) {materialFunction=&Near_Camera_Fade}
  n95 = Multiply(
   A=LinearInterpolate(A=@n93, B=Multiply(A=@n93, B=@n94.Result), Alpha=ScalarParameter() {parameterName=Use_CameraFade})
   B=MaterialFunctionCall("FadeSensitivity (optional)"=ScalarParameter() {defaultValue=1, parameterName=MeshFade}) {materialFunction=&FadeBasedOnViewAngle}.Result
  )
  n96 = DepthFade(Opacity=@n95, FadeDistance=ScalarParameter() {defaultValue=5, parameterName=DepthFade})
  ShadingPathSwitch(Default=@n96, Deferred=@n96, Forward=@n95, Mobile=@n95)
 ]
}
