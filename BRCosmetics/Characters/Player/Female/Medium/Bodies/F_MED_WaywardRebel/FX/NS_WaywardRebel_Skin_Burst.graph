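# Niagara burst effect for the Wayward Rebel female body skin: a short, color-shifting star flash with rays and particles.
# Star meshes, glows, mesh rays and an optional light, disabled by User.Hide light, burst with staggered timing and animate through scale, color and material-parameter curves.
# A sprite group scatters from a cylinder surface with random colors and visibility tags, distance culling and time-of-day alpha.
# User.IsLeft is declared but unreferenced in the visible modules.
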NiagaraSystem /BRCosmetics/Characters/Player/Female/Medium/Bodies/F_MED_WaywardRebel/FX/NS_WaywardRebel_Skin_Burst
dependencies {
 usedDataInterfaces [&NiagaraDataInterfaceCurve, &NiagaraDataInterfaceColorCurve, &NiagaraDataInterfaceArrayFloat3]
 usedDynamicInputs [&AddVectorToPosition]
 usedModules [&SystemState]
}
emitters {
 Light {
  emitterUpdateScript [
   &EmitterState(
    Inactive Response=::NewEnumerator1
    Life Cycle Mode=::NewEnumerator1
    Loop Behavior=::NewEnumerator1
    Loop Delay=0.3
    Loop Duration=1
    UseLoopDelay=true
   )
   &SpawnBurst_Instantaneous(
    Spawn Count=&MakeCustomIntFromBool(Bool=$User.Hide light, True INT=0, False INT=1)
    Use Spawn Probability=false
    UseLoopCountLimit=false
   )
  ]
  particleSpawnScript [&InitializeParticle(
    Lifetime=0.5
    Mesh Scale=<0.3 -1 -1>
    Mesh Scale Mode=::NewEnumerator4
    Position Offset=<10 -20 0>
    Sprite Rotation Angle=&RandomRangeFloat(Minimum=0, Maximum=1)
    Sprite Rotation Mode=::NewEnumerator1
    Sprite Size Mode=::NewEnumerator3
    UsePositionOffset=true
    Write Material Random=true
   )]
  particleUpdateScript [
   &ParticleState()
   &FN_Light_Attributes(
    Color=&ColorFromCurve(ColorCurve=DataInterface() {propertyValues {
       AlphaCurve {keys []}
       BlueCurve {keys [<at 0 0 RCIM_Cubic>,<at 0.1896 0>,<at 0.4403 0>,<at 1 0 RCIM_Cubic>]}
       GreenCurve {keys [<at 0 0 RCIM_Cubic>,<at 0.1896 1>,<at 0.4403 0.2>,<at 1 0 RCIM_Cubic>]}
       RedCurve {keys [<at 0 0 RCIM_Cubic>,<at 0.1896 0.0226>,<at 0.4403 1>,<at 1 0 RCIM_Cubic>]}
      }})
    Create Color Attribute=true
    Create Exponent Attribute=true
    Create Position Attribute=true
    Create Radius Attribute=true
    Exponent=1.5
    Position=<20 -30 0>
    Radius=80
    Scale Color=30
   )
  ]
  rendererClasses [&NiagaraLightRendererProperties]
  renderers [{rendererClass=&NiagaraLightRendererProperties, rendererIndex=0}]
 }
 Ray_Line {
  emitterUpdateScript [
   &EmitterState(Inactive Response=::NewEnumerator1, Life Cycle Mode=::NewEnumerator1, Loop Behavior=::NewEnumerator1, Loop Duration=1, UseLoopDelay=false)
   &SpawnBurst_Instantaneous(Spawn Count=2, Use Spawn Probability=true, UseLoopCountLimit=false)
  ]
  particleSpawnScript [
   &PartitionParticles()
   &InitializeParticle(
    Lifetime=1.5
    Mesh Scale=<0.6 -1.1 -1.1>
    Mesh Scale Mode=::NewEnumerator4
    Position Offset=<-3 0 0>
    Sprite Rotation Angle=&RandomRangeFloat(Minimum=0, Maximum=1)
    Sprite Rotation Mode=::NewEnumerator1
    UsePositionOffset=true
    Write Material Random=true
   )
   &InitialMeshOrientation(
    Mesh Orientation Mode=::NewEnumerator2
    Rotation=&SelectVectorFromArray(
     Direct Array Index=$Particles.PartitionParticles.Partition
     Vector Selection Array=DataInterface() {propertyValues {FloatData [<0 0 0>, <0.5 0 0>]}}
    )
    Use Rotation Vector=true
   )
  ]
  particleUpdateScript [
   &ParticleState()
   &UpdateMeshOrientation()
   &ScaleMeshSize(Scale Factor=&VectorFromFloat(Value=&FloatFromCurve(
      FloatCurve=DataInterface() {propertyValues {Curve {keys [
          {arriveTangent=2.3969, interpMode=RCIM_Cubic, leaveTangent=2.3969, tangentMode=RCTM_User, time=0, value=0}
          {arriveTangent=0.0007, interpMode=RCIM_Cubic, leaveTangent=0.0007, tangentMode=RCTM_User, time=0.2, value=0.35}
          {arriveTangent=-0.4123, interpMode=RCIM_Cubic, leaveTangent=-0.4123, tangentMode=RCTM_User, time=1, value=0.1}
         ]}}}
      Scale Curve=&RandomRangeFloat(Minimum=0.65, Maximum=0.65)
     )))
   &ScaleColor(
    Scale Alpha=&FloatFromCurve(FloatCurve=DataInterface() {propertyValues {Curve {keys [
         <at 0 0 RCIM_Cubic>
         {arriveTangent=2.8571, interpMode=RCIM_Cubic, leaveTangent=2.8571, time=0.4, value=1}
         {arriveTangent=0.0225, interpMode=RCIM_Cubic, leaveTangent=0.0225, tangentMode=RCTM_User, time=0.7, value=2}
         <at 1 0 RCIM_Cubic>
        ]}}})
    Scale RGB=&Multiply_VectorByFloat(Vector=<1 1 1>, Float=1)
    ScaleA=true
    ScaleRGB=true
   )
  ]
  rendererClasses [&NiagaraMeshRendererProperties]
  renderers [{rendererClass=&NiagaraMeshRendererProperties, rendererIndex=0}]
 }
 Rays {
  emitterUpdateScript [
   &EmitterState(
    Inactive Response=::NewEnumerator1
    Life Cycle Mode=::NewEnumerator1
    Loop Behavior=::NewEnumerator1
    Loop Delay=0.3
    Loop Duration=1
    UseLoopDelay=true
   )
   &SpawnBurst_Instantaneous(Spawn Count=2, Use Spawn Probability=false, UseLoopCountLimit=false)
  ]
  particleSpawnScript [
   &InitializeParticle(
    Color=<rgba 1 1 1 0.5>
    Lifetime=&RandomRangeFloat(Minimum=0.5, Maximum=0.6)
    Mesh Scale=<0.3 -1 -1>
    Mesh Scale Mode=::NewEnumerator4
    Sprite Rotation Angle=&RandomRangeFloat(Minimum=0, Maximum=1)
    Sprite Rotation Mode=::NewEnumerator1
    UsePositionOffset=true
    Write Material Random=true
   )
   &Color(Color=&MakeLinearColorFromVectorAndFloat("Vector (RGB)"=<0.1 0.002 0.12>, "Float (Alpha)"=&ClampFloat(Float=1)))
   &DynamicMaterialParameters(
    Index 0 Param 1=&RandomRangeFloat(Minimum=0, Maximum=10)
    Index 0 Param 2=&RandomRangeFloat(Minimum=2.5, Maximum=4.5)
    Index 0 Param 3=&RandomRangeFloat(Minimum=0.8, Maximum=1.6)
    Module.Param0WriteEnabled=true
    Module.Param1WriteEnabled=true
    Module.Param2WriteEnabled=true
    Module.Param3WriteEnabled=false
   )
  ]
  particleUpdateScript [
   &ParticleState()
   &ScaleMeshSize(Scale Factor=&VectorFromFloat(Value=&FloatFromCurve(
      FloatCurve=DataInterface() {propertyValues {Curve {keys [
          {arriveTangent=15.9108, interpMode=RCIM_Cubic, time=0, value=0.0062}
          {arriveTangent=1.323, interpMode=RCIM_Cubic, leaveTangent=1.323, time=0.2, value=0.6}
          {arriveTangent=-0.0562, interpMode=RCIM_Cubic, leaveTangent=-0.0562, tangentMode=RCTM_User, time=0.6, value=0.8}
          {arriveTangent=-0.9004, interpMode=RCIM_Cubic, leaveTangent=-0.9004, tangentMode=RCTM_User, time=1, value=0.4}
         ]}}}
      Scale Curve=&RandomRangeFloat(Minimum=0.5, Maximum=0.7)
     )))
   &ScaleColor(
    Scale Alpha=&FloatFromCurve(FloatCurve=DataInterface() {propertyValues {Curve {keys [
         <at 0 0 RCIM_Cubic>
         {arriveTangent=-2.0099, interpMode=RCIM_Cubic, leaveTangent=-2.0099, tangentMode=RCTM_User, time=0.3, value=1}
         <at 1 0 RCIM_Cubic>
        ]}}})
    Scale RGB=&Multiply_VectorByFloat(Vector=<1 1 1>, Float=1)
    ScaleA=true
    ScaleRGB=true
   )
   &DynamicMaterialParameters(
    Index 0 Param 4=1
    Module.Param0WriteEnabled=false
    Module.Param1WriteEnabled=false
    Module.Param2WriteEnabled=false
    Module.Param3WriteEnabled=true
   ) {module=DynamicMaterialParameters001}
  ]
  rendererClasses [&NiagaraMeshRendererProperties]
  renderers [{rendererClass=&NiagaraMeshRendererProperties, rendererIndex=0}]
 }
 Sprites_group {
  emitterUpdateScript [
   &EmitterState(
    Enable Distance Culling=true
    Inactive Response=::NewEnumerator1
    Life Cycle Mode=::NewEnumerator1
    Loop Behavior=::NewEnumerator1
    Loop Delay=&RandomRangeFloat(Minimum=0, Maximum=1)
    Loop Duration=1
    Max Distance Response=::KillImmediately
    Module.bUseMaxDistance=true
    Module.bUseMinDistance=false
    Scalability Mode=::NewEnumerator1
    UseLoopDelay=false
   )
   &SpawnBurst_Instantaneous(
    Spawn Count=&RandomRangeInt(Minimum=10, Maximum=13)
    Spawn Time=0.3
    Use Spawn Probability=false
    UseLoopCountLimit=false
   )
  ]
  particleSpawnScript [
   &InitializeParticle(
    Color=&MakeCustomLinearColorFromBool(
     Bool=&RandomBool()
     True Color=<rgba 1 0.6 0 1>
     False Color=&MakeCustomLinearColorFromBool(Bool=&RandomBool(), True Color=<rgba 0.7188 0.0905 0 1>, False Color=<rgba 0.0701 0.5417 0 1>)
    )
    Lifetime Max=0.5
    Lifetime Min=0.4
    Lifetime Mode=::NewEnumerator1
    Sprite Rotation Mode=::NewEnumerator2
    Sprite Size Mode=::NewEnumerator1
    Uniform Sprite Size Min=6
    UsePositionOffset=false
    Write Material Random=true
   )
   &ShapeLocation(
    Cylinder Height=10
    Cylinder Radius=6
    MaskSpawnGroup=false
    Override Seed=false
    Rotation Mode=::NewEnumerator1
    Shape Origin=&VectorFromFloat(Value=0)
    Shape Primitive=::NewEnumerator1
    Surface Only=true
    Surface Only Band Thickness=0.1
    Use Endcaps In Surface Only Mode=false
    UseOwnerScale=true
   )
   &AddVelocity(
    Module.Use Falloff Distance=false
    Origin Offset=<-5 0 0>
    Override Seed=false
    Velocity Mode=::NewEnumerator1
    Velocity Speed=&RandomRangeFloat(Minimum=20, Maximum=35)
   )
   /SetVariables(Particles.VisibilityTag=&RandomRangeInt(Minimum=0, Maximum=3)) {isSetParameters=true}
   /SetVariables(Particles.Color_triangle:LinearColor=&MakeCustomLinearColorFromBool(
     Bool=&RandomBool()
     True Color=<rgba 0.8 0.6577 0 1>
     False Color=&MakeCustomLinearColorFromBool(Bool=&RandomBool(), True Color=<rgba 1 0.25 0 1.5>, False Color=<rgba 0.0906 0.7 0 1>)
    )) {isSetParameters=true}
   /SetVariables(Particles.Color_triangleglow:LinearColor=&MultiplyLinearColor(
     Linear Color=$Particles.Color_triangle
     Scale Factor=&MakeVector4_FromLinearColor(LinearColor=/Niagara/DynamicInputs/TypeConversions/MakeLinearColorFromFloat(Value:NiagaraFloat=0.5))
    )) {isSetParameters=true}
  ]
  particleUpdateScript [
   &ParticleState()
   &SpriteRotationRate(Rotation Rate=&RandomRangeFloat(Minimum=-200, Maximum=200))
   &ScaleSpriteSize(Scale Factor=&Vector2DFromFloat(Value=&FloatFromCurve(FloatCurve=DataInterface() {propertyValues {Curve {keys [
          {interpMode=RCIM_Cubic, tangentMode=RCTM_User, time=0, value=0}
          {interpMode=RCIM_Cubic, tangentMode=RCTM_User, time=0.4, value=1}
          {tangentMode=RCTM_User, time=1, value=0}
         ]}}})))
   /SetVariables(Particles.Size002:Vector2f=&Multiply_Vector2D(A=$Particles.SpriteSize, B=&Vector2DFromFloat(Value=1.5))) {isSetParameters=true}
   /SetVariables(Particles.Size_Triangle:Vector2f=&Multiply_Vector2D(A=$Particles.SpriteSize, B=&Vector2DFromFloat(Value=0.8))) {isSetParameters=true}
   &Drag(Drag=&RandomRangeFloat(Minimum=0.5, Maximum=1), Use Linear Drag=true, Use Rotational Drag=true)
   &SolveForcesAndVelocity(Clamp Velocity=false, Limit Acceleration=false)
   &ScaleColor(
    Scale Alpha=&Multiply_Float(
     A=&FNTimeOfDayScalar(Night Scale=1, Morning Scale=0.3, Evening Scale=0.3)
     B=&FloatFromCurve(FloatCurve=DataInterface() {propertyValues {Curve {keys [
          {interpMode=RCIM_Cubic, tangentMode=RCTM_User, time=0.1, value=0}
          {interpMode=RCIM_Cubic, tangentMode=RCTM_User, time=0.3, value=0.8}
          {tangentMode=RCTM_User, time=0.8, value=0}
         ]}}})
    )
    ScaleA=true
    ScaleRGB=true
   )
   /SetVariables(Particles.Color002=&MultiplyLinearColor(Linear Color=<rgba 1.2 1.2 1.2 1>, Scale Factor=&MakeVector4_FromLinearColor(LinearColor=$Particles.Color))) {isSetParameters=true}
  ]
  rendererClasses [&NiagaraSpriteRendererProperties]
  renderers [
   {rendererClass=&NiagaraSpriteRendererProperties}
   {rendererClass=&NiagaraSpriteRendererProperties, rendererIndex=1}
   {rendererClass=&NiagaraSpriteRendererProperties, rendererIndex=2}
   {rendererClass=&NiagaraSpriteRendererProperties, rendererIndex=3}
  ]
 }
 Star_Glow {
  emitterUpdateScript [
   &EmitterState(
    Inactive Response=::NewEnumerator1
    Life Cycle Mode=::NewEnumerator1
    Loop Behavior=::NewEnumerator1
    Loop Delay=0.3
    Loop Duration=1
    UseLoopDelay=true
   )
   &SpawnBurst_Instantaneous(Use Spawn Probability=false, UseLoopCountLimit=false)
  ]
  particleSpawnScript [
   &InitializeParticle(
    Color=<rgba 1.5 1.5 1.5 1>
    Lifetime=0.3
    Sprite Size Mode=::NewEnumerator3
    Uniform Sprite Size=23
    UsePositionOffset=false
    Write Material Random=true
   )
   &SpriteFacingAndAlignment(Sprite Alignment=<0 0 1>, WriteAlignment=true, WriteFacing=true)
   &AddVelocity(Velocity=<5 0 0>)
  ]
  particleUpdateScript [
   &ParticleState()
   &ScaleSpriteSize(
    Uniform Curve Scale=1.05
    Uniform Curve Sprite Scale=DataInterface() {propertyValues {Curve {keys [{tangentMode=RCTM_User, time=0, value=1}, {tangentMode=RCTM_User, time=1, value=2}]}}}
   )
   &ScaleColor(
    Scale Alpha=&FloatFromCurve(FloatCurve=DataInterface() {propertyValues {Curve {keys [
         <at 0 0 RCIM_Cubic>
         {arriveTangent=-2.0518, interpMode=RCIM_Cubic, leaveTangent=-2.0518, tangentMode=RCTM_User, time=0.2, value=1}
         <at 1 0 RCIM_Cubic>
        ]}}})
    Scale RGB=&Multiply_VectorByFloat(
     Vector=&MakeVectorFromLinearColorRGB(Color=&ColorFromCurve(ColorCurve=DataInterface() {propertyValues {
         AlphaCurve {keys []}
         BlueCurve {keys [<at 0 0.1582 RCIM_Cubic>,<at 1 0.0657 RCIM_Cubic>]}
         GreenCurve {keys [<at 0 1 RCIM_Cubic>,<at 1 0.5461 RCIM_Cubic>]}
         RedCurve {keys [<at 0 0.4087 RCIM_Cubic>,<at 1 1 RCIM_Cubic>]}
        }}))
     Float=1
    )
    ScaleA=true
    ScaleRGB=true
   )
   &SolveForcesAndVelocity(Clamp Velocity=false, Limit Acceleration=false)
  ]
  rendererClasses [&NiagaraSpriteRendererProperties]
  renderers [{rendererClass=&NiagaraSpriteRendererProperties}]
 }
 Star_Glowmesh {
  emitterUpdateScript [
   &EmitterState(
    Inactive Response=::NewEnumerator1
    Life Cycle Mode=::NewEnumerator1
    Loop Behavior=::NewEnumerator1
    Loop Delay=0.3
    Loop Duration=1
    UseLoopDelay=true
   )
   &SpawnBurst_Instantaneous(Use Spawn Probability=false, UseLoopCountLimit=false)
  ]
  particleSpawnScript [
   &InitializeParticle(
    Color=<rgba 1.1 1.1 1.1 1>
    Lifetime=0.5
    Mesh Scale=<4.8 2.6 2.6>
    Mesh Scale Mode=::NewEnumerator4
    Sprite Size Mode=::NewEnumerator3
    Uniform Sprite Size=23
    UsePositionOffset=true
    Write Material Random=true
   )
   &InitialMeshOrientation(Mesh Orientation Mode=::NewEnumerator2, Use Rotation Vector=true)
  ]
  particleUpdateScript [
   &ParticleState()
   &ScaleMeshSize(Scale Factor=&MakeVector(
     X=&FloatFromCurve(FloatCurve=DataInterface() {propertyValues {Curve {keys [
          {arriveTangent=2.3556, interpMode=RCIM_Cubic, leaveTangent=2.3556, tangentMode=RCTM_User, time=0, value=0.3}
          {arriveTangent=0.0629, interpMode=RCIM_Cubic, leaveTangent=0.0629, tangentMode=RCTM_User, time=1, value=1}
         ]}}})
     Y=1
     Z=1
    ))
   &ScaleColor(
    Scale Alpha=&FloatFromCurve(FloatCurve=DataInterface() {propertyValues {Curve {keys [
         <at 0 0 RCIM_Cubic>
         {arriveTangent=-2.0518, interpMode=RCIM_Cubic, leaveTangent=-2.0518, tangentMode=RCTM_User, time=0.2, value=1}
         <at 1 0 RCIM_Cubic>
        ]}}})
    Scale RGB=&Multiply_VectorByFloat(
     Vector=&MakeVectorFromLinearColorRGB(Color=&ColorFromCurve(ColorCurve=DataInterface() {propertyValues {
         AlphaCurve {keys []}
         BlueCurve {keys [<at 0.3313 0 RCIM_Cubic>,<at 0.6213 0.3148>,<at 0.8325 0 RCIM_Cubic>]}
         GreenCurve {keys [<at 0.3313 0.6282 RCIM_Cubic>,<at 0.6213 0>,<at 0.8325 1 RCIM_Cubic>]}
         RedCurve {keys [<at 0.3313 1 RCIM_Cubic>,<at 0.6213 1>,<at 0.8325 0.0389 RCIM_Cubic>]}
        }}))
     Float=1
    )
    ScaleA=true
    ScaleRGB=true
   )
   &SolveForcesAndVelocity(Clamp Velocity=false, Limit Acceleration=false)
  ]
  rendererClasses [&NiagaraMeshRendererProperties]
  renderers [{rendererClass=&NiagaraMeshRendererProperties, rendererIndex=0}]
 }
 Star_Main {
  emitterUpdateScript [
   &EmitterState(
    Inactive Response=::NewEnumerator1
    Life Cycle Mode=::NewEnumerator1
    Loop Behavior=::NewEnumerator1
    Loop Delay=0.3
    Loop Duration=1
    UseLoopDelay=true
   )
   &SpawnBurst_Instantaneous(Use Spawn Probability=true, UseLoopCountLimit=false)
  ]
  particleSpawnScript [
   &InitializeParticle(
    Color=<rgba 1.5 1.5 1.5 2>
    Lifetime=0.55
    Mesh Scale=&VectorFromFloat(Value=0.48)
    Mesh Scale Mode=::NewEnumerator4
    Sprite Size Mode=::NewEnumerator3
    Uniform Sprite Size=30
    UsePositionOffset=true
    Write Material Random=true
   )
   &AddVelocity(Velocity=<10 0 0>)
   &InitialMeshOrientation(Mesh Orientation Mode=::NewEnumerator2, Rotation=<0 0 0.25>, Use Rotation Vector=true)
   &SpriteFacingAndAlignment(Sprite Alignment=<0 0 1>, WriteAlignment=true, WriteFacing=true)
  ]
  particleUpdateScript [
   &ParticleState()
   &ScaleMeshSize(Scale Factor=&VectorFromFloat(Value=&FloatFromCurve(
      FloatCurve=DataInterface() {propertyValues {Curve {keys [
          {arriveTangent=1.2339, interpMode=RCIM_Cubic, leaveTangent=1.2339, tangentMode=RCTM_User, time=0, value=0.6}
          {arriveTangent=0.2683, interpMode=RCIM_Cubic, leaveTangent=0.2683, tangentMode=RCTM_User, time=1, value=1.5}
         ]}}}
      Scale Curve=0.6
     )))
   &ScaleSpriteSize(
    Uniform Curve Scale=1.05
    Uniform Curve Sprite Scale=DataInterface() {propertyValues {Curve {keys [
        {arriveTangent=0.8538, interpMode=RCIM_Cubic, leaveTangent=0.8538, tangentMode=RCTM_User, time=0, value=0.6}
        {arriveTangent=0.3287, interpMode=RCIM_Cubic, leaveTangent=0.3287, tangentMode=RCTM_User, time=1, value=1.2}
       ]}}}
   )
   &ScaleColor(
    Scale Alpha=&FloatFromCurve(
     FloatCurve=DataInterface() {propertyValues {Curve {keys [
         {arriveTangent=5.3825, interpMode=RCIM_Cubic, leaveTangent=5.3825, tangentMode=RCTM_User, time=0, value=0}
         {arriveTangent=6.6821, tangentMode=RCTM_User, time=0.15, value=2}
         {arriveTangent=-1.5609, interpMode=RCIM_Cubic, leaveTangent=-1.5609, tangentMode=RCTM_User, time=0.3, value=0.5}
         {arriveTangent=-0.2298, interpMode=RCIM_Cubic, leaveTangent=-0.2298, tangentMode=RCTM_User, time=1, value=0}
        ]}}}
     Scale Curve=1.5
    )
    Scale RGB=&MakeVectorFromLinearColorRGB(Color=&ColorFromCurve(ColorCurve=DataInterface() {propertyValues {
        AlphaCurve {keys []}
        BlueCurve {keys [<at 0.0959 3 RCIM_Cubic>,<at 0.3593 0.0272 RCIM_Cubic>]}
        GreenCurve {keys [<at 0.0959 0.2725 RCIM_Cubic>,<at 0.3593 1.5 RCIM_Cubic>]}
        RedCurve {keys [<at 0.0959 2.4554 RCIM_Cubic>,<at 0.3593 0.8681 RCIM_Cubic>]}
       }}))
    ScaleA=true
    ScaleRGB=true
   )
   &SolveForcesAndVelocity(Clamp Velocity=false, Limit Acceleration=false)
   &DynamicMaterialParameters(
    Index 0 Param 1=&FloatFromCurve(FloatCurve=DataInterface() {propertyValues {Curve {keys [{arriveTangent=0.2079, time=0.1, value=0}, {arriveTangent=0.7869, leaveTangent=0.7869, time=1, value=1}]}}})
    Index 0 Param 3=&FloatFromCurve(FloatCurve=DataInterface() {propertyValues {Curve {keys [<at 0 5>,<at 1 1.5>]}}})
    Index 0 Param 4=&FloatFromCurve(FloatCurve=DataInterface() {propertyValues {Curve {keys [<at 0 5>,<at 1 5>]}}})
    Module.Index0Param0Write=true
    Module.Index0Param1Write=true
    Module.Index0Param2Write=true
    Module.Index0Param3Write=true
   )
   /SetVariables(Particles.Dynamic002:Vector4f=&MakeVector4_FromFloat(Float=&FloatFromCurve(FloatCurve=DataInterface() {propertyValues {Curve {keys [
          {arriveTangent=2.0325, interpMode=RCIM_Cubic, leaveTangent=2.0325, tangentMode=RCTM_User, time=0.4, value=0}
          {arriveTangent=0.0327, interpMode=RCIM_Cubic, leaveTangent=0.0327, tangentMode=RCTM_User, time=0.7, value=1.5}
         ]}}}))) {isSetParameters=true}
  ]
  rendererClasses [&NiagaraMeshRendererProperties, &NiagaraSpriteRendererProperties]
  renderers [
   {rendererClass=&NiagaraMeshRendererProperties, rendererIndex=0}
   {rendererClass=&NiagaraSpriteRendererProperties, rendererIndex=1}
  ]
 }
 Star_mesh {
  emitterUpdateScript [
   &EmitterState(Inactive Response=::NewEnumerator1, Life Cycle Mode=::NewEnumerator1, Loop Behavior=::NewEnumerator1, Loop Duration=1, UseLoopDelay=false)
   &SpawnBurst_Instantaneous(Spawn Time=0.3, Use Spawn Probability=false, UseLoopCountLimit=false)
  ]
  particleSpawnScript [
   &InitializeParticle(
    Color Mode=::NewEnumerator0
    Lifetime=0.6
    Mesh Scale=<10 2.5 2.5>
    Mesh Scale Mode=::NewEnumerator4
    UsePositionOffset=true
    Write Material Random=true
   )
   &InitialMeshOrientation(Mesh Orientation Mode=::NewEnumerator2, Rotation=<0 10 0>, Use Rotation Vector=true)
  ]
  particleUpdateScript [
   &ParticleState()
   &ScaleColor(
    Scale Alpha=&FloatFromCurve(
     FloatCurve=DataInterface() {propertyValues {Curve {keys [
         {arriveTangent=8.6813, interpMode=RCIM_Cubic, leaveTangent=8.6813, tangentMode=RCTM_User, time=0, value=0}
         {arriveTangent=-7.0112, interpMode=RCIM_Cubic, leaveTangent=-7.0112, tangentMode=RCTM_User, time=0.1484, value=1.6177}
         {arriveTangent=-2.4724, interpMode=RCIM_Cubic, leaveTangent=-2.4724, tangentMode=RCTM_User, time=0.3, value=0.75}
         {interpMode=RCIM_Cubic, leaveTangent=-0.3003, time=1, value=0}
        ]}}}
     Scale Curve=5
    )
    Scale RGB=&Multiply_VectorByFloat(
     Vector=&MakeVectorFromLinearColorRGB(Color=&ColorFromCurve(ColorCurve=DataInterface() {propertyValues {
         AlphaCurve {keys []}
         BlueCurve {keys [<at 0.0428 0 RCIM_Cubic>,<at 0.2609 0>,<at 0.6295 0.1615>,<at 0.9678 0.2233 RCIM_Cubic>]}
         GreenCurve {keys [<at 0.0428 0.5 RCIM_Cubic>,<at 0.2609 0.9998>,<at 0.6295 0.964>,<at 0.9678 0.5 RCIM_Cubic>]}
         RedCurve {keys [<at 0.0428 0.0398 RCIM_Cubic>,<at 0.2609 1.3>,<at 0.6295 1>,<at 0.9678 0 RCIM_Cubic>]}
        }}))
     Float=1
    )
    ScaleA=true
    ScaleRGB=true
   )
   /SetVariables(Particles.Color002=&MultiplyLinearColor(
     Linear Color=$Particles.Color
     Scale Factor=&MakeVector4_FromLinearColor(LinearColor=<rgba 0.3542 0.93 1 1>)
    )) {isSetParameters=true}
   /SetVariables(Particles.Size003:Vector3f=&Multiply_Vector(
     A=<0.5 0.5 0.5>
     B=&VectorFromFloat(Value=&FloatFromCurve(
       FloatCurve=DataInterface() {propertyValues {Curve {keys [{arriveTangent=1.8728, interpMode=RCIM_Cubic, leaveTangent=1.8728, tangentMode=RCTM_User, time=0, value=1}, <at 1 2 RCIM_Cubic>]}}}
       Scale Curve=0.95
      ))
    )) {isSetParameters=true}
   /SetVariables(Particles.Color003:LinearColor=&MultiplyLinearColor(
     Linear Color=<rgba 1.2 0.7326 0.1114 1>
     Scale Factor=&MakeVector4_FromFloat(Float=&FloatFromCurve(FloatCurve=DataInterface() {propertyValues {Curve {keys [
           {arriveTangent=0.9513, interpMode=RCIM_Cubic, leaveTangent=0.9513, tangentMode=RCTM_User, time=0, value=0}
           {arriveTangent=10.1951, interpMode=RCIM_Cubic, leaveTangent=-5.7722, tangentMode=RCTM_Break, time=0.1, value=1.5}
           {arriveTangent=-1.8916, interpMode=RCIM_Cubic, leaveTangent=-1.8916, tangentMode=RCTM_User, time=0.3, value=0.575}
           {arriveTangent=-0.2689, interpMode=RCIM_Cubic, leaveTangent=-0.2689, tangentMode=RCTM_User, time=1, value=0}
          ]}}}))
    )) {isSetParameters=true}
   &DynamicMaterialParameters(
    Index 0 Param 1=&FloatFromCurve(FloatCurve=DataInterface() {propertyValues {Curve {keys [
         {arriveTangent=-4.0585, interpMode=RCIM_Cubic, leaveTangent=-4.0585, tangentMode=RCTM_User, time=0, value=1}
         {arriveTangent=-1.0113, interpMode=RCIM_Cubic, leaveTangent=-1.0113, tangentMode=RCTM_User, time=1, value=-1.5}
        ]}}})
    Module.Index0Param0Write=true
    Module.Index0Param1Write=true
    Module.Index0Param2Write=true
    Module.Index0Param3Write=true
   )
  ]
  rendererClasses [&NiagaraMeshRendererProperties]
  renderers [
   {rendererClass=&NiagaraMeshRendererProperties, rendererIndex=0}
   {rendererClass=&NiagaraMeshRendererProperties, rendererIndex=1}
  ]
 }
}
summary {
 systemName=NS_WaywardRebel_Skin_Burst
 emitters [
  {emitterName=Star_Main, rendererClasses [&NiagaraMeshRendererProperties, &NiagaraSpriteRendererProperties]}
  {emitterName=Star_Glow, rendererClasses [&NiagaraSpriteRendererProperties]}
  {emitterName=Rays, rendererClasses [&NiagaraMeshRendererProperties]}
  {emitterName=Ray_Line, rendererClasses [&NiagaraMeshRendererProperties]}
  {emitterName=Star_mesh, rendererClasses [&NiagaraMeshRendererProperties]}
  {emitterName=Sprites_group, rendererClasses [&NiagaraSpriteRendererProperties]}
  {emitterName=Star_Glowmesh, rendererClasses [&NiagaraMeshRendererProperties]}
  {emitterName=Light, rendererClasses [&NiagaraLightRendererProperties]}
 ]
 userVariables [
  User.Hide light {type=NiagaraBool}
  User.IsLeft {type=NiagaraBool}
 ]
}
