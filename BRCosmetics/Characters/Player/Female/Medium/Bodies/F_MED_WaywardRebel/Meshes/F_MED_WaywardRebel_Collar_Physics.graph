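# Physics asset for the collar of the F_MED_WaywardRebel female player body cosmetic, defining a short jointed chain for collar motion.
# Capsule bodies run from base_collar_01_m through three dyn_collar bones, with the base constrained to spine_04 at zero angular limits.
# Dynamic joints allow small swings without twist, and the first has range on only one swing axis.
# Two capsules on spine_04 give the upper torso collision shapes.
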PhysicsAsset /BRCosmetics/Characters/Player/Female/Medium/Bodies/F_MED_WaywardRebel/Meshes/F_MED_WaywardRebel_Collar_Physics
bodies {
 base_collar_01_m [{shapeType=Capsule, length=0.505, radius=0.505, shapeName=base_collar_01_m_capsule, rotation=<rot 90 0 0>}]
 dyn_collar_01_m [{shapeType=Capsule, length=3.4019, radius=0.5679, shapeName=dyn_collar_01_m_capsule, center=<1.6267 0 0>, rotation=<rot 90 0 0>}]
 dyn_collar_02_m [{shapeType=Capsule, length=3.8857, radius=0.505, shapeName=dyn_collar_02_m_capsule, center=<2.0016 0 0>, rotation=<rot 90 0 0>}]
 dyn_collar_03_m [{shapeType=Capsule, length=3.484, radius=1.6053, shapeName=dyn_collar_03_m_capsule, center=<1.8838 0.3286 0.4578>, rotation=<rot 90 180 -140>}]
 spine_04 [
  {shapeType=Capsule, length=6.3879, radius=9.1533, shapeName=spine_04_capsule, center=<0.5828 3.6879 0>, rotation=<rot 0 -90 0>}
  {shapeType=Capsule, length=10.6484, radius=1.6887, shapeName=spine_04_capsule, center=<3.0388 11.6602 4.9713>, rotation=<rot -60 -0.0002 -30>}
 ]
}
constraints [
 {bone1Name=dyn_collar_01_m, bone2Name=base_collar_01_m, swing1LimitDegrees=15, swing2LimitDegrees=0, twistLimitDegrees=0}
 {bone1Name=dyn_collar_02_m, bone2Name=dyn_collar_01_m, swing1LimitDegrees=30, swing2LimitDegrees=10, twistLimitDegrees=0}
 {bone1Name=dyn_collar_03_m, bone2Name=dyn_collar_02_m, swing1LimitDegrees=15, swing2LimitDegrees=10, twistLimitDegrees=0}
 {bone1Name=base_collar_01_m, bone2Name=spine_04, swing1LimitDegrees=0, swing2LimitDegrees=0, twistLimitDegrees=0}
]
