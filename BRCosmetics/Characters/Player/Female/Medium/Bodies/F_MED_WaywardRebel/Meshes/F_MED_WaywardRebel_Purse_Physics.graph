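# Physics asset for the left-side purse on the F_MED_WaywardRebel female character body.
# A base bone locked to spine_02 anchors a three-bone dyn_purse chain with small swing limits and zero-degree twist limits.
# The last bone uses three side-by-side capsules to approximate the wide bag.
# Capsules on pelvis, spine, left thigh, arm and hand form the surrounding body, with locked joints except the free upper-arm-to-spine constraint.
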PhysicsAsset /BRCosmetics/Characters/Player/Female/Medium/Bodies/F_MED_WaywardRebel/Meshes/F_MED_WaywardRebel_Purse_Physics
bodies {
 base_purse_01_l [{shapeType=Capsule, length=0.505, radius=0.505, shapeName=base_purse_01_l_capsule, rotation=<rot 0 -90 0>}]
 dyn_purse_01_l [{shapeType=Capsule, length=6.4589, radius=2, shapeName=dyn_purse_01_l_capsule, center=<5.6568 0 0>, rotation=<rot 90 0 0>}]
 dyn_purse_02_l [{shapeType=Capsule, length=1.7759, radius=2, shapeName=dyn_purse_02_l_capsule, center=<1.6215 0 -0.3907>, rotation=<rot 90 0 0>}]
 dyn_purse_03_l [
  {shapeType=Capsule, length=6.1359, radius=2.2573, shapeName=dyn_purse_03_l_capsule, center=<2.9847 0 0.0223>, rotation=<rot 90 0 0>}
  {shapeType=Capsule, length=6.1359, radius=2.2573, shapeName=dyn_purse_03_l_capsule, center=<2.9847 4.2602 0.0223>, rotation=<rot 90 0 0>}
  {shapeType=Capsule, length=6.1359, radius=2.2573, shapeName=dyn_purse_03_l_capsule, center=<2.9847 -4.2319 0.0223>, rotation=<rot 90 0 0>}
 ]
 hand_l [{shapeType=Capsule, length=4.1636, radius=5.4557, shapeName=hand_l_capsule, center=<5.7537 -0.9975 -0.497>, rotation=<rot 90 0 0>}]
 lowerarm_l [{shapeType=Capsule, length=15.7237, radius=3.0036, shapeName=lowerarm_l_capsule, center=<9.1262 0 0>, rotation=<rot 90 0 0>}]
 pelvis [{shapeType=Capsule, length=9.9883, radius=10.1834, shapeName=spine_04_capsule, center=<0 -0.7693 0>, rotation=<rot 0 -90 0>}]
 spine_02 [{shapeType=Capsule, length=3.5124, radius=9.1533, shapeName=spine_04_capsule, center=<-1.4441 2.491 0>, rotation=<rot 0 -90 0>}]
 thigh_l [{
   shapeType=Capsule
   length=10.7346
   radius=7.972
   shapeName=thigh_l_capsule
   center=<-11.8313 -0.1592 0.0882>
   rotation=<rot -0.5988 90.752 -89.2678>
  }]
 upperarm_l [{shapeType=Capsule, length=10.3917, radius=3.0036, shapeName=lowerarm_l_capsule, center=<19.6996 0 0>, rotation=<rot 90 0 0>}]
}
constraints [
 {
  bone1Name=spine_02
  bone2Name=pelvis
  swing1LimitDegrees=0
  swing1Motion=Locked
  swing2LimitDegrees=0
  swing2Motion=Locked
  twistLimitDegrees=0
  twistMotion=Locked
 }
 {
  bone1Name=base_purse_01_l
  bone2Name=spine_02
  swing1LimitDegrees=15
  swing1Motion=Locked
  swing2LimitDegrees=15
  swing2Motion=Locked
  twistLimitDegrees=0
  twistMotion=Locked
 }
 {bone1Name=dyn_purse_01_l, bone2Name=base_purse_01_l, swing1LimitDegrees=10, swing2LimitDegrees=15, twistLimitDegrees=0}
 {bone1Name=dyn_purse_02_l, bone2Name=dyn_purse_01_l, swing1LimitDegrees=10, swing2LimitDegrees=15, twistLimitDegrees=0}
 {bone1Name=dyn_purse_03_l, bone2Name=dyn_purse_02_l, swing1LimitDegrees=10, swing2LimitDegrees=15, twistLimitDegrees=0}
 {
  bone1Name=thigh_l
  bone2Name=pelvis
  swing1LimitDegrees=45
  swing1Motion=Locked
  swing2LimitDegrees=45
  swing2Motion=Locked
  twistLimitDegrees=45
  twistMotion=Locked
 }
 {
  bone1Name=hand_l
  bone2Name=lowerarm_l
  swing1LimitDegrees=45
  swing1Motion=Locked
  swing2LimitDegrees=45
  swing2Motion=Locked
  twistLimitDegrees=45
  twistMotion=Locked
 }
 {
  bone1Name=lowerarm_l
  bone2Name=upperarm_l
  swing1LimitDegrees=45
  swing1Motion=Locked
  swing2LimitDegrees=45
  swing2Motion=Locked
  twistLimitDegrees=45
  twistMotion=Locked
 }
 {
  bone1Name=upperarm_l
  bone2Name=spine_02
  swing1LimitDegrees=45
  swing1Motion=Free
  swing2LimitDegrees=45
  swing2Motion=Free
  twistLimitDegrees=45
  twistMotion=Free
 }
]
