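# Post-process style AnimBlueprint for the F_MED_WaywardRebel player body, adding deformation and secondary motion to the pose copied from the parent mesh.
# A deform Control Rig feeds purse bone offsets, collar and purse RigidBody simulations, then cape-attach and upper-arm twist corrections via cached poses.
# The EventGraph derives alphas and gravity from weapon and movement states.
# Outside the front end, it also drives WPO cloth wind and turbulence material parameters.
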AnimBlueprint /BRCosmetics/Characters/Player/Female/Medium/Bodies/F_MED_WaywardRebel/Meshes/F_MED_WaywardRebel_AnimBP
target_skeleton=&/F_MED_WaywardRebel_Skeleton
graphs [
 AnimGraph {nodes [
   "Transform (Modify) Bone - Bone: base_purse_01_l"!(ComponentPose, bAlphaBoolEnabled:bool=True, Translation="0,-4,0", Rotation="0,0,0") -> (Pose) {
    alphaInputType=Bool
    rotationMode=BMM_Additive
    rotationSpace=BCS_BoneSpace
    translationMode=BMM_Additive
    translationSpace=BCS_BoneSpace
    alphaBoolBlend {blendOption=Linear}
    alphaScaleBiasClamp {bInterpResult=true}
    boneToModify {boneName=base_purse_01_l}
    translation {y=-4}
   }
   Output Pose(Result=Use cached pose 'CRCache_CapeFix'()) {layerGroup=None}
   n1 = Control Rig!(
    Source=Copy Pose From Mesh() {bUseAttachedParent=true}
    Alpha
    LargeGunEquipped:double=Get LargeGunEquipped() -> (LargeGunEquipped:double)
    JumpingSkydive:double=Get Floating() -> (Floating:double)
   ) -> (Pose) {
    alphaBoolBlend {blendOption=Linear}
    controlRigAssetReference {blueprintRigClass=&/F_MED_DeformRig_WaywardRebel_ControlRig.F_MED_DeformRig_WaywardRebel_ControlRig_C}
   }
   Save cached pose 'CRCache_Entry'(Pose=@n1) {cacheName=CRCache_Entry}
   n2 = Get GravityOverride()
   n3 = Get AlphaDyn() -> (AlphaDyn:double)
   n4 = RigidBody!(ComponentPose=Local To Component(LocalPose=Use cached pose 'CRCache_Fix'()), Alpha=@n3, OverrideWorldGravity:Vector=@n2, ExternalForce) -> (Pose) {
    bClampLinearTranslationLimitToRefPose=true
    bOverrideWorldGravity=true
    overridePhysicsAsset=&/F_MED_WaywardRebel_Collar_Physics
    simulationSpace=BaseBoneSpace
    alphaBoolBlend {blendOption=Linear}
    baseBoneRef {boneName=spine_04}
    componentAppliedLinearAccClamp=<2000 2000 1000>
    componentLinearAccScale {x=0.2}
    componentLinearVelScale {x=0.2}
    overrideWorldGravity {z=-980}
    simSpaceSettings {maxAngularAcceleration=100, maxAngularVelocity=5, maxLinearAcceleration=300, maxLinearVelocity=50, velocityScaleZ=0, worldAlpha=0.35}
   }
   n5 = RigidBody!(
    ComponentPose=@n4
    Alpha=@n3
    OverrideWorldGravity=@n2
    SimSpaceSettings:Sim Space Settings Structure=(WorldAlpha=0.8,VelocityScaleZ=1,DampingAlpha=1,MaxLinearVelocity=500,MaxAngularVelocity=3,MaxLinearAcceleration=700,MaxAngularAcceleration=1000,ExternalLinearDragV=(X=0.2,Y=0.2,Z=0.1),ExternalLinearVelocity=(X=0,Y=0,Z=0),ExternalAngularVelocity=(X=0,Y=0,Z=0))
   ) -> (Pose) {
    bOverrideWorldGravity=true
    overridePhysicsAsset=&/F_MED_WaywardRebel_Purse_Physics
    simulationSpace=BaseBoneSpace
    alphaBoolBlend {blendOption=Linear}
    baseBoneRef {boneName=spine_01}
    componentAppliedLinearAccClamp=<3000 3000 1000>
    componentLinearAccScale=<0.25 0.25 0.1>
    componentLinearVelScale=<0.25 0.25 0.1>
    overrideWorldGravity {z=-980}
    simSpaceSettings {
     maxAngularAcceleration=1000
     maxAngularVelocity=3
     maxLinearAcceleration=700
     maxLinearVelocity=500
     worldAlpha=0.8
     externalLinearDragV=<0.2 0.2 0.1>
    }
   }
   n6 = "Transform (Modify) Bone - Bone: base_purse_01_l"!(ComponentPose=@n5, Alpha:float=1, Translation="0,1.5,0", Rotation="0,-10,0") -> (Pose) {
    rotationMode=BMM_Additive
    rotationSpace=BCS_BoneSpace
    translationMode=BMM_Additive
    translationSpace=BCS_BoneSpace
    alphaBoolBlend {blendOption=Linear}
    alphaScaleBiasClamp {bInterpResult=true}
    boneToModify {boneName=base_purse_01_l}
    rotation {yaw=-10}
    translation {y=1.5}
   }
   Save cached pose 'CRCache_Dyn'(Pose=Component To Local(ComponentPose=@n6)) {cacheName=CRCache_Dyn}
   n7 = "Transform (Modify) Bone - Bone: attach_cape"!(
    ComponentPose=Local To Component(LocalPose=Use cached pose 'CRCache_Dyn'())
    Alpha=Get CapeFixer() -> (CapeFixer:float)
    Translation="0,-2.75,0"
   ) -> (Pose) {
    rotationMode=BMM_Additive
    rotationSpace=BCS_BoneSpace
    scaleMode=BMM_Additive
    scaleSpace=BCS_BoneSpace
    translationMode=BMM_Additive
    translationSpace=BCS_BoneSpace
    alphaBoolBlend {blendOption=Linear}
    alphaScaleBiasClamp {bInterpResult=true}
    boneToModify {boneName=attach_cape}
    translation {y=-2.75}
   }
   n8 = "Transform (Modify) Bone - Bone: attach_cape"(ComponentPose=@n7, Alpha=1, Translation="0,-2.75,0") {
    rotationMode=BMM_Additive
    rotationSpace=BCS_BoneSpace
    scaleMode=BMM_Additive
    scaleSpace=BCS_BoneSpace
    translationMode=BMM_Additive
    translationSpace=BCS_BoneSpace
    alphaBoolBlend {blendOption=Linear}
    alphaScaleBiasClamp {bInterpResult=true, interpSpeedDecreasing=4, interpSpeedIncreasing=6}
    boneToModify {boneName=attach_cape}
    translation {y=-2.75}
   }
   n9 = "Transform (Modify) Bone - Bone: upperarm_twist_01_l"!(ComponentPose=@n8, Alpha=1, Rotation="0,0,-30") -> (Pose) {
    rotationMode=BMM_Additive
    rotationSpace=BCS_BoneSpace
    translationMode=BMM_Additive
    translationSpace=BCS_BoneSpace
    alphaBoolBlend {blendOption=Linear}
    alphaScaleBiasClamp {bInterpResult=true}
    boneToModify {boneName=upperarm_twist_01_l}
    rotation {roll=-30}
   }
   n10 = "Transform (Modify) Bone - Bone: upperarm_twist_01_r"!(ComponentPose=@n9, Alpha=1, Rotation="0,0,-30") -> (Pose) {
    rotationMode=BMM_Additive
    rotationSpace=BCS_BoneSpace
    translationMode=BMM_Additive
    translationSpace=BCS_BoneSpace
    alphaBoolBlend {blendOption=Linear}
    alphaScaleBiasClamp {bInterpResult=true}
    boneToModify {boneName=upperarm_twist_01_r}
    rotation {roll=-30}
   }
   Save cached pose 'CRCache_CapeFix'(Pose=Component To Local(ComponentPose=@n10)) {cacheName=CRCache_CapeFix}
   n11 = "Transform (Modify) Bone - Bone: base_purse_01_l"(ComponentPose=Local To Component(LocalPose=Use cached pose 'CRCache_Entry'()), Alpha=1, Translation="0,0,3", Rotation="5,0,0") {
    rotationMode=BMM_Additive
    rotationSpace=BCS_BoneSpace
    translationMode=BMM_Additive
    translationSpace=BCS_BoneSpace
    alphaBoolBlend {blendOption=Linear}
    alphaScaleBiasClamp {bInterpResult=true}
    boneToModify {boneName=base_purse_01_l}
    rotation {pitch=5}
    translation {z=3}
   }
   n12 = "Transform (Modify) Bone - Bone: base_purse_01_l"!(ComponentPose=@n11, Alpha=1, Rotation="5,0,0") -> (Pose) {
    rotationMode=BMM_Additive
    rotationSpace=BCS_BoneSpace
    translationMode=BMM_Additive
    translationSpace=BCS_BoneSpace
    alphaBoolBlend {blendOption=Linear}
    alphaScaleBiasClamp {bInterpResult=true}
    boneToModify {boneName=base_purse_01_l}
    rotation {pitch=5}
   }
   n13 = "Transform (Modify) Bone - Bone: base_purse_01_l"!(ComponentPose=@n12, Alpha=1, Translation:Vector="0,0,3") -> (Pose) {
    rotationMode=BMM_Additive
    rotationSpace=BCS_BoneSpace
    translationMode=BMM_Additive
    translationSpace=BCS_BoneSpace
    alphaBoolBlend {blendOption=Linear}
    alphaScaleBiasClamp {bInterpResult=true}
    boneToModify {boneName=base_purse_01_l}
    translation {z=3}
   }
   n14 = "Transform (Modify) Bone - Bone: base_purse_01_l"!(ComponentPose=@n13, Alpha=1, Translation="0,0,3", Rotation:Rotator="0,-10,-10") -> (Pose) {
    rotationMode=BMM_Additive
    rotationSpace=BCS_BoneSpace
    translationMode=BMM_Additive
    translationSpace=BCS_BoneSpace
    alphaBoolBlend {blendOption=Linear}
    alphaScaleBiasClamp {bInterpResult=true}
    boneToModify {boneName=base_purse_01_l}
    rotation {roll=-10, yaw=-10}
    translation {z=3}
   }
   n15 = "Transform (Modify) Bone - Bone: base_purse_01_l"(ComponentPose=@n14, Alpha=1, Translation="0,0,2", Rotation="10,0,0") {
    rotationMode=BMM_Additive
    rotationSpace=BCS_BoneSpace
    translationMode=BMM_Additive
    translationSpace=BCS_BoneSpace
    alphaBoolBlend {blendOption=Linear}
    alphaScaleBiasClamp {bInterpResult=true}
    boneToModify {boneName=base_purse_01_l}
    rotation {pitch=10}
    translation {z=2}
   }
   Save cached pose 'CRCache_Fix'(Pose=Component To Local(ComponentPose=@n15)) {cacheName=CRCache_Fix}
  ]}
 EventGraph {nodes [
   Event BlueprintInitializeAnimation()
   > Set SkelMesh(SkelMesh=GetOwnerSkeletalMeshComponent())
   n1 = TryGetPawnOwner()
   > Is Valid(InputObject=@n1)
   > n2 = Cast To FortPlayerPawnAthena(Object=@n1)
   > n3 = Set PlayerPawn(PlayerPawn:Fort Player Pawn Athena Object Reference=@n2) -> (Output_Get:Fort Player Pawn Athena Object Reference)
   > Set Body_Mesh(Body_Mesh=GetSkeletalMeshForPartType(self=@n3))
   n4 = Event BlueprintProcessAnimInputEvent()
   > n5 = Branch(Condition=IsNamed(self=@n4.AnimInputEvent, InEventName=aie.Player.Zipline.Entered))
   > Set OnZiplineFloat(OnZiplineFloat:double=1) -> (Output_Get:double)
   Branch(execute=@n5.else, Condition=IsNamed(self=@n4.AnimInputEvent, InEventName=aie.Player.Zipline.Exited))
   > Set OnZiplineFloat(OnZiplineFloat=0)
   n6 = Event BlueprintUpdateAnimation()
   > n7 = Sequence()
   n8 = Sequence!(execute=@n7.then_1) -> (then_0, then_1, then_2:Exec, then_3:Exec, then_4:Exec, then_5:Exec, then_6:Exec, then_7:Exec)
   n9 = OR Boolean(A=Get bIsJumping(), B=Get bIsFalling())
   n10 = Branch(execute=@n8.then_6, Condition=AND Boolean(A=NOT Boolean(A=@n9), B=Get bIsPlayingMeleeAnim()))
   > Set AtkMelee(AtkMelee:double=1) -> (Output_Get:double)
   Set AtkMelee(execute=@n10.else, AtkMelee=0)
   n11 = Cast To FortPlayerAnimInstance(execute=@n8.then_1, Object=GetAnimInstance(self=GetOwnerSkeletalMeshComponent()))
   > Set ar_firingPosition_GC(ar_firingPosition_GC:int=0) -> (Output_Get:int=0)
   > n12 = Switch on EFortWeaponCoreAnimation!(execute, Selection:EFortWeaponCoreAnimation Enum=Get WeaponCoreAnim(self=@n11)) -> (
    Melee:Exec
    Pistol:Exec
    Shotgun:Exec
    PaperBlueprint:Exec
    Rifle:Exec
    MeleeOneHand:Exec
    MachinePistol:Exec
    RocketLauncher:Exec
    GrenadeLauncher:Exec
    GoingCommando:Exec
    AssaultRifle:Exec
    TacticalShotgun:Exec
    SniperRifle:Exec
    TrapPlacement:Exec
    ShoulderLauncher:Exec
    AbilityDecoTool:Exec
    Crossbow:Exec
    C4:Exec
    RemoteControl:Exec
    DualWield:Exec
    AR_BullPup:Exec
    AR_ForwardGrip:Exec
    MedPackPaddles:Exec
    SMG_P90:Exec
    AR_DrumGun:Exec
    Consumable_Small:Exec
    Consumable_Large:Exec
    Balloon:Exec
    MountedTurret:Exec
    CreativeTool:Exec
    ExplosiveBow:Exec
    AshtonIndigo:Exec
    AshtonChicago:Exec
    MeleeDualWield:Exec
    Unarmed:Exec
   )
   Branch(execute=@n12.AssaultRifle)
   > n13 = Set ar_firingPosition_GC(ar_firingPosition_GC=1) -> (Output_Get=0)
   > Set Gun_out(Gun_out:bool="To Boolean (Integer)"!(InInt:int=@n13) -> (ReturnValue:bool)) -> (Output_Get:bool)
   Branch(execute=[
     @n12.GoingCommando
     @n12.MachinePistol
     @n12.Melee
     @n12.Pistol
     @n12.Rifle
     @n12.RocketLauncher
     @n12.Shotgun
     @n12.SniperRifle
     @n12.TacticalShotgun
    ])
   > n14 = Set ar_firingPosition_GC(ar_firingPosition_GC=0) -> (Output_Get=0)
   > Set Gun_out(Gun_out="To Boolean (Integer)"(InInt=@n14))
   Set Floating(execute=@n8.then_7, Floating:double="To Float (Boolean)"(InBool=Get bIsSkydiveFloating())) -> (Output_Get:double)
   Set Crouching(execute=@n8.then_3, Crouching:double="To Float (Boolean)"(InBool=Get bIsCrouching())) -> (Output_Get:double)
   > Set KneeSlide(KneeSlide:double="To Float (Boolean)"(InBool=Get bIsKneeSliding())) -> (Output_Get:double)
   n15 = OR Boolean(A=Get bIsParachuteOpened(), B=Get bIsPlayingEmote())
   n16 = Get UpperArmY_Rot_R() -> (UpperArmY_Rot_R:bool)
   > Set TwistArmFix(TwistArmFix:double="To Float (Boolean)"(InBool=AND Boolean(A=@n16, B=@n15))) -> (Output_Get:double)
   n17 = Get bIsParachuteOpened()
   n18 = Get bIsSkydiveFloating()
   n19 = Get bIsSkyDiving()
   n20 = Get bIsSkydiveDiveMode()
   > Set isInAir(isInAir:double="To Float (Boolean)"(InBool=OR Boolean!(A=@n20, B=@n19, C:bool=@n18, D:bool=@n17) -> (ReturnValue))) -> (Output_Get:double)
   n21 = NOT Boolean(A=Get bIsCrouching())
   n22 = Get bRigidBodyEnabled()
   > Set PurseOffsetAlpha(PurseOffsetAlpha:double="To Float (Boolean)"(InBool=AND Boolean(A=@n22, B=@n21))) -> (Output_Get:double)
   n23 = Get bRigidBodyEnabled()
   n24 = Get bIsInFrontEnd()
   > Set FrontEndLevel(FrontEndLevel:double="To Float (Boolean)"(InBool=AND Boolean(A=@n24, B=@n23))) -> (Output_Get:double)
   > n25 = Branch(Condition=OR Boolean(A=Get bIsSkydiveDiveMode(), B=Get bIsSkyDiving(), C=Get bIsSkydiveFloating(), D=Get bIsParachuteOpened()))
   > Set AlphaDyn(AlphaDyn:double=0.5) -> (Output_Get:double)
   Set AlphaDyn(execute=@n25.else, AlphaDyn=1)
   n26 = Branch(execute=@n8.then_2, Condition=Get Gun_out() -> (Gun_out:bool))
   > Set LargeGunEquipped(LargeGunEquipped:double=1) -> (Output_Get:double)
   Set LargeGunEquipped(execute=@n26.else, LargeGunEquipped=0)
   n27 = Get bIsSkydiveFloating()
   n28 = Get bIsSkydiveDiveMode()
   n29 = Get bIsSkyDiving()
   n30 = Get bIsParachuteOpened()
   n31 = Branch(execute=@n8.then_4, Condition=OR Boolean!(A=OR Boolean!(A=@n29, B=@n28, C:bool=@n27, D:bool=@n30) -> (ReturnValue), B=@n9) -> (ReturnValue))
   > Set CapeFixer(CapeFixer:float=1) -> (Output_Get:float)
   Set CapeFixer(execute=@n31.else, CapeFixer=0)
   n32 = Branch(execute=@n8.then_5, Condition=Get bIsInFrontEnd())
   > Set Sim Space Settings(Sim Space Settings:Sim Space Settings Structure=Get Sim Space Settings FE() -> (Sim Space Settings FE:Sim Space Settings Structure)) -> (Output_Get:Sim Space Settings Structure)
   n33 = Get bIsMoving2D()
   n34 = Branch(execute=@n32.else, Condition=AND Boolean(A=Get bIsCrouching(), B=NOT Boolean(A=@n33)))
   > Set Sim Space Settings(Sim Space Settings=Get Sim Space Settings Crouch() -> (Sim Space Settings Crouch:Sim Space Settings Structure))
   Set Sim Space Settings(execute=@n34.else, Sim Space Settings=Get Sim Space Settings Regular() -> (Sim Space Settings Regular:Sim Space Settings Structure))
   n35 = Gravity Override State Function DEPRECATED(
    execute=@n8.then_0
    CharacterPartAnimInstance=Self-Reference()
    GameTimeInSeconds=GetGameTimeInSeconds()
    GravityOverrideParams=Get Gravity Override Params()
   )
   > Set GravityOverride(GravityOverride=@n35)
   n36 = BreakTransform(InTransform=GetDeltaTransformFromRefPose(self=GetOwnerSkeletalMeshComponent(), BoneName=clavicle_l, BaseName=spine_05))
   > n37 = Branch(Condition="float >= float"(A=BreakRotator(InRot=@n36.Rotation).Pitch, B=45))
   > Set UpperArmY_Rot_R(UpperArmY_Rot_R:bool=true) -> (Output_Get:bool)
   Set UpperArmY_Rot_R(execute=@n37.else, UpperArmY_Rot_R=false)
   Is Valid(exec=@n7.then_0, InputObject=Get PlayerPawn() -> (PlayerPawn:Fort Player Pawn Athena Object Reference))
   > n38 = Branch(Condition=Get bIsInFrontEnd())
   Set WPO Globals Function(
    execute=@n38.else
    Material=Get Mat Instance Dynamic()
    Overlay Material=Get Overlay Mat Instance Dynamic()
    Skeletal Mesh=Get SkelMesh()
    "Local Noise Transform  Location"=Get Local Noise Transform Location() -> (Local Noise Transform Location:Vector)
    Local Noise Offset=Get Local Noise Offset()
   )
   > Set WPO Regional Function!(
    execute
    Material:Material Instance Dynamic Object Reference=Get Mat Instance Dynamic()
    Overlay Material:Material Instance Dynamic Object Reference=Get Overlay Mat Instance Dynamic()
    Skeletal Mesh:Skeletal Mesh Component Object Reference=Get SkelMesh()
    Velocity Transform:Name=foot_r
    Velocity Interp Speed:double=Get Velocity Interp Speed()
    Collision Transform:Name=calf_r
    Collision Transform Axis:EAxis Enum
    Invert Collision Transform Axis:bool=false
    Delta Time X:double=@n6.DeltaTimeX
    Previous Velocity Transform Location:"Vector (by ref)"=Get R1 Velocity Transform Location()
    Previous Velocity Transform Velocity:"Vector (by ref)"=Get R1 Velocity Transform Velocity()
    Region:int=1
   ) -> (then)
   > Set WPO Regional Function(
    Material=Get Mat Instance Dynamic()
    Overlay Material=Get Overlay Mat Instance Dynamic()
    Skeletal Mesh=Get SkelMesh()
    Velocity Transform=foot_l
    Velocity Interp Speed=Get Velocity Interp Speed()
    Collision Transform=calf_l
    Invert Collision Transform Axis=true
    Delta Time X=@n6.DeltaTimeX
    Previous Velocity Transform Location=Get R2 Velocity Transform Location()
    Previous Velocity Transform Velocity=Get R2 Velocity Transform Velocity()
    Region=2
   )
   > n39 = Branch(Condition=OR Boolean!(A=Get bIsSkydiveFloating(), B=Get bIsParachuteOpened()) -> (ReturnValue))
   > Set WPO Pho Wind Intensity(WPO Pho Wind Intensity=0.8)
   > n40 = Set WPO Pho Turbulence(WPO Pho Turbulence:double=1.2) -> (Output_Get:double)
   SetScalarParameterValue(
    execute=[@n40.then, @n42.then]
    self=[Get Mat Instance Dynamic(), Get Overlay Mat Instance Dynamic()]
    ParameterName=ClothSim_WindStrength
    Value=Get WPO Pho Wind Intensity()
   )
   n41 = Get WPO Pho Turbulence() -> (WPO Pho Turbulence:double)
   > SetScalarParameterValue(self=[Get Mat Instance Dynamic(), Get Overlay Mat Instance Dynamic()], ParameterName=ClothSim_R1_TurbulenceStrength, Value=@n41)
   > SetScalarParameterValue(self=[Get Overlay Mat Instance Dynamic(), Get Mat Instance Dynamic()], ParameterName=ClothSim_R2_TurbulenceStrength, Value=@n41)
   Set WPO Pho Wind Intensity(execute=@n39.else, WPO Pho Wind Intensity=0)
   > n42 = Set WPO Pho Turbulence(WPO Pho Turbulence=0.1)
   Event OnCharacterCustomizationLoaded()
   > n43 = Set SkelMesh(SkelMesh=GetPartSkeletalMeshComponent())
   > n44 = Get Dynamic Material Instance(SkeletalMeshComponent=@n43)
   > Set Mat Instance Dynamic(Mat Instance Dynamic=@n44)
   > n45 = GetOrCreateDynamicOverlayMaterial(MeshComponent=@n43, MaterialName=None)
   > Set Overlay Mat Instance Dynamic(Overlay Mat Instance Dynamic=@n45)
  ]}
]
variables [
 AlphaDyn {type=number}
 ar_firingPosition_GC {type=integer}
 AtkMelee {type=number}
 Body_Mesh {type=SkeletalMeshComponent}
 CapeFixer {type=number}
 Crouching {type=number}
 Floating {type=number}
 FrontEndLevel {type=number}
 Gravity Override Params {type=GravityOverrideParamsStruct}
 GravityOverride {type=Vector}
 Gun_out {type=boolean}
 isInAir {type=number}
 Jumping {type=number}
 KneeSlide {type=number}
 LargeGunEquipped {type=number}
 Local Noise Offset {type=Vector}
 Local Noise Transform Location {type=Vector}
 Mat Instance Dynamic {type=MaterialInstanceDynamic}
 OnZiplineFloat {type=number}
 Overlay Mat Instance Dynamic {type=MaterialInstanceDynamic}
 PlayerPawn {type=FortPlayerPawnAthena}
 PurseOffsetAlpha {type=number}
 R1 Velocity Transform Location {type=Vector}
 R1 Velocity Transform Velocity {type=Vector}
 R2 Velocity Transform Location {type=Vector}
 R2 Velocity Transform Velocity {type=Vector}
 Sim Space Settings {
  type=SimSpaceSettings
  default {
   dampingAlpha=1
   maxAngularAcceleration=1000
   maxAngularVelocity=3
   maxLinearAcceleration=700
   maxLinearVelocity=500
   velocityScaleZ=1
   worldAlpha=1
   externalLinearDragV=<0.2 0.2 0.1>
  }
 }
 Sim Space Settings Crouch {
  type=SimSpaceSettings
  default {
   dampingAlpha=1
   maxAngularAcceleration=1000
   maxAngularVelocity=7
   maxLinearAcceleration=700
   maxLinearVelocity=500
   velocityScaleZ=2
   worldAlpha=0.6
   externalLinearDragV=<0 0 0>
  }
 }
 Sim Space Settings FE {
  type=SimSpaceSettings
  default {
   dampingAlpha=1
   maxAngularAcceleration=1000
   maxAngularVelocity=3
   maxLinearAcceleration=700
   maxLinearVelocity=500
   velocityScaleZ=1
   worldAlpha=0.15
   externalLinearDragV=<0.2 0.2 0.1>
  }
 }
 Sim Space Settings Regular {
  type=SimSpaceSettings
  default {
   dampingAlpha=1
   maxAngularAcceleration=1000
   maxAngularVelocity=3
   maxLinearAcceleration=700
   maxLinearVelocity=500
   velocityScaleZ=1
   worldAlpha=0.3
   externalLinearDragV=<0.2 0.2 0.1>
  }
 }
 SkelMesh {type=SkeletalMeshComponent}
 TwistArmFix {type=number}
 UpperArmY_Rot_R {type=boolean}
 Velocity Interp Speed {type=number}
 WPO Pho Turbulence {type=number}
 WPO Pho Wind Intensity {type=number}
]
