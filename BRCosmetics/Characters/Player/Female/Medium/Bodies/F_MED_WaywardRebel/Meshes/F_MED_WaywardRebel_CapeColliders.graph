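# Physics asset of colliders for the cape on the female medium WaywardRebel body, likely for cape simulation to collide against.
# Capsules cover pelvis, layered spine, clavicles, head, feet and left purse bone, while arm, thigh, calf and neck bodies have no shapes.
# A huge, near-spherical root capsule sits below the character, apparently acting as a floor.
# Constraints link the body hierarchy with all angular motion free; the purse body has none.
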PhysicsAsset /BRCosmetics/Characters/Player/Female/Medium/Bodies/F_MED_WaywardRebel/Meshes/F_MED_WaywardRebel_CapeColliders
bodies {
 calf_l []
 calf_r []
 clavicle_l [{shapeType=Capsule, length=7.6964, radius=6.3853, center=<4.0147 -0.8354 -1.7419>, rotation=<rot -77.5566 177.3915 7.7314>}]
 clavicle_r [{shapeType=Capsule, length=7.6964, radius=8.2957, center=<-4.0147 0.8354 1.7419>, rotation=<rot -77.5565 177.3915 -172.2685>}]
 dyn_purse_03_l [{shapeType=Capsule, length=4.0257, radius=6.5597, shapeName=dyn_purse_03_l_capsule, center=<2.981 0 0>, rotation=<rot 90 180 90>}]
 foot_l [{shapeType=Capsule, length=12.5982, radius=6.5127, center=<-2.2072 6.0534 -0.3367>, rotation=<rot 0.0388 1.4701 -89.3982>}]
 foot_r [{shapeType=Capsule, length=12.5982, radius=6.5127, center=<2.2072 -6.0534 0.3367>, rotation=<rot 0.0388 1.4701 90.6018>}]
 head [{shapeType=Capsule, length=0, radius=18, center=<9.0697 0.0746 0>, rotation=<rot 0 -90 0>}]
 lowerarm_l []
 lowerarm_r []
 neck_02 []
 pelvis [{shapeType=Capsule, length=9.5, radius=15, shapeName=pelvis_capsule, center=<-1.5557 0 0>}]
 root [{shapeType=Capsule, length=3, radius=298, shapeName=root_sphere, center=<0 -24.5697 -289.8553>, rotation=<rot 0 -90 0>}]
 spine_01 [
  {shapeType=Capsule, length=15.0484, radius=8.4854, center=<1.179 2.2974 0>}
  {shapeType=Capsule, length=20.6875, radius=9.3125, center=<-1.8317 2.606 0>, rotation=<rot 90 0 0>}
 ]
 spine_02 [
  {shapeType=Capsule, length=9.5205, radius=12.0298, center=<1.6253 2.2974 0>}
  {shapeType=Capsule, length=20.6875, radius=9.3125, center=<-1.8317 2.606 0>, rotation=<rot 90 0 0>}
 ]
 spine_03 [
  {shapeType=Capsule, length=9.2984, radius=8.4854, center=<-0.0747 2.3725 0>}
  {shapeType=Capsule, length=20.7188, radius=9.2812, center=<-1.2326 2.6059 0>, rotation=<rot 90 0 0>}
 ]
 spine_04 [
  {shapeType=Capsule, length=5.8296, radius=10.3916, center=<5.5182 1.4112 0>}
  {shapeType=Capsule, length=10.7537, radius=12.2023, center=<-1.2326 1.308 0>, rotation=<rot 90 0 0>}
 ]
 spine_05 [{shapeType=Capsule, length=10.9859, radius=10.0791, center=<5.6377 0.3227 0>}]
 thigh_l []
 thigh_r []
 upperarm_l []
 upperarm_r []
}
constraints [
 {
  bone1Name=spine_01
  bone2Name=pelvis
  swing1LimitDegrees=45
  swing1Motion=Free
  swing2LimitDegrees=45
  swing2Motion=Free
  twistLimitDegrees=45
  twistMotion=Free
 }
 {
  bone1Name=spine_02
  bone2Name=spine_01
  swing1LimitDegrees=45
  swing1Motion=Free
  swing2LimitDegrees=45
  swing2Motion=Free
  twistLimitDegrees=45
  twistMotion=Free
 }
 {
  bone1Name=spine_03
  bone2Name=spine_02
  swing1LimitDegrees=45
  swing1Motion=Free
  swing2LimitDegrees=45
  swing2Motion=Free
  twistLimitDegrees=45
  twistMotion=Free
 }
 {
  bone1Name=spine_04
  bone2Name=spine_03
  swing1LimitDegrees=45
  swing1Motion=Free
  swing2LimitDegrees=45
  swing2Motion=Free
  twistLimitDegrees=45
  twistMotion=Free
 }
 {
  bone1Name=spine_05
  bone2Name=spine_04
  swing1LimitDegrees=45
  swing1Motion=Free
  swing2LimitDegrees=45
  swing2Motion=Free
  twistLimitDegrees=45
  twistMotion=Free
 }
 {
  bone1Name=clavicle_l
  bone2Name=spine_05
  swing1LimitDegrees=45
  swing1Motion=Free
  swing2LimitDegrees=45
  swing2Motion=Free
  twistLimitDegrees=45
  twistMotion=Free
 }
 {
  bone1Name=upperarm_l
  bone2Name=clavicle_l
  swing1LimitDegrees=45
  swing1Motion=Free
  swing2LimitDegrees=45
  swing2Motion=Free
  twistLimitDegrees=45
  twistMotion=Free
 }
 {
  bone1Name=lowerarm_l
  bone2Name=upperarm_l
  swing1LimitDegrees=45
  swing1Motion=Free
  swing2LimitDegrees=45
  swing2Motion=Free
  twistLimitDegrees=45
  twistMotion=Free
 }
 {
  bone1Name=neck_02
  bone2Name=spine_05
  swing1LimitDegrees=45
  swing1Motion=Free
  swing2LimitDegrees=45
  swing2Motion=Free
  twistLimitDegrees=45
  twistMotion=Free
 }
 {
  bone1Name=head
  bone2Name=neck_02
  swing1LimitDegrees=45
  swing1Motion=Free
  swing2LimitDegrees=45
  swing2Motion=Free
  twistLimitDegrees=45
  twistMotion=Free
 }
 {
  bone1Name=foot_l
  bone2Name=calf_l
  swing1LimitDegrees=45
  swing1Motion=Free
  swing2LimitDegrees=45
  swing2Motion=Free
  twistLimitDegrees=45
  twistMotion=Free
 }
 {
  bone1Name=clavicle_r
  bone2Name=spine_05
  swing1LimitDegrees=45
  swing1Motion=Free
  swing2LimitDegrees=45
  swing2Motion=Free
  twistLimitDegrees=45
  twistMotion=Free
 }
 {
  bone1Name=lowerarm_r
  bone2Name=upperarm_r
  swing1LimitDegrees=45
  swing1Motion=Free
  swing2LimitDegrees=45
  swing2Motion=Free
  twistLimitDegrees=45
  twistMotion=Free
 }
 {
  bone1Name=foot_r
  bone2Name=calf_r
  swing1LimitDegrees=45
  swing1Motion=Free
  swing2LimitDegrees=45
  swing2Motion=Free
  twistLimitDegrees=45
  twistMotion=Free
 }
 {
  bone1Name=upperarm_r
  bone2Name=clavicle_r
  swing1LimitDegrees=45
  swing1Motion=Free
  swing2LimitDegrees=45
  swing2Motion=Free
  twistLimitDegrees=45
  twistMotion=Free
 }
 {
  bone1Name=thigh_l
  bone2Name=pelvis
  swing1LimitDegrees=45
  swing1Motion=Free
  swing2LimitDegrees=45
  swing2Motion=Free
  twistLimitDegrees=45
  twistMotion=Free
 }
 {
  bone1Name=thigh_r
  bone2Name=pelvis
  swing1LimitDegrees=45
  swing1Motion=Free
  swing2LimitDegrees=45
  swing2Motion=Free
  twistLimitDegrees=45
  twistMotion=Free
 }
 {
  bone1Name=calf_r
  bone2Name=thigh_r
  swing1LimitDegrees=45
  swing1Motion=Free
  swing2LimitDegrees=45
  swing2Motion=Free
  twistLimitDegrees=45
  twistMotion=Free
 }
 {
  bone1Name=calf_l
  bone2Name=thigh_l
  swing1LimitDegrees=45
  swing1Motion=Free
  swing2LimitDegrees=45
  swing2Motion=Free
  twistLimitDegrees=45
  twistMotion=Free
 }
 {
  bone1Name=pelvis
  bone2Name=root
  swing1LimitDegrees=45
  swing1Motion=Free
  swing2LimitDegrees=45
  swing2Motion=Free
  twistLimitDegrees=45
  twistMotion=Free
 }
]
