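# Physics asset for the back hair of the female medium WaywardRebel character, defining simulation bodies for three hair strands.
# Left, center and right strands are identical five-capsule chains, each root locked to neck_01, with twist-free links whose swing limits are tightest near the root.
# Capsules on the neck, upper spine, upper arms and forearms supply body shapes around the hair; the arm capsules have no constraints.
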PhysicsAsset /BRCosmetics/Characters/Player/Female/Medium/Bodies/F_MED_WaywardRebel/Meshes/Parts/F_MED_WaywardRebel_FaceAcc_BackHair_Physics
bodies {
 dyn_Hair_BackL_01 [{shapeType=Capsule, length=0, radius=2, shapeName=dyn_Hair_BackL_01_capsule, rotation=<rot 90 0 0>}]
 dyn_Hair_BackL_02 [{shapeType=Capsule, length=2.4027, radius=2, shapeName=dyn_Hair_BackL_02_capsule, center=<2.0336 0 0>, rotation=<rot 90 0 0>}]
 dyn_Hair_BackL_03 [{shapeType=Capsule, length=3.2215, radius=2, shapeName=dyn_Hair_BackL_03_capsule, center=<2.3848 0 0>, rotation=<rot 90 0 0>}]
 dyn_Hair_BackL_04 [{shapeType=Capsule, length=5, radius=2, shapeName=dyn_Hair_BackL_04_capsule, center=<3.3553 0 0>, rotation=<rot 90 0 0>}]
 dyn_Hair_BackL_05 [{shapeType=Capsule, length=5, radius=2, shapeName=dyn_Hair_BackL_05_capsule, center=<3.3034 0 0>, rotation=<rot 90 0 0>}]
 dyn_Hair_BackR_01 [{shapeType=Capsule, length=0, radius=2, shapeName=dyn_Hair_BackR_01_capsule, rotation=<rot 90 0 0>}]
 dyn_Hair_BackR_02 [{shapeType=Capsule, length=2.4027, radius=2, shapeName=dyn_Hair_BackR_02_capsule, center=<2.0336 0 0>, rotation=<rot 90 0 0>}]
 dyn_Hair_BackR_03 [{shapeType=Capsule, length=3.2215, radius=2, shapeName=dyn_Hair_BackR_03_capsule, center=<2.3848 0 0>, rotation=<rot 90 0 0>}]
 dyn_Hair_BackR_04 [{shapeType=Capsule, length=5, radius=2, shapeName=dyn_Hair_BackR_04_capsule, center=<3.3553 0 0>, rotation=<rot 90 0 0>}]
 dyn_Hair_BackR_05 [{shapeType=Capsule, length=5, radius=2, shapeName=dyn_Hair_BackR_05_capsule, center=<3.3034 0 0>, rotation=<rot 90 0 0>}]
 dyn_Hair_Back_01 [{shapeType=Capsule, length=0, radius=2, shapeName=dyn_Hair_Back_01_capsule, rotation=<rot 90 0 0>}]
 dyn_Hair_Back_02 [{shapeType=Capsule, length=2.4027, radius=2, shapeName=dyn_Hair_Back_02_capsule, center=<2.0336 0 0>, rotation=<rot 90 0 0>}]
 dyn_Hair_Back_03 [{shapeType=Capsule, length=3.2215, radius=2, shapeName=dyn_Hair_Back_03_capsule, center=<2.3848 0 0>, rotation=<rot 90 0 0>}]
 dyn_Hair_Back_04 [{shapeType=Capsule, length=5, radius=2, shapeName=dyn_Hair_Back_04_capsule, center=<3.3553 0 0>, rotation=<rot 90 0 0>}]
 dyn_Hair_Back_05 [{shapeType=Capsule, length=5, radius=2, shapeName=dyn_Hair_Back_05_capsule, center=<3.3034 0 0>, rotation=<rot 90 0 0>}]
 lowerarm_l [{
   shapeType=Capsule
   length=21.5
   radius=8.5
   shapeName=lowerarm_l_capsule
   center=<15.4871 0.0001 -2.7308>
   rotation=<rot 80.0001 -0.0008 -0.0013>
  }]
 lowerarm_r [{shapeType=Capsule, length=21.5, radius=8.5, shapeName=lowerarm_r_capsule, center=<-14.2117 0 2.5059>, rotation=<rot 80.0001 -0.0008 -0.001>}]
 neck_01 [{shapeType=Capsule, length=8.1563, radius=3.59, shapeName=neck_01_capsule, center=<0.4029 1.5035 0>}]
 spine_04 [{shapeType=Capsule, length=19.5, radius=10.5, shapeName=spine_04_capsule, center=<-0.5867 2.9031 0>}]
 spine_05 [{shapeType=Capsule, length=21, radius=9, shapeName=spine_05_capsule, center=<-0.3981 -0.625 0>}]
 upperarm_l [{shapeType=Capsule, length=25.75, radius=4.755, shapeName=upperarm_l_capsule, center=<10.5007 0 0>, rotation=<rot -0.0001 -89.9999 90>}]
 upperarm_r [{shapeType=Capsule, length=27.25, radius=5.255, shapeName=upperarm_r_capsule, center=<-11.1137 0 0>, rotation=<rot -0.0002 -89.9998 90>}]
}
constraints [
 {
  bone1Name=spine_05
  bone2Name=spine_04
  swing1LimitDegrees=45
  swing1Motion=Free
  swing2LimitDegrees=45
  swing2Motion=Free
  twistLimitDegrees=45
  twistMotion=Free
 }
 {
  bone1Name=neck_01
  bone2Name=spine_05
  swing1LimitDegrees=45
  swing1Motion=Locked
  swing2LimitDegrees=45
  swing2Motion=Locked
  twistLimitDegrees=45
  twistMotion=Locked
 }
 {
  bone1Name=dyn_Hair_BackL_01
  bone2Name=neck_01
  swing1LimitDegrees=45
  swing1Motion=Locked
  swing2LimitDegrees=45
  swing2Motion=Locked
  twistLimitDegrees=45
  twistMotion=Locked
 }
 {bone1Name=dyn_Hair_BackL_02, bone2Name=dyn_Hair_BackL_01, swing1LimitDegrees=5, swing2LimitDegrees=5, twistLimitDegrees=0}
 {bone1Name=dyn_Hair_BackL_03, bone2Name=dyn_Hair_BackL_02, swing1LimitDegrees=10, swing2LimitDegrees=10, twistLimitDegrees=0}
 {bone1Name=dyn_Hair_BackL_04, bone2Name=dyn_Hair_BackL_03, swing1LimitDegrees=10, swing2LimitDegrees=20, twistLimitDegrees=0}
 {bone1Name=dyn_Hair_BackL_05, bone2Name=dyn_Hair_BackL_04, swing1LimitDegrees=10, swing2LimitDegrees=15, twistLimitDegrees=0}
 {
  bone1Name=dyn_Hair_BackR_01
  bone2Name=neck_01
  swing1LimitDegrees=45
  swing1Motion=Locked
  swing2LimitDegrees=45
  swing2Motion=Locked
  twistLimitDegrees=45
  twistMotion=Locked
 }
 {bone1Name=dyn_Hair_BackR_02, bone2Name=dyn_Hair_BackR_01, swing1LimitDegrees=5, swing2LimitDegrees=5, twistLimitDegrees=0}
 {bone1Name=dyn_Hair_BackR_03, bone2Name=dyn_Hair_BackR_02, swing1LimitDegrees=10, swing2LimitDegrees=10, twistLimitDegrees=0}
 {bone1Name=dyn_Hair_BackR_04, bone2Name=dyn_Hair_BackR_03, swing1LimitDegrees=10, swing2LimitDegrees=20, twistLimitDegrees=0}
 {bone1Name=dyn_Hair_BackR_05, bone2Name=dyn_Hair_BackR_04, swing1LimitDegrees=10, swing2LimitDegrees=15, twistLimitDegrees=0}
 {
  bone1Name=dyn_Hair_Back_01
  bone2Name=neck_01
  swing1LimitDegrees=45
  swing1Motion=Locked
  swing2LimitDegrees=45
  swing2Motion=Locked
  twistLimitDegrees=45
  twistMotion=Locked
 }
 {bone1Name=dyn_Hair_Back_02, bone2Name=dyn_Hair_Back_01, swing1LimitDegrees=5, swing2LimitDegrees=5, twistLimitDegrees=0}
 {bone1Name=dyn_Hair_Back_03, bone2Name=dyn_Hair_Back_02, swing1LimitDegrees=10, swing2LimitDegrees=10, twistLimitDegrees=0}
 {bone1Name=dyn_Hair_Back_04, bone2Name=dyn_Hair_Back_03, swing1LimitDegrees=10, swing2LimitDegrees=20, twistLimitDegrees=0}
 {bone1Name=dyn_Hair_Back_05, bone2Name=dyn_Hair_Back_04, swing1LimitDegrees=10, swing2LimitDegrees=15, twistLimitDegrees=0}
]
