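# Control Rig for the Wayward Rebel female character's FaceAcc part that corrects the long back hair against body and arm motion.
# Weighted parent constraints tie center, left and right hair chain bones and nulls to head, neck, spine, shoulders or the center chain.
# Spherical pose readers on the upper arms then drive corrective hair offsets, some scaled by CrouchADS_Fix or by BackHairDynAlpha times JumpFix, all defaulting to zero.
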ControlRigBlueprint /BRCosmetics/Characters/Player/Female/Medium/Bodies/F_MED_WaywardRebel/Meshes/Parts/F_MED_WaywardRebel_FaceAcc_CtrlRig
graphs [
 ContainedGraph {nodes [
   n1 = RigVMFunctionEntryNode()
   SequenceAggregate = &RigVMFunction_Sequence(ExecuteContext=@n1.ExecuteContext)
   SequenceAggregate_1 = &RigVMFunction_Sequence(ExecuteContext=@SequenceAggregate.B)
   SequenceAggregate_2 = &RigVMFunction_Sequence(ExecuteContext=@SequenceAggregate_1.B)
   RigVMFunctionReturnNode(A=@SequenceAggregate.A, B=@SequenceAggregate_1.A, C=@SequenceAggregate_2.A, D=@SequenceAggregate_2.B)
  ]}
 Multiply_1_1_ContainedGraph {nodes [
   n1 = RigVMFunctionEntryNode()
   RigVMFunctionReturnNode(Result=&RigVMFunction_MathDoubleMul(B=@n1.C, A=&RigVMFunction_MathDoubleMul(A=@n1.A, B=@n1.B).Result).Result)
  ]}
 Multiply_2_ContainedGraph {nodes [
   n1 = RigVMFunctionEntryNode()
   Multiply = &RigVMFunction_MathDoubleMul(A=@n1.A, B=@n1.B)
   RigVMFunctionReturnNode(Result=&RigVMFunction_MathDoubleMul(B=@n1.C, A=@Multiply.Result).Result)
  ]}
 RigVMFunctionLibrary {links [], nodes []}
 RigVMModel {nodes [
   RigUnit_BeginExecution = &RigUnit_BeginExecution()
   SequenceAggregate = &RigVMFunction_Sequence(ExecuteContext=@RigUnit_BeginExecution.ExecutePin)
   &RigUnit_ParentConstraint(
    AdvancedSettings=(RotationOrderForFilter=XZY)
    Child=(Type=Null,Name="dyn_Hair_Back_01_Null")
    Filter=(TranslationFilter=(bX=true,bY=true,bZ=true),RotationFilter=(bX=false,bY=false,bZ=false),ScaleFilter=(bX=True,bY=True,bZ=True))
    Parents=((Item=(Type=Bone,Name="head"),Weight=0.9),(Item=(Type=Bone,Name="spine_05"),Weight=0.1))
    Weight=1
    ExecutePin=@SequenceAggregate.A
   )
   > &RigUnit_ParentConstraint(
    AdvancedSettings=(RotationOrderForFilter=XZY)
    Child=(Type=Null,Name="dyn_Hair_Back_02_Null")
    Filter=(TranslationFilter=(bX=True,bY=True,bZ=True),RotationFilter=(bX=false,bY=false,bZ=false),ScaleFilter=(bX=True,bY=True,bZ=True))
    Parents=((Item=(Type=Bone,Name="neck_02"),Weight=0.6),(Item=(Type=Bone,Name="spine_05"),Weight=0.4))
    Weight=1
   )
   > &RigUnit_ParentConstraint(
    AdvancedSettings=(RotationOrderForFilter=XZY)
    Child=(Type=Null,Name="dyn_Hair_Back_03_Null")
    Filter=(TranslationFilter=(bX=True,bY=True,bZ=True),RotationFilter=(bX=false,bY=false,bZ=false),ScaleFilter=(bX=True,bY=True,bZ=True))
    Parents=((Item=(Type=Bone,Name="neck_02"),Weight=0.5),(Item=(Type=Bone,Name="spine_05"),Weight=0.5))
    Weight=1
   )
   > &RigUnit_ParentConstraint(
    AdvancedSettings=(RotationOrderForFilter=XZY)
    Child=(Type=Bone,Name="dyn_Hair_Back_04")
    Parents=((Item=(Type=Bone,Name="neck_01"),Weight=0.0055),(Item=(Type=Bone,Name="spine_03"),Weight=0.0076),(Item=(Type=Bone,Name="clavicle_r"),Weight=0.0885),(Item=(Type=Bone,Name="clavicle_l"),Weight=0.0918),(Item=(Type=Bone,Name="spine_05"),Weight=0.8066))
    Weight=1
   )
   > &RigUnit_ParentConstraint(
    AdvancedSettings=(RotationOrderForFilter=XZY)
    Child=(Type=Bone,Name="dyn_Hair_Back_05")
    Parents="((Item=(Type=Bone,Name=\"spine_01\"),Weight=0.0025),(Item=(Type=Bone,Name=\"spine_03\"),Weight=0.0108),(Item=(Type=Bone,Name=\"spine_02\"),Weight=0.011),(Item=(Type=Bone,Name=\"upperarm_l\"),Weight=0.0192),(Item=(Type=Bone,Name=\"spine_04\"),Weight=0.1175),(Item=(Ty ...(truncated)"
    Weight=1
   )
   > &RigUnit_ParentConstraint(
    AdvancedSettings=(RotationOrderForFilter=XZY)
    Child=(Type=Bone,Name="dyn_Hair_BackL_01")
    Filter=(TranslationFilter=(bX=true,bY=true,bZ=true),RotationFilter=(bX=true,bY=true,bZ=true),ScaleFilter=(bX=True,bY=True,bZ=True))
    Parents=((Item=(Type=Bone,Name="dyn_Hair_Back_01"),Weight=1),(Item=(Type=Bone,Name="spine_05"),Weight=0))
    Weight=1
   )
   > &RigUnit_ParentConstraint(
    AdvancedSettings=(RotationOrderForFilter=XZY)
    Child=(Type=Bone,Name="dyn_Hair_BackL_02")
    Filter=(TranslationFilter=(bX=True,bY=True,bZ=True),RotationFilter=(bX=true,bY=true,bZ=true),ScaleFilter=(bX=True,bY=True,bZ=True))
    Parents=((Item=(Type=Bone,Name="dyn_Hair_Back_02"),Weight=1),(Item=(Type=Bone,Name="spine_05"),Weight=0))
    Weight=1
   )
   > &RigUnit_ParentConstraint(
    AdvancedSettings=(RotationOrderForFilter=XZY)
    Child=(Type=Bone,Name="dyn_Hair_BackL_03")
    Filter=(TranslationFilter=(bX=True,bY=True,bZ=True),RotationFilter=(bX=true,bY=true,bZ=true),ScaleFilter=(bX=True,bY=True,bZ=True))
    Parents=((Item=(Type=Bone,Name="dyn_Hair_Back_03"),Weight=0.5),(Item=(Type=Bone,Name="spine_05"),Weight=0),(Item=(Type=Bone,Name="clavicle_l"),Weight=0.5))
    Weight=1
   )
   > &RigUnit_ParentConstraint(
    AdvancedSettings=(RotationOrderForFilter=XZY)
    Child=(Type=Bone,Name="dyn_Hair_BackL_04")
    Parents=((Item=(Type=Bone,Name="spine_04"),Weight=0.002),(Item=(Type=Bone,Name="upperarm_l"),Weight=0.083),(Item=(Type=Bone,Name="spine_05"),Weight=0.1205),(Item=(Type=Bone,Name="clavicle_l"),Weight=0.7945))
    Weight=1
   )
   > &RigUnit_ParentConstraint(
    AdvancedSettings=(RotationOrderForFilter=XZY)
    Child=(Type=Bone,Name="dyn_Hair_BackL_05")
    Parents="((Item=(Type=Bone,Name=\"upperarm_l\"),Weight=0.0002),(Item=(Type=Bone,Name=\"spine_03\"),Weight=0.0005),(Item=(Type=Bone,Name=\"spine_02\"),Weight=0.0015),(Item=(Type=Bone,Name=\"clavicle_r\"),Weight=0.0026),(Item=(Type=Bone,Name=\"spine_04\"),Weight=0.1052),(Item= ...(truncated)"
    Weight=1
   )
   > &RigUnit_ParentConstraint(
    AdvancedSettings=(RotationOrderForFilter=XZY)
    Child=(Type=Bone,Name="dyn_Hair_BackR_01")
    Filter=(TranslationFilter=(bX=true,bY=true,bZ=true),RotationFilter=(bX=true,bY=true,bZ=true),ScaleFilter=(bX=True,bY=True,bZ=True))
    Parents=((Item=(Type=Bone,Name="dyn_Hair_Back_01"),Weight=1),(Item=(Type=Bone,Name="spine_05"),Weight=0))
    Weight=1
   )
   > &RigUnit_ParentConstraint(
    AdvancedSettings=(RotationOrderForFilter=XZY)
    Child=(Type=Bone,Name="dyn_Hair_BackR_02")
    Filter=(TranslationFilter=(bX=True,bY=True,bZ=True),RotationFilter=(bX=true,bY=true,bZ=true),ScaleFilter=(bX=True,bY=True,bZ=True))
    Parents=((Item=(Type=Bone,Name="dyn_Hair_Back_02"),Weight=1),(Item=(Type=Bone,Name="spine_05"),Weight=0))
    Weight=1
   )
   > &RigUnit_ParentConstraint(
    AdvancedSettings=(RotationOrderForFilter=XZY)
    Child=(Type=Bone,Name="dyn_Hair_BackR_03")
    Filter=(TranslationFilter=(bX=True,bY=True,bZ=True),RotationFilter=(bX=true,bY=true,bZ=true),ScaleFilter=(bX=True,bY=True,bZ=True))
    Parents=((Item=(Type=Bone,Name="dyn_Hair_Back_03"),Weight=0.5),(Item=(Type=Bone,Name="spine_05"),Weight=0),(Item=(Type=Bone,Name="clavicle_r"),Weight=0.5))
    Weight=1
   )
   > &RigUnit_ParentConstraint(
    AdvancedSettings=(RotationOrderForFilter=XZY)
    Child=(Type=Bone,Name="dyn_Hair_BackR_04")
    Parents=((Item=(Type=Bone,Name="spine_04"),Weight=0.002),(Item=(Type=Bone,Name="upperarm_r"),Weight=0.083),(Item=(Type=Bone,Name="spine_05"),Weight=0.1205),(Item=(Type=Bone,Name="clavicle_r"),Weight=0.7945))
    Weight=1
   )
   > &RigUnit_ParentConstraint(
    AdvancedSettings=(RotationOrderForFilter=XZY)
    Child=(Type=Bone,Name="dyn_Hair_BackR_05")
    Parents="((Item=(Type=Bone,Name=\"upperarm_r\"),Weight=0.0002),(Item=(Type=Bone,Name=\"spine_03\"),Weight=0.0005),(Item=(Type=Bone,Name=\"spine_02\"),Weight=0.0015),(Item=(Type=Bone,Name=\"clavicle_l\"),Weight=0.0026),(Item=(Type=Bone,Name=\"spine_04\"),Weight=0.1052),(Item= ...(truncated)"
    Weight=1
   )
   SequenceAggregate_1 = &RigVMFunction_Sequence(ExecuteContext=@SequenceAggregate.B)
   n1 = &RigVMFunction_Sequence(ExecuteContext=@SequenceAggregate_1.B)
   n2 = &RigUnit_SphericalPoseReader(
    ActiveRegionScaleFactors=(PositiveWidth=1.25,NegativeWidth=1.5,PositiveHeight=1.25,NegativeHeight=2)
    ActiveRegionSize=0.1
    Debug=(bDrawDebug=false,bDraw2D=false,bDrawLocalAxes=False,DebugScale=25,DebugSegments=20,DebugThickness=0.25)
    DriverAxis=(X=1,Y=0,Z=0)
    DriverItem=(Type=Bone,Name="upperarm_l")
    FalloffRegionScaleFactors=(PositiveWidth=1,NegativeWidth=0.5,PositiveHeight=1,NegativeHeight=1)
    FalloffSize=0.25
    FlipHeightScaling=False
    FlipWidthScaling=False
    OptionalParentItem=(Type=Bone,Name="None")
    RotationOffset=(X=70,Y=-20,Z=-35)
    ExecutePin=@n1.A
   )
   > &RigUnit_ModifyTransforms(
    ItemToModify=((Item=(Type=Bone,Name="dyn_Hair_BackL_03"),Transform=(Rotation=(X=0,Y=0,Z=0,W=1),Translation=(X=-2.75,Y=0,Z=0.5),Scale3D=(X=1,Y=1,Z=1))))
    WeightMinimum=0
    Weight=@n2.OutputParam
   )
   > &RigUnit_ModifyTransforms(
    ItemToModify=((Item=(Type=Bone,Name="dyn_Hair_BackR_03"),Transform=(Rotation=(X=0,Y=0,Z=0,W=1),Translation=(X=-2.75,Y=0,Z=0.5),Scale3D=(X=1,Y=1,Z=1))))
    WeightMinimum=0
    Weight=@n2.OutputParam
   )
   n3 = &RigUnit_SphericalPoseReader(
    ActiveRegionScaleFactors=(PositiveWidth=1,NegativeWidth=1,PositiveHeight=1,NegativeHeight=1.5)
    ActiveRegionSize=0.1
    Debug=(bDrawDebug=false,bDraw2D=false,bDrawLocalAxes=False,DebugScale=25,DebugSegments=20,DebugThickness=0.25)
    DriverAxis=(X=1,Y=0,Z=0)
    DriverItem=(Type=Bone,Name="upperarm_l")
    FalloffRegionScaleFactors=(PositiveWidth=0.5,NegativeWidth=0.5,PositiveHeight=0.5,NegativeHeight=1)
    FalloffSize=0.25
    FlipHeightScaling=False
    FlipWidthScaling=False
    OptionalParentItem=(Type=Bone,Name="None")
    RotationOffset=(X=50,Y=0,Z=-20)
    ExecutePin=@n1.B
   )
   n4 = RigVMVariableNode() {variable=CrouchADS_Fix}
   > &RigUnit_ModifyTransforms(
    ItemToModify="((Item=(Type=Bone,Name=\"dyn_Hair_BackL_03\"),Transform=(Rotation=(X=0,Y=0,Z=0,W=1),Translation=(X=-2,Y=0,Z=-2),Scale3D=(X=1,Y=1,Z=1))),(Item=(Type=Bone,Name=\"dyn_Hair_BackR_03\"),Transform=(Rotation=(X=0,Y=0,Z=0,W=1),Translation=(X=-2,Y=0,Z=-2),Scale3D=(X=1, ...(truncated)"
    WeightMinimum=0
    Weight=&RigVMFunction_MathDoubleRemap(
     SourceMaximum=1
     SourceMinimum=0
     TargetMaximum=1
     TargetMinimum=0
     bClamp=False
     Value=&RigVMFunction_MathDoubleMul(A=@n3.OutputParam, B=@n4.Value).Result
    ).Result
   )
   RigVMFunction_Sequence = &RigVMFunction_Sequence(ExecuteContext=@SequenceAggregate_1.A)
   n5 = &RigUnit_SphericalPoseReader(
    ActiveRegionScaleFactors=(PositiveWidth=1,NegativeWidth=1,PositiveHeight=1.3,NegativeHeight=1)
    ActiveRegionSize=0.1
    Debug=(bDrawDebug=false,bDraw2D=false,bDrawLocalAxes=False,DebugScale=25,DebugSegments=20,DebugThickness=0.25)
    DriverAxis=(X=1,Y=0,Z=0)
    DriverItem=(Type=Bone,Name="upperarm_l")
    FalloffRegionScaleFactors=(PositiveWidth=1,NegativeWidth=1,PositiveHeight=1,NegativeHeight=1)
    FalloffSize=0.2
    FlipHeightScaling=False
    FlipWidthScaling=False
    OptionalParentItem=(Type=Bone,Name="None")
    RotationOffset=(X=-20,Y=-90,Z=0)
    ExecutePin=@RigVMFunction_Sequence.A
   )
   Multiply_1 = RigVMAggregateNode(C=@n5.OutputParam, A=RigVMVariableNode() {variable=BackHairDynAlpha}.Value, B=RigVMVariableNode() {variable=JumpFix}.Value)
   Remap = &RigVMFunction_MathDoubleRemap(SourceMaximum=1, SourceMinimum=0, TargetMaximum=1, TargetMinimum=0, bClamp=False, Value=@Multiply_1.Result)
   > &RigUnit_ModifyTransforms(
    ItemToModify="((Item=(Type=Bone,Name=\"dyn_Hair_BackL_04\"),Transform=(Rotation=(X=0,Y=0,Z=0,W=1),Translation=(X=0,Y=-1,Z=1.5),Scale3D=(X=1,Y=1,Z=1))),(Item=(Type=Bone,Name=\"dyn_Hair_BackL_03\"),Transform=(Rotation=(X=0,Y=0,Z=-0.0872,W=0.9962),Translation=(X=-1.5,Y=1,Z=0), ...(truncated)"
    WeightMinimum=0
    Weight=@Remap.Result
   )
   n6 = &RigUnit_SphericalPoseReader(
    ActiveRegionScaleFactors=(PositiveWidth=1,NegativeWidth=1.3,PositiveHeight=1,NegativeHeight=1)
    ActiveRegionSize=0.1
    Debug=(bDrawDebug=false,bDraw2D=false,bDrawLocalAxes=False,DebugScale=25,DebugSegments=20,DebugThickness=0.25)
    DriverAxis=(X=-1,Y=0,Z=0)
    DriverItem=(Type=Bone,Name="upperarm_r")
    FalloffRegionScaleFactors=(PositiveWidth=1,NegativeWidth=1,PositiveHeight=1,NegativeHeight=1)
    FalloffSize=0.2
    FlipHeightScaling=False
    FlipWidthScaling=False
    OptionalParentItem=(Type=Bone,Name="None")
    RotationOffset=(X=35,Y=80,Z=10)
    ExecutePin=@RigVMFunction_Sequence.B
   )
   Multiply = RigVMAggregateNode(C=@n6.OutputParam, A=RigVMVariableNode() {variable=BackHairDynAlpha}.Value, B=RigVMVariableNode() {variable=JumpFix}.Value)
   > &RigUnit_ModifyTransforms(
    ItemToModify="((Item=(Type=Bone,Name=\"dyn_Hair_BackR_04\"),Transform=(Rotation=(X=0,Y=0,Z=0,W=1),Translation=(X=0,Y=1,Z=1.5),Scale3D=(X=1,Y=1,Z=1))),(Item=(Type=Bone,Name=\"dyn_Hair_BackR_03\"),Transform=(Rotation=(X=0.013,Y=0.0011,Z=0.0871,W=0.9961),Translation=(X=-1.5,Y= ...(truncated)"
    WeightMinimum=0
    Weight=&RigVMFunction_MathDoubleRemap(SourceMaximum=1, SourceMinimum=0, TargetMaximum=1, TargetMinimum=0, bClamp=False, Value=@Multiply.Result).Result
   )
   &RigUnit_SphericalPoseReader(
    ActiveRegionScaleFactors=(PositiveWidth=1.5,NegativeWidth=1.5,PositiveHeight=1,NegativeHeight=2)
    ActiveRegionSize=0.1
    Debug=(bDrawDebug=false,bDraw2D=false,bDrawLocalAxes=False,DebugScale=25,DebugSegments=20,DebugThickness=0.25)
    DriverAxis=(X=-1,Y=0,Z=0)
    DriverItem=(Type=Bone,Name="upperarm_r")
    ExecutePin=()
    FalloffRegionScaleFactors=(PositiveWidth=0.5,NegativeWidth=0.5,PositiveHeight=1,NegativeHeight=1)
    FalloffSize=0.25
    FlipHeightScaling=False
    FlipWidthScaling=False
    OptionalParentItem=(Type=Bone,Name="None")
    RotationOffset=(X=-110,Y=20,Z=-5)
   )
  ]}
 RigVMModel Construction Event Graph {nodes [
   n1 = &RigUnit_PrepareForExecution()
   SequenceAggregate = RigVMAggregateNode(ExecuteContext=@n1.ExecutePin)
   n2 = RigVMRerouteNode(Value=())
   RerouteNode_6_1 = RigVMDispatchNode(Values="((Type=Control,Name=\"dyn_Hair_Back_01_ctrl\"),((),Type=Control,Name=\"dyn_Hair_Back_02_ctrl\"),((),Type=Control,Name=\"dyn_Hair_Back_03_ctrl\"),((),Type=Control,Name=\"dyn_Hair_Back_04_ctrl\"),((),Type=Control,Name=\"dyn_Hair_Back_05_ctrl\"),((),Type=Control,Name=\" ...(truncated)") {template="DISPATCH_RigVMDispatch_ArrayMake(in Values,out Array)"}
   RigVMVariableNode(ExecuteContext=@n2.Value, Value=@RerouteNode_6_1.Array) {variable=dyn_Hair_Back_ctrls}
   > RigVMVariableNode(Value=/Script/ControlRig.RigUnit_HierarchyGetPoseItemArray(ElementType=All, Initial=false, ItemsToGet=@RerouteNode_6_1.Array) {is=struct RigVMUnitNode}.Pose) {variable=dyn_Hair_Back_pose}
  ]}
]
hierarchy [
 root {loc=<0 0 0>, rot=<0 0 0 1>}
 pelvis {parent=root, loc=<0 -0.5301 98.0571>, rot=<0.0127 0.707 -0.0127 -0.707>}
 spine_01 {parent=pelvis, loc=<6.1208 -0.9328 0>, rot=<0 0 0.034 -0.9994>}
 spine_02 {parent=spine_01, loc=<6.5996 -0.2095 0>, rot=<0 0 0.0238 -0.9997>}
 spine_03 {parent=spine_02, loc=<6.4914 -0.1237 0>, rot=<0 0 0.0149 -0.9999>}
 spine_04 {parent=spine_03, loc=<6.9598 0.0794 0>, rot=<0 0 -0.0132 -0.9999>}
 spine_05 {parent=spine_04, loc=<6.9987 0 0>, rot=<0 0 -0.0015 -1>}
 neck_01 {parent=spine_05, loc=<11.3138 0.1529 0>, rot=<0 0 -0.2003 -0.9797>}
 neck_02 {parent=neck_01, loc=<4.5627 0 0>, rot=<0 0 0.0504 -0.9987>}
 head {parent=neck_02, loc=<5.3914 0 0>, rot=<0 0 0.115 -0.9934>}
 dyn_Hair_Back_01 {parent=head, loc=<8.5016 -1.2653 -0.3258>, rot=<0.3093 -0.6359 0.6359 -0.3093>}
 dyn_Hair_Back_02 {parent=dyn_Hair_Back_01, loc=<4.2813 0 0>, rot=<0 -0.3651 0 -0.931>}
 dyn_Hair_Back_03 {parent=dyn_Hair_Back_02, loc=<6.9171 0 0>, rot=<0.0028 0.1439 -0.0193 -0.9894>}
 dyn_Hair_Back_04 {parent=dyn_Hair_Back_03, loc=<7.2053 0.0745 -0.0331>, rot=<0.0002 0.0125 0.0195 -0.9997>}
 dyn_Hair_Back_05 {parent=dyn_Hair_Back_04, loc=<11.9252 0 -0.0148>, rot=<0 -0.1961 0 -0.9806>}
 dyn_Hair_Back_06 {parent=dyn_Hair_Back_05, loc=<7.2417 0 0>, rot=<0 0 0 1>}
 dyn_Hair_BackL_01 {parent=head, loc=<7.9722 -1.0353 -5.5286>, rot=<0.3093 -0.6359 0.6359 -0.3093>}
 dyn_Hair_BackL_02 {parent=dyn_Hair_BackL_01, loc=<4.2813 0 0>, rot=<0 -0.3651 0 -0.931>}
 dyn_Hair_BackL_03 {parent=dyn_Hair_BackL_02, loc=<6.9171 0 0>, rot=<0.0146 0.1158 -0.1241 -0.9854>}
 dyn_Hair_BackL_04 {parent=dyn_Hair_BackL_03, loc=<7.5718 0.4723 -0.1376>, rot=<0.0045 0.0358 0.1249 -0.9915>}
 dyn_Hair_BackL_05 {parent=dyn_Hair_BackL_04, loc=<9.201 0 0>, rot=<-0.0026 -0.1231 -0.0208 -0.9922>}
 dyn_Hair_BackL_06 {parent=dyn_Hair_BackL_05, loc=<8.0052 0 0>, rot=<0 0 0 1>}
 dyn_Hair_BackR_01 {parent=head, loc=<8.1854 -0.439 5.529>, rot=<0.3093 -0.6359 0.6359 -0.3093>}
 dyn_Hair_BackR_02 {parent=dyn_Hair_BackR_01, loc=<4.2813 0 0>, rot=<0 -0.3651 0 -0.931>}
 dyn_Hair_BackR_03 {parent=dyn_Hair_BackR_02, loc=<6.9171 0 0>, rot=<-0.0053 0.16 0.0328 -0.9866>}
 dyn_Hair_BackR_04 {parent=dyn_Hair_BackR_03, loc=<7.3884 -0.1267 0.0295>, rot=<0.0003 -0.0077 -0.0332 -0.9994>}
 dyn_Hair_BackR_05 {parent=dyn_Hair_BackR_04, loc=<9.201 0 0>, rot=<0 -0.1207 0 -0.9927>}
 dyn_Hair_BackR_06 {parent=dyn_Hair_BackR_05, loc=<7.3927 0 0>, rot=<0 0 0 1>}
 base_headphone_01_l {parent=head, loc=<3.0186 0.5765 -10.6653>, rot=<-0.0515 -0.7173 -0.0593 -0.6923>}
 dyn_headphone_01_l {parent=base_headphone_01_l, loc=<0 0 0>, rot=<0 0 0 1>}
 base_headphone_01_r {parent=head, loc=<3.0186 0.5765 10.6653>, rot=<0.7173 -0.0515 -0.6923 0.0593>}
 dyn_headphone_01_r {parent=base_headphone_01_r, loc=<0 0 0>, rot=<0 0 0 1>}
 base_Hair_a {parent=head, loc=<-0.9978 4.3065 -5.9553>, rot=<0 0 0.7097 -0.7045>}
 dyn_Hair_a {parent=base_Hair_a, loc=<0 0 0>, rot=<0 0 0 1>}
 base_Hair_b {parent=head, loc=<6.0019 11.0572 -1.0561>, rot=<0.0293 -0.7065 -0.7042 0.0644>}
 dyn_Hair_b01 {parent=base_Hair_b, loc=<0 0 0>, rot=<0 0 0 1>}
 dyn_Hair_b02 {parent=dyn_Hair_b01, loc=<2.7469 0 0>, rot=<0 0 0 1>}
 base_Hair_c {parent=head, loc=<5.9214 10.5885 1.8187>, rot=<0.0023 0.6428 0.766 0.0028>}
 dyn_Hair_c {parent=base_Hair_c, loc=<0 0 0>, rot=<0 0 0 1>}
 base_Hair_d {parent=head, loc=<5.6343 10.1675 3.9123>, rot=<0.0018 0.5 0.866 0.0031>}
 dyn_Hair_d01 {parent=base_Hair_d, loc=<0 0 0>, rot=<0 0 0 1>}
 base_Hair_e {parent=head, loc=<5.0303 7.1416 5.6435>, rot=<-0.0576 0.1358 0.9843 0.0965>}
 dyn_Hair_e01 {parent=base_Hair_e, loc=<0 0 0>, rot=<0 0 0 1>}
 clavicle_l {parent=spine_05, loc=<5.4831 -0.1053 -3.2946>, rot=<0.0465 -0.7289 -0.009 -0.6829>}
 upperarm_l {parent=clavicle_l, loc=<9.8658 -1.2136 -0.4105>, rot=<0.0164 -0.3245 -0.0269 -0.9454>}
 lowerarm_l {parent=upperarm_l, loc=<26.465 -0.8345 0.259>, rot=<0.0036 0.0161 -0.2298 -0.9731>}
 clavicle_r {parent=spine_05, loc=<5.4831 -0.1053 3.2946>, rot=<0.7289 0.0465 -0.6829 0.009>}
 upperarm_r {parent=clavicle_r, loc=<-9.8658 1.2136 0.4105>, rot=<0.0164 -0.3245 -0.0269 -0.9454>}
 lowerarm_r {parent=upperarm_r, loc=<-26.465 0.8345 -0.259>, rot=<0.0036 0.0161 -0.2298 -0.9731>}
 dyn_Hair_l_base {parent=spine_05, loc=<12.581 3.5489 -6.7>, rot=<0 0 0.6839 -0.7296>}
 dyn_Hair_l_01 {parent=dyn_Hair_l_base, loc=<-1.3189 -1.8868 -0.0994>, rot=<0 0 0 1>}
 dyn_Hair_l_02 {parent=dyn_Hair_l_01, loc=<-3.2934 -4.4989 -1.4146>, rot=<0 0 0 1>}
 dyn_Hair_l_03 {parent=dyn_Hair_l_02, loc=<-4.3079 -5.8053 -0.6797>, rot=<0 0 0 1>}
 dyn_Hair_l_04 {parent=dyn_Hair_l_03, loc=<-1.1163 -5.792 0.0902>, rot=<0 0 0 1>}
 dyn_Hair_r_base {parent=spine_05, loc=<12.581 3.5489 6.7>, rot=<0 0 0.6839 -0.7296>}
 dyn_Hair_r_01 {parent=dyn_Hair_r_base, loc=<-1.3189 -1.8868 -0.0994>, rot=<0 0 0 1>}
 dyn_Hair_r_02 {parent=dyn_Hair_r_01, loc=<-3.2934 -4.4989 1.2363>, rot=<0 0 0 1>}
 dyn_Hair_r_03 {parent=dyn_Hair_r_02, loc=<-4.3079 -5.8053 0.8949>, rot=<0 0 0 1>}
 dyn_Hair_r_04 {parent=dyn_Hair_r_03, loc=<-1.1163 -5.792 0.0902>, rot=<0 0 0 1>}
 dyn_Hair_BackL_02_Null {parent=pelvis, type=Null, loc=<59.0305 -7.7237 -5.5286>, rot=<0 0 -0.6508 0.7593>}
 dyn_Hair_BackL_02_ctrl {controlType=EULER_TRANSFORM, parent=dyn_Hair_BackL_02_Null, type=Control, loc=<0 0 0>, rot=<0 0 0 1>}
 dyn_Hair_BackL_03_Null {parent=pelvis, type=Null, loc=<52.158 -8.5088 -5.5286>, rot=<0 0 0.6478 -0.7618>}
 dyn_Hair_BackL_03_ctrl {controlType=EULER_TRANSFORM, parent=dyn_Hair_BackL_03_Null, type=Control, loc=<0 0 0>, rot=<0 0 0 1>}
 dyn_Hair_BackL_04_Null {parent=pelvis, type=Null, loc=<43.581 -11.6175 -7.8636>, rot=<0 0 -0.6478 0.7618>}
 dyn_Hair_BackL_04_ctrl {controlType=EULER_TRANSFORM, parent=dyn_Hair_BackL_04_Null, type=Control, loc=<0 0 0>, rot=<0 0 0 1>}
 dyn_Hair_BackL_05_Null {parent=pelvis, type=Null, loc=<43.581 -11.6175 -7.8636>, rot=<0 0 -0.6478 0.7618>}
 dyn_Hair_BackL_05_ctrl {controlType=EULER_TRANSFORM, parent=dyn_Hair_BackL_05_Null, type=Control, loc=<0 0 0>, rot=<0 0 0 1>}
 dyn_Hair_BackL_06_Null {parent=pelvis, type=Null, loc=<43.581 -11.6175 -7.8636>, rot=<0 0 -0.6478 0.7618>}
 dyn_Hair_BackL_06_ctrl {controlType=EULER_TRANSFORM, parent=dyn_Hair_BackL_06_Null, type=Control, loc=<0 0 0>, rot=<0 0 0 1>}
 dyn_Hair_BackL_01_Null {parent=pelvis, type=Null, loc=<61.8198 -4.4758 -5.5286>, rot=<0 0 0.1148 -0.9934>}
 dyn_Hair_BackL_01_ctrl {controlType=EULER_TRANSFORM, parent=dyn_Hair_BackL_01_Null, type=Control, loc=<0 0 0>, rot=<0 0 0 1>}
 dyn_Hair_Back_01_Null {parent=pelvis, type=Null, loc=<62.3385 -4.7288 -0.3258>, rot=<0 0 0.1148 -0.9934>}
 dyn_Hair_Back_01_ctrl {controlType=EULER_TRANSFORM, parent=dyn_Hair_Back_01_Null, type=Control, loc=<0 0 0>, rot=<0 0 0 1>}
 dyn_Hair_Back_02_Null {parent=pelvis, type=Null, loc=<59.5492 -7.9768 -0.3258>, rot=<0 0 -0.6508 0.7593>}
 dyn_Hair_Back_02_ctrl {controlType=EULER_TRANSFORM, parent=dyn_Hair_Back_02_Null, type=Control, loc=<0 0 0>, rot=<0 0 0 1>}
 dyn_Hair_Back_03_Null {parent=pelvis, type=Null, loc=<52.6768 -8.7618 -0.3258>, rot=<0 0 0.6478 -0.7618>}
 dyn_Hair_Back_03_ctrl {controlType=EULER_TRANSFORM, parent=dyn_Hair_Back_03_Null, type=Control, loc=<0 0 0>, rot=<0 0 0 1>}
 dyn_Hair_Back_04_Null {parent=pelvis, type=Null, loc=<44.2983 -12.3291 -0.6812>, rot=<0 0 -0.6478 0.7618>}
 dyn_Hair_Back_04_ctrl {controlType=EULER_TRANSFORM, parent=dyn_Hair_Back_04_Null, type=Control, loc=<0 0 0>, rot=<0 0 0 1>}
 dyn_Hair_Back_05_Null {parent=pelvis, type=Null, loc=<35.1845 -16.4827 -0.6812>, rot=<0 0 0.6478 -0.7618>}
 dyn_Hair_Back_05_ctrl {controlType=EULER_TRANSFORM, parent=dyn_Hair_Back_05_Null, type=Control, loc=<0 0 0>, rot=<0 0 0 1>}
 dyn_Hair_Back_06_Null {parent=pelvis, type=Null, loc=<27.9468 -16.7205 -0.6812>, rot=<0 0 -0.6935 0.7204>}
 dyn_Hair_Back_06_ctrl {controlType=EULER_TRANSFORM, parent=dyn_Hair_Back_06_Null, type=Control, loc=<0 0 0>, rot=<0 0 0 1>}
 dyn_Hair_BackR_01_Null {parent=pelvis, type=Null, loc=<62.0591 -3.8895 5.0959>, rot=<0 0 0.1148 -0.9934>}
 dyn_Hair_BackR_01_ctrl {controlType=EULER_TRANSFORM, parent=dyn_Hair_BackR_01_Null, type=Control, loc=<0 0 0>, rot=<0 0 0 1>}
 dyn_Hair_BackR_02_Null {parent=pelvis, type=Null, loc=<59.2698 -7.1374 5.0959>, rot=<0 0 -0.6508 0.7593>}
 dyn_Hair_BackR_02_ctrl {controlType=EULER_TRANSFORM, parent=dyn_Hair_BackR_02_Null, type=Control, loc=<0 0 0>, rot=<0 0 0 1>}
 dyn_Hair_BackR_03_Null {parent=pelvis, type=Null, loc=<52.3973 -7.9224 5.0959>, rot=<0 0 0.6478 -0.7618>}
 dyn_Hair_BackR_03_ctrl {controlType=EULER_TRANSFORM, parent=dyn_Hair_BackR_03_Null, type=Control, loc=<0 0 0>, rot=<0 0 0 1>}
 dyn_Hair_BackR_04_Null {parent=pelvis, type=Null, loc=<43.9889 -11.8327 5.7125>, rot=<0 0 -0.6478 0.7618>}
 dyn_Hair_BackR_04_ctrl {controlType=EULER_TRANSFORM, parent=dyn_Hair_BackR_04_Null, type=Control, loc=<0 0 0>, rot=<0 0 0 1>}
 dyn_Hair_BackR_05_Null {parent=pelvis, type=Null, loc=<37.3308 -14.8048 5.7125>, rot=<0 0 0.6478 -0.7618>}
 dyn_Hair_BackR_05_ctrl {controlType=EULER_TRANSFORM, parent=dyn_Hair_BackR_05_Null, type=Control, loc=<0 0 0>, rot=<0 0 0 1>}
 dyn_Hair_BackR_06_Null {parent=pelvis, type=Null, loc=<30.0547 -16.1126 5.7125>, rot=<0 0 -0.6935 0.7204>}
 dyn_Hair_BackR_06_ctrl {controlType=EULER_TRANSFORM, parent=dyn_Hair_BackR_06_Null, type=Control, loc=<0 0 0>, rot=<0 0 0 1>}
]
variables [
 BackHairDynAlpha {type=double, default=0}
 CrouchADS_Fix {type=double, default=0}
 JumpFix {type=double, default=0}
 dyn_Hair_BackL_ctrls {
  type="TArray<FRigElementKey>"
  default=((Type=Control,Name="dyn_Hair_BackL_02"),(Type=Control,Name="dyn_Hair_BackL_03"),(Type=Control,Name="dyn_Hair_BackL_04"),(Type=Control,Name="dyn_Hair_BackL_05"),(Type=Control,Name="dyn_Hair_BackL_06"))
 }
 dyn_Hair_BackL_pose {type=FRigPose, default=(Elements=,HierarchyTopologyVersion=-1,PoseHash=-1)}
 dyn_Hair_BackR_ctrls {
  type="TArray<FRigElementKey>"
  default=((Type=Control,Name="dfrm_arm_spline_01_r_ctrl"),(Type=Control,Name="dfrm_arm_spline_03_r_ctrl"),(Type=Control,Name="dfrm_arm_spline_05_r_ctrl"),(Type=Control,Name="dfrm_arm_spline_07_r_ctrl"),(Type=Control,Name="dfrm_arm_spline_09_r_ctrl"))
 }
 dyn_Hair_BackR_pose {type=FRigPose, default=(Elements=,HierarchyTopologyVersion=-1,PoseHash=-1)}
 dyn_Hair_Back_ctrls {
  type="TArray<FRigElementKey>"
  default=((Type=Control,Name="dyn_Hair_BackL_02"),(Type=Control,Name="dyn_Hair_BackL_03"),(Type=Control,Name="dyn_Hair_BackL_04"),(Type=Control,Name="dyn_Hair_BackL_05"),(Type=Control,Name="dyn_Hair_BackL_06"))
 }
 dyn_Hair_Back_pose {type=FRigPose, default=(Elements=,HierarchyTopologyVersion=-1,PoseHash=-1)}
]
