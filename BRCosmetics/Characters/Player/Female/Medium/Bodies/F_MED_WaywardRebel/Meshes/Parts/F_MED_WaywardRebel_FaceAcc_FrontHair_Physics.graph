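# Physics asset for the front hair of the Wayward Rebel female character's face accessory, simulating left and right dynamic hair strands.
# Each side chains base, 01 and 02 capsule bodies with swing and twist limits, tighter at the base, which hangs from spine_05 rather than the head.
# Capsules on head, neck, upper spine and upper arms supply body shapes around the hair; neck is angularly locked to spine_05, while upper arms have no constraints.
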PhysicsAsset /BRCosmetics/Characters/Player/Female/Medium/Bodies/F_MED_WaywardRebel/Meshes/Parts/F_MED_WaywardRebel_FaceAcc_FrontHair_Physics
bodies {
 dyn_Hair_l_01 [
  {
   shapeType=Capsule
   length=6.5
   radius=3.005
   shapeName=dyn_Hair_l_01_capsule
   center=<-1.4506 -1.6818 -0.6122>
   rotation=<rot 80.6045 -110.0709 20.0703>
  }
  {
   shapeType=Capsule
   length=9.3425
   radius=1.4428
   shapeName=dyn_Hair_l_01_capsule
   center=<0.218 -2.9548 -4.1126>
   rotation=<rot 65 -130.0003 -10.0001>
  }
 ]
 dyn_Hair_l_02 [{shapeType=Capsule, length=5, radius=2.505, shapeName=dyn_Hair_l_02_capsule, center=<-1.9074 -2.7136 -0.2374>, rotation=<rot 85 -90.0004 35>}]
 dyn_Hair_l_base [{shapeType=Capsule, length=0.26, radius=2.25, shapeName=dyn_Hair_l_base_capsule, rotation=<rot 0 -90 0>}]
 dyn_Hair_r_01 [{
   shapeType=Capsule
   length=3
   radius=2.505
   shapeName=dyn_Hair_r_01_capsule
   center=<-1.4252 -1.7316 0.3046>
   rotation=<rot 83.0508 22.6964 152.4165>
  }]
 dyn_Hair_r_02 [{
   shapeType=Capsule
   length=3.0459
   radius=3.2091
   shapeName=dyn_Hair_r_02_capsule
   center=<-3.0882 -1.7994 0.557>
   rotation=<rot 83.4425 29.29 154.0401>
  }]
 dyn_Hair_r_base [{shapeType=Capsule, length=0.26, radius=2.25, shapeName=dyn_Hair_r_base_capsule, rotation=<rot 0 -90 0>}]
 head [{shapeType=Capsule, length=7.9134, radius=5.7151, shapeName=head_capsule, center=<3.4039 3.0754 0>, rotation=<rot 0 69.9999 90>}]
 neck_01 [{shapeType=Capsule, length=4.3322, radius=3.59, shapeName=neck_01_capsule, center=<0.4029 1.5035 0>, rotation=<rot 90 0 15.0001>}]
 spine_04 [
  {shapeType=Capsule, length=19.5, radius=10.5, shapeName=spine_04_capsule, center=<-0.5867 2.9031 0>}
  {shapeType=Capsule, length=10.6484, radius=1.6887, shapeName=spine_04_capsule, center=<3.0388 11.6602 4.9713>, rotation=<rot -60 -0.0002 -30>}
 ]
 spine_05 [{shapeType=Capsule, length=21, radius=9, shapeName=spine_05_capsule, center=<-0.3981 -0.625 0>}]
 upperarm_l [{shapeType=Capsule, length=25.75, radius=4.755, shapeName=upperarm_l_capsule, center=<10.5007 0 0>, rotation=<rot -0.0001 -89.9999 90>}]
 upperarm_r [{shapeType=Capsule, length=28.3032, radius=4.2018, shapeName=upperarm_r_capsule, center=<-11.1137 0 0>, rotation=<rot -0.0002 -89.9998 90>}]
}
constraints [
 {
  bone1Name=spine_05
  bone2Name=spine_04
  swing1LimitDegrees=45
  swing1Motion=Free
  swing2LimitDegrees=45
  swing2Motion=Free
  twistLimitDegrees=45
  twistMotion=Free
 }
 {bone1Name=dyn_Hair_r_02, bone2Name=dyn_Hair_r_01, swing1LimitDegrees=25, swing2LimitDegrees=25, twistLimitDegrees=10}
 {bone1Name=dyn_Hair_r_01, bone2Name=dyn_Hair_r_base, swing1LimitDegrees=25, swing2LimitDegrees=25, twistLimitDegrees=10}
 {bone1Name=dyn_Hair_l_02, bone2Name=dyn_Hair_l_01, swing1LimitDegrees=25, swing2LimitDegrees=25, twistLimitDegrees=10}
 {bone1Name=dyn_Hair_l_01, bone2Name=dyn_Hair_l_base, swing1LimitDegrees=25, swing2LimitDegrees=25, twistLimitDegrees=10}
 {bone1Name=dyn_Hair_r_base, bone2Name=spine_05, swing1LimitDegrees=15, swing2LimitDegrees=15, twistLimitDegrees=15}
 {bone1Name=dyn_Hair_l_base, bone2Name=spine_05, swing1LimitDegrees=15, swing2LimitDegrees=15, twistLimitDegrees=15}
 {
  bone1Name=neck_01
  bone2Name=spine_05
  swing1LimitDegrees=45
  swing1Motion=Locked
  swing2LimitDegrees=45
  swing2Motion=Locked
  twistLimitDegrees=45
  twistMotion=Locked
 }
 {
  bone1Name=head
  bone2Name=neck_01
  swing1LimitDegrees=45
  swing1Motion=Free
  swing2LimitDegrees=45
  swing2Motion=Free
  twistLimitDegrees=45
  twistMotion=Free
 }
]
