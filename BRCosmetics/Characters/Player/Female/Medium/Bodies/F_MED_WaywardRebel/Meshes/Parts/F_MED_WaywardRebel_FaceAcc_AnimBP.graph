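# Post-process style AnimBlueprint for the Wayward Rebel face accessory mesh, driving secondary motion on hair and headphones.
# The copied parent pose runs through a Control Rig, then chained cached poses simulate back hair, bangs and front hair using RigidBody or AnimDynamics, switched by state.
# Additive bone offsets correct hair for backpack, cape, aiming, skydiving and other states.
# Sine rotators fake wind, and Kill_Happened blends in headphone spin before output.
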AnimBlueprint /BRCosmetics/Characters/Player/Female/Medium/Bodies/F_MED_WaywardRebel/Meshes/Parts/F_MED_WaywardRebel_FaceAcc_AnimBP
target_skeleton=&/F_MED_WaywardRebel_FaceAcc_Skeleton
graphs [
 AnimGraph {nodes [
   n1 = "Transform (Modify) Bone - Bone: dyn_Hair_r_base"!(ComponentPose, bAlphaBoolEnabled:bool=True, Translation="-1,1,0", Rotation="-10.4,8,-13.6") -> (Pose) {
    alphaInputType=Bool
    rotationMode=BMM_Additive
    rotationSpace=BCS_BoneSpace
    scaleSpace=BCS_BoneSpace
    translationMode=BMM_Additive
    translationSpace=BCS_BoneSpace
    alphaBoolBlend {blendOption=Linear}
    boneToModify {boneName=dyn_Hair_r_base}
    rotation=<rot -10.4 8 -13.6>
    translation=<-1 1>
   }
   "Transform (Modify) Bone - Bone: dyn_Hair_r_02"!(ComponentPose=@n1, bAlphaBoolEnabled:bool=True, Translation="0,0,0", Rotation="-1.235,0.928,-11.229") -> (Pose) {
    alphaInputType=Bool
    rotationMode=BMM_Additive
    rotationSpace=BCS_BoneSpace
    scaleSpace=BCS_BoneSpace
    translationMode=BMM_Additive
    translationSpace=BCS_BoneSpace
    alphaBoolBlend {blendOption=Linear}
    boneToModify {boneName=dyn_Hair_r_02}
    rotation=<rot -1.235 0.928 -11.229>
   }
   "Transform (Modify) Bone - Bone: dyn_Hair_r_base"(bAlphaBoolEnabled=True, Translation="0,1.25,0", Rotation="0,5,0") {
    alphaInputType=Bool
    rotationMode=BMM_Additive
    rotationSpace=BCS_BoneSpace
    scaleSpace=BCS_BoneSpace
    translationMode=BMM_Additive
    translationSpace=BCS_BoneSpace
    alphaBoolBlend {blendOption=Linear}
    boneToModify {boneName=dyn_Hair_r_base}
    rotation {yaw=5}
    translation {y=1.25}
   }
   Output Pose(Result=Use cached pose 'CRCache_Headphones'()) {layerGroup=None}
   n2 = Control Rig!(
    Source=Copy Pose From Mesh(SourceMeshComponent=Get Skel_Mesh()) {bUseAttachedParent=true}
    Alpha
    BackHairDynAlpha:double=Get AirFloatingAlpha() -> (AirFloatingAlpha:double)
    JumpFix:double=Get JumpFix() -> (JumpFix:double)
    CrouchADS_Fix:double=Get BackADS_Fix() -> (BackADS_Fix:double)
   ) -> (Pose) {
    alphaBoolBlend {blendOption=Linear}
    controlRigAssetReference {blueprintRigClass=&/F_MED_WaywardRebel_FaceAcc_CtrlRig.F_MED_WaywardRebel_FaceAcc_CtrlRig_C}
   }
   Save cached pose 'CRCache_HairPose'(Pose=@n2) {cacheName=CRCache_HairPose}
   n3 = Get Kill_Happened() -> (Kill_Happened:double)
   n4 = Get Spinning_Rotator() -> (Spinning_Rotator:Rotator)
   n5 = "Transform (Modify) Bone - Bone: dyn_headphone_01_l"(
    ComponentPose=Local To Component(LocalPose=Use cached pose 'CRCache_FakeWind'())
    Alpha=@n3
    Translation="0,0,0"
    Rotation=@n4
    Scale="1,1,1"
   ) {
    rotationMode=BMM_Additive
    rotationSpace=BCS_BoneSpace
    translationMode=BMM_Additive
    translationSpace=BCS_BoneSpace
    alphaBoolBlend {blendOption=Linear}
    alphaScaleBiasClamp {bInterpResult=true, interpSpeedDecreasing=5, interpSpeedIncreasing=5}
    boneToModify {boneName=dyn_headphone_01_l}
    rotation {yaw=359}
   }
   n6 = "Transform (Modify) Bone - Bone: dyn_headphone_01_r"(ComponentPose=@n5, Alpha=@n3, Translation="0,0,0", Rotation=@n4, Scale="1,1,1") {
    rotationMode=BMM_Additive
    rotationSpace=BCS_BoneSpace
    translationMode=BMM_Additive
    translationSpace=BCS_BoneSpace
    alphaBoolBlend {blendOption=Linear}
    alphaScaleBiasClamp {bInterpResult=true, interpSpeedDecreasing=5, interpSpeedIncreasing=5}
    boneToModify {boneName=dyn_headphone_01_r}
   }
   Save cached pose 'CRCache_Headphones'(Pose=Component To Local(ComponentPose=@n6)) {cacheName=CRCache_Headphones}
   n7 = "Transform (Modify) Bone - Bone: dyn_Hair_r_base"!(ComponentPose=Local To Component(LocalPose=Use cached pose 'CRCache_HairFrontDyn'()), Alpha:float=1, Translation="-1,2,0", Rotation="0,0,0") -> (Pose) {
    rotationMode=BMM_Additive
    rotationSpace=BCS_BoneSpace
    scaleSpace=BCS_BoneSpace
    translationMode=BMM_Additive
    translationSpace=BCS_BoneSpace
    alphaBoolBlend {blendOption=Linear}
    boneToModify {boneName=dyn_Hair_r_base}
    translation=<-1 2>
   }
   n8 = "Transform (Modify) Bone - Bone: dyn_Hair_l_base"!(ComponentPose=@n7, Alpha=1, Translation="-1,2,0", Rotation="0,-5,0") -> (Pose) {
    rotationMode=BMM_Additive
    rotationSpace=BCS_BoneSpace
    scaleSpace=BCS_BoneSpace
    translationMode=BMM_Additive
    translationSpace=BCS_BoneSpace
    alphaBoolBlend {blendOption=Linear}
    boneToModify {boneName=dyn_Hair_l_base}
    rotation {yaw=-5}
    translation=<-1 2>
   }
   n9 = "Transform (Modify) Bone - Bone: dyn_Hair_l_base"(ComponentPose=@n8, Alpha=1, Translation="-1,0,-3", Rotation="0,-50,0") {
    rotationMode=BMM_Additive
    rotationSpace=BCS_BoneSpace
    scaleSpace=BCS_BoneSpace
    translationMode=BMM_Additive
    translationSpace=BCS_BoneSpace
    alphaBoolBlend {blendOption=Linear}
    boneToModify {boneName=dyn_Hair_l_base}
    rotation {yaw=-50}
    translation {x=-1, z=-3}
   }
   Save cached pose 'CRCache_FrontFix'(Pose=Component To Local(ComponentPose=@n9)) {cacheName=CRCache_FrontFix}
   Save cached pose 'CRCache_DynBlend'(Pose=Blend Poses by bool(
     BlendPose_0=Use cached pose 'CRCache_HairBangsDyn'()
     BlendPose_1=Use cached pose 'CRCache_HairPose'()
     bActiveValue=Get bAnimDynamicsEnabled() -> (bAnimDynamicsEnabled:bool)
    ) {blendType=Linear}) {cacheName=CRCache_DynBlend}
   n10 = Get FakeWindAlpha() -> (FakeWindAlpha:double)
   n11 = "Transform (Modify) Bone - Bone: dyn_Hair_Back_03"!(
    ComponentPose=Local To Component(LocalPose=Use cached pose 'CRCache_BackFix'())
    Alpha=@n10
    Translation="0,0,-5"
    Rotation=Get sineRotate4() -> (sineRotate4:Rotator)
   ) -> (Pose) {
    rotationMode=BMM_Additive
    rotationSpace=BCS_BoneSpace
    translationMode=BMM_Additive
    translationSpace=BCS_BoneSpace
    alphaBoolBlend {blendOption=Linear}
    alphaScaleBiasClamp {bInterpResult=true, interpSpeedDecreasing=5, interpSpeedIncreasing=5}
    boneToModify {boneName=dyn_Hair_Back_03}
    translation {z=-5}
   }
   n12 = "Transform (Modify) Bone - Bone: dyn_Hair_BackR_03"!(ComponentPose=@n11, Alpha=@n10, Translation="0,0,-3", Rotation=Get sineRotate5() -> (sineRotate5:Rotator)) -> (Pose) {
    rotationMode=BMM_Additive
    rotationSpace=BCS_BoneSpace
    translationMode=BMM_Additive
    translationSpace=BCS_BoneSpace
    alphaBoolBlend {blendOption=Linear}
    alphaScaleBiasClamp {bInterpResult=true, interpSpeedDecreasing=5, interpSpeedIncreasing=5}
    boneToModify {boneName=dyn_Hair_BackR_03}
    translation {z=-3}
   }
   n13 = "Transform (Modify) Bone - Bone: dyn_Hair_BackL_03"!(ComponentPose=@n12, Alpha=@n10, Translation="0,0,-3", Rotation=Get sineRotate6() -> (sineRotate6:Rotator)) -> (Pose) {
    rotationMode=BMM_Additive
    rotationSpace=BCS_BoneSpace
    translationMode=BMM_Additive
    translationSpace=BCS_BoneSpace
    alphaBoolBlend {blendOption=Linear}
    alphaScaleBiasClamp {bInterpResult=true, interpSpeedDecreasing=5, interpSpeedIncreasing=5}
    boneToModify {boneName=dyn_Hair_BackL_03}
    translation {z=-3}
   }
   n14 = Get FakeWindAlpha()
   n15 = "Transform (Modify) Bone - Bone: dyn_Hair_Back_03"(
    ComponentPose=Local To Component(LocalPose=Use cached pose 'CRCache_BackFix'())
    Alpha=@n14
    Translation="0,0,0"
    Rotation=Get sineRotate1() -> (sineRotate1:Rotator)
   ) {
    rotationMode=BMM_Additive
    rotationSpace=BCS_BoneSpace
    translationMode=BMM_Additive
    translationSpace=BCS_BoneSpace
    alphaBoolBlend {blendOption=Linear}
    alphaScaleBiasClamp {bInterpResult=true, interpSpeedDecreasing=5, interpSpeedIncreasing=5}
    boneToModify {boneName=dyn_Hair_Back_03}
   }
   n16 = "Transform (Modify) Bone - Bone: dyn_Hair_BackR_03"(ComponentPose=@n15, Alpha=@n14, Translation="0,1,0", Rotation=Get sineRotate2() -> (sineRotate2:Rotator)) {
    rotationMode=BMM_Additive
    rotationSpace=BCS_BoneSpace
    translationMode=BMM_Additive
    translationSpace=BCS_BoneSpace
    alphaBoolBlend {blendOption=Linear}
    alphaScaleBiasClamp {bInterpResult=true, interpSpeedDecreasing=5, interpSpeedIncreasing=5}
    boneToModify {boneName=dyn_Hair_BackR_03}
    translation {y=1}
   }
   n17 = "Transform (Modify) Bone - Bone: dyn_Hair_BackL_03"(ComponentPose=@n16, Alpha=@n14, Translation="0,0,0", Rotation=Get sineRotate3() -> (sineRotate3:Rotator)) {
    rotationMode=BMM_Additive
    rotationSpace=BCS_BoneSpace
    translationMode=BMM_Additive
    translationSpace=BCS_BoneSpace
    alphaBoolBlend {blendOption=Linear}
    alphaScaleBiasClamp {bInterpResult=true, interpSpeedDecreasing=5, interpSpeedIncreasing=5}
    boneToModify {boneName=dyn_Hair_BackL_03}
   }
   Save cached pose 'CRCache_FakeWind'(Pose=Blend Poses by bool(
     BlendPose_0=Component To Local(ComponentPose=@n13)
     BlendPose_1=Component To Local(ComponentPose=@n17)
     bActiveValue=Get isWearingPackBool() -> (isWearingPackBool:bool)
    ) {blendType=Linear}) {cacheName=CRCache_FakeWind}
   Save cached pose 'CRCache_HairBackDyn'(Pose=Blend Poses by bool(
     BlendPose_0=Use cached pose 'CRCache_HairBack_RB'()
     BlendPose_1=Use cached pose 'CRCache_HairBack_AD'()
     bActiveValue=Get bIsParachuteOpened()
    ) {blendType=Linear}) {cacheName=CRCache_HairBackDyn}
   n18 = Two Way Blend(A=Use cached pose 'CRCache_HairBackDyn'(), B=Use cached pose 'CRCache_HairPose'(), Alpha=Get WearPackAlpha() -> (WearPackAlpha:double)) {blendNode {alphaBoolBlend {blendOption=Linear}}}
   n19 = Use cached pose 'CRCache_HairBackDyn'()
   n20 = Get DiveCapeAlpha() -> (DiveCapeAlpha:bool)
   n21 = Get is_WearingCape() -> (is_WearingCape:bool)
   n22 = "Transform (Modify) Bone - Bone: dyn_Hair_Back_04"!(
    ComponentPose=Local To Component(LocalPose=Blend Poses by bool(BlendPose_0=@n18, BlendPose_1=@n19, bActiveValue=@n20) {blendType=Linear})
    bAlphaBoolEnabled:bool=@n21
    Scale="0.7,1,1"
   ) -> (Pose) {
    alphaInputType=Bool
    rotationSpace=BCS_BoneSpace
    scaleMode=BMM_Additive
    scaleSpace=BCS_BoneSpace
    translationSpace=BCS_BoneSpace
    alphaBoolBlend {blendOption=Linear}
    boneToModify {boneName=dyn_Hair_Back_04}
    scale {x=0.7}
   }
   n23 = "Transform (Modify) Bone - Bone: dyn_Hair_BackL_04"!(ComponentPose=@n22, bAlphaBoolEnabled:bool=@n21, Scale="0.7,1,1") -> (Pose) {
    alphaInputType=Bool
    rotationSpace=BCS_BoneSpace
    scaleMode=BMM_Additive
    scaleSpace=BCS_BoneSpace
    translationSpace=BCS_BoneSpace
    alphaBoolBlend {blendOption=Linear}
    boneToModify {boneName=dyn_Hair_BackL_04}
    scale {x=0.7}
   }
   n24 = "Transform (Modify) Bone - Bone: dyn_Hair_BackR_04"!(ComponentPose=@n23, bAlphaBoolEnabled:bool=@n21, Scale="0.7,1,1") -> (Pose) {
    alphaInputType=Bool
    rotationSpace=BCS_BoneSpace
    scaleMode=BMM_Additive
    scaleSpace=BCS_BoneSpace
    translationSpace=BCS_BoneSpace
    alphaBoolBlend {blendOption=Linear}
    boneToModify {boneName=dyn_Hair_BackR_04}
    scale {x=0.7}
   }
   n25 = "Transform (Modify) Bone - Bone: dyn_Hair_Back_04"!(ComponentPose=@n24, bAlphaBoolEnabled=True, Rotation:Rotator="0,-3,0", Scale="0.8,0.85,1") -> (Pose) {
    alphaInputType=Bool
    rotationMode=BMM_Additive
    rotationSpace=BCS_BoneSpace
    scaleMode=BMM_Additive
    scaleSpace=BCS_BoneSpace
    translationSpace=BCS_BoneSpace
    alphaBoolBlend {blendOption=Linear}
    boneToModify {boneName=dyn_Hair_Back_04}
    rotation {yaw=-3}
    scale=<0.8 0.85>
   }
   Save cached pose 'CRCache_Backpack'(Pose=Component To Local(ComponentPose=@n25)) {cacheName=CRCache_Backpack}
   n26 = RigidBody!(
    ComponentPose=Local To Component(LocalPose=Use cached pose 'CRCache_HairPose'())
    Alpha=0.5
    OverrideWorldGravity:Vector="Get Gravity Override "() -> ("Gravity Override ":Vector)
    ExternalForce
   ) -> (Pose) {
    bOverrideWorldGravity=true
    overridePhysicsAsset=&/F_MED_WaywardRebel_FaceAcc_BackHair_Physics
    simulationSpace=BaseBoneSpace
    alphaBoolBlend {blendOption=Linear}
    baseBoneRef {boneName=spine_05}
    componentAppliedLinearAccClamp=<3000 3000 3000>
    componentLinearAccScale=<0.3 0.3 0.4>
    componentLinearVelScale=<0.3 0.3 0.4>
    simSpaceSettings {
     maxAngularAcceleration=1000
     maxAngularVelocity=3
     maxLinearAcceleration=700
     maxLinearVelocity=500
     worldAlpha=0.3
     externalLinearDragV=<0.2 0.2 0.1>
    }
   }
   Save cached pose 'CRCache_HairBack_RB'(Pose=Component To Local(ComponentPose=@n26)) {cacheName=CRCache_HairBack_RB}
   n27 = Anim Dynamics - Bone: dyn_Hair_BackL_02!(
    SimulationSpace:AnimPhysSimSpaceType Enum=World
    ComponentPose=Local To Component(LocalPose=Use cached pose 'CRCache_HairPose'())
    Alpha=0.3
    GravityOverride:Vector="0,0,-980"
   ) -> (Pose) {
    angularDampingOverride=1
    angularSpringConstant=10000
    bAngularSpring=true
    bOverrideAngularDamping=true
    bUseGravityOverride=true
    gravityScale=0.35
    simulationSpace=World
    alphaBoolBlend {blendOption=Linear}
    boundBone {boneName=dyn_Hair_BackL_02}
    componentAppliedLinearAccClamp=<2000 2000 2000>
    componentLinearAccScale=<0.5 0.5 0.5>
    componentLinearVelScale=<0.5 0.5 0.5>
    gravityOverride {z=-980}
    physicsBodyDefinitions [{
      boundBone {boneName=dyn_Hair_BackL_02}
      boxExtents=<25 25 5>
      constraintSetup {angularLimitsMax=<0 2 0>, angularLimitsMin=<0 -20 0>, angularTarget=<1 0 0>}
      localJointOffset=<2 0 0>
     }]
    relativeSpaceBone {boneName=spine_02}
   }
   n28 = Anim Dynamics - Bone: dyn_Hair_BackL_03!(SimulationSpace:AnimPhysSimSpaceType Enum, ComponentPose=@n27, Alpha=1, GravityOverride:Vector="0,0,-980") -> (Pose) {
    angularDampingOverride=1
    angularSpringConstant=20000
    bAngularSpring=true
    bGravityOverrideInSimSpace=true
    bOverrideAngularDamping=true
    simulationSpace=RootRelative
    alphaBoolBlend {blendOption=Linear}
    boundBone {boneName=dyn_Hair_BackL_03}
    componentAppliedLinearAccClamp=<2000 2000 2000>
    componentLinearAccScale=<0.5 0.5 0.5>
    componentLinearVelScale=<0.5 0.5 0.5>
    gravityOverride {z=-980}
    physicsBodyDefinitions [{
      boundBone {boneName=dyn_Hair_BackL_03}
      boxExtents=<25 25 5>
      constraintSetup {angularLimitsMax=<0 2 0>, angularLimitsMin=<0 -20 0>, angularTarget=<1 0 0>}
      localJointOffset=<2 0 0>
     }]
    relativeSpaceBone {boneName=spine_03}
   }
   n29 = Anim Dynamics - Bone: dyn_Hair_BackL_04!(SimulationSpace:AnimPhysSimSpaceType Enum=World, ComponentPose=@n28, Alpha=1, GravityOverride:Vector="0,0,-980") -> (Pose) {
    angularDampingOverride=1
    angularSpringConstant=8000
    bAngularSpring=true
    bGravityOverrideInSimSpace=true
    bOverrideAngularDamping=true
    simulationSpace=World
    alphaBoolBlend {blendOption=Linear}
    alphaScaleBiasClamp {bInterpResult=true}
    boundBone {boneName=dyn_Hair_BackL_04}
    componentAppliedLinearAccClamp=<2000 2000 2000>
    componentLinearAccScale=<0.5 0.5 0.5>
    componentLinearVelScale=<0.5 0.5 0.5>
    gravityOverride {z=-980}
    physicsBodyDefinitions [{
      boundBone {boneName=dyn_Hair_BackL_04}
      boxExtents=<25 23 23>
      constraintSetup {angularLimitsMax=<0 0 20>, angularLimitsMin=<0 -70 -20>, angularTarget=<1 0 0>}
      localJointOffset=<2 0 0>
     }]
    relativeSpaceBone {boneName=root}
   }
   n30 = Anim Dynamics - Bone: dyn_Hair_BackL_05!(SimulationSpace:AnimPhysSimSpaceType Enum=World, ComponentPose=@n29, Alpha=1, GravityOverride:Vector="0,0,-980") -> (Pose) {
    angularDampingOverride=1
    angularSpringConstant=5000
    bAngularSpring=true
    bGravityOverrideInSimSpace=true
    bOverrideAngularDamping=true
    simulationSpace=World
    alphaBoolBlend {blendOption=Linear}
    alphaScaleBiasClamp {bInterpResult=true}
    boundBone {boneName=dyn_Hair_BackL_05}
    componentAppliedLinearAccClamp=<2000 2000 2000>
    componentLinearAccScale=<0.5 0.5 0.5>
    componentLinearVelScale=<0.5 0.5 0.5>
    gravityOverride {z=-980}
    physicsBodyDefinitions [{
      boundBone {boneName=dyn_Hair_BackL_05}
      boxExtents=<25 23 23>
      constraintSetup {angularLimitsMax=<0 0 30>, angularLimitsMin=<0 -70 -30>, angularTarget=<1 0 0>}
      localJointOffset=<2 0 0>
     }]
    relativeSpaceBone {boneName=root}
   }
   n31 = Anim Dynamics - Bone: dyn_Hair_BackR_02!(SimulationSpace:AnimPhysSimSpaceType Enum=World, ComponentPose=@n30, Alpha=0.3, GravityOverride:Vector="0,0,-980") -> (Pose) {
    angularDampingOverride=1
    angularSpringConstant=10000
    bAngularSpring=true
    bOverrideAngularDamping=true
    bUseGravityOverride=true
    gravityScale=0.35
    simulationSpace=World
    alphaBoolBlend {blendOption=Linear}
    boundBone {boneName=dyn_Hair_BackR_02}
    componentAppliedLinearAccClamp=<2000 2000 2000>
    componentLinearAccScale=<0.5 0.5 0.5>
    componentLinearVelScale=<0.5 0.5 0.5>
    gravityOverride {z=-980}
    physicsBodyDefinitions [{
      boundBone {boneName=dyn_Hair_BackR_02}
      boxExtents=<25 25 5>
      constraintSetup {angularLimitsMax=<0 2 0>, angularLimitsMin=<0 -20 0>, angularTarget=<1 0 0>}
      localJointOffset=<2 0 0>
     }]
    relativeSpaceBone {boneName=spine_02}
   }
   n32 = Anim Dynamics - Bone: dyn_Hair_Back_03!(SimulationSpace:AnimPhysSimSpaceType Enum, ComponentPose=@n31, Alpha=1, GravityOverride:Vector="0,0,-980") -> (Pose) {
    angularDampingOverride=1
    angularSpringConstant=20000
    bAngularSpring=true
    bGravityOverrideInSimSpace=true
    bOverrideAngularDamping=true
    simulationSpace=RootRelative
    alphaBoolBlend {blendOption=Linear}
    boundBone {boneName=dyn_Hair_Back_03}
    componentAppliedLinearAccClamp=<2000 2000 2000>
    componentLinearAccScale=<0.5 0.5 0.5>
    componentLinearVelScale=<0.5 0.5 0.5>
    gravityOverride {z=-980}
    physicsBodyDefinitions [{
      boundBone {boneName=dyn_Hair_Back_03}
      boxExtents=<25 25 5>
      constraintSetup {angularLimitsMax=<0 2 0>, angularLimitsMin=<0 -20 0>, angularTarget=<1 0 0>}
      localJointOffset=<2 0 0>
     }]
    relativeSpaceBone {boneName=spine_03}
   }
   n33 = Anim Dynamics - Bone: dyn_Hair_Back_04!(SimulationSpace:AnimPhysSimSpaceType Enum=World, ComponentPose=@n32, Alpha=1, GravityOverride:Vector="0,0,-980") -> (Pose) {
    angularDampingOverride=1
    angularSpringConstant=8000
    bAngularSpring=true
    bGravityOverrideInSimSpace=true
    bOverrideAngularDamping=true
    simulationSpace=World
    alphaBoolBlend {blendOption=Linear}
    alphaScaleBiasClamp {bInterpResult=true}
    boundBone {boneName=dyn_Hair_Back_04}
    componentAppliedLinearAccClamp=<2000 2000 2000>
    componentLinearAccScale=<0.5 0.5 0.5>
    componentLinearVelScale=<0.5 0.5 0.5>
    gravityOverride {z=-980}
    physicsBodyDefinitions [{
      boundBone {boneName=dyn_Hair_Back_04}
      boxExtents=<25 23 23>
      constraintSetup {angularLimitsMax=<0 0 20>, angularLimitsMin=<0 -70 -20>, angularTarget=<1 0 0>}
      localJointOffset=<2 0 0>
     }]
    relativeSpaceBone {boneName=root}
   }
   n34 = Anim Dynamics - Bone: dyn_Hair_Back_05!(SimulationSpace:AnimPhysSimSpaceType Enum=World, ComponentPose=@n33, Alpha=1, GravityOverride:Vector="0,0,-980") -> (Pose) {
    angularDampingOverride=1
    angularSpringConstant=5000
    bAngularSpring=true
    bGravityOverrideInSimSpace=true
    bOverrideAngularDamping=true
    simulationSpace=World
    alphaBoolBlend {blendOption=Linear}
    alphaScaleBiasClamp {bInterpResult=true}
    boundBone {boneName=dyn_Hair_Back_05}
    componentAppliedLinearAccClamp=<2000 2000 2000>
    componentLinearAccScale=<0.5 0.5 0.5>
    componentLinearVelScale=<0.5 0.5 0.5>
    gravityOverride {z=-980}
    physicsBodyDefinitions [{
      boundBone {boneName=dyn_Hair_Back_05}
      boxExtents=<25 23 23>
      constraintSetup {angularLimitsMax=<0 0 30>, angularLimitsMin=<0 -70 -30>, angularTarget=<1 0 0>}
      localJointOffset=<2 0 0>
     }]
    relativeSpaceBone {boneName=root}
   }
   n35 = Anim Dynamics - Bone: dyn_Hair_BackR_02(SimulationSpace=World, ComponentPose=@n34, Alpha=0.3, GravityOverride="0,0,-980") {
    angularDampingOverride=1
    angularSpringConstant=10000
    bAngularSpring=true
    bOverrideAngularDamping=true
    bUseGravityOverride=true
    gravityScale=0.35
    simulationSpace=World
    alphaBoolBlend {blendOption=Linear}
    boundBone {boneName=dyn_Hair_BackR_02}
    componentAppliedLinearAccClamp=<2000 2000 2000>
    componentLinearAccScale=<0.5 0.5 0.5>
    componentLinearVelScale=<0.5 0.5 0.5>
    gravityOverride {z=-980}
    physicsBodyDefinitions [{
      boundBone {boneName=dyn_Hair_BackR_02}
      boxExtents=<25 25 5>
      constraintSetup {angularLimitsMax=<0 2 0>, angularLimitsMin=<0 -20 0>, angularTarget=<1 0 0>}
      localJointOffset=<2 0 0>
     }]
    relativeSpaceBone {boneName=spine_02}
   }
   n36 = Anim Dynamics - Bone: dyn_Hair_BackR_03!(SimulationSpace:AnimPhysSimSpaceType Enum, ComponentPose=@n35, Alpha=1, GravityOverride:Vector="0,0,-980") -> (Pose) {
    angularDampingOverride=1
    angularSpringConstant=20000
    bAngularSpring=true
    bGravityOverrideInSimSpace=true
    bOverrideAngularDamping=true
    simulationSpace=RootRelative
    alphaBoolBlend {blendOption=Linear}
    boundBone {boneName=dyn_Hair_BackR_03}
    componentAppliedLinearAccClamp=<2000 2000 2000>
    componentLinearAccScale=<0.5 0.5 0.5>
    componentLinearVelScale=<0.5 0.5 0.5>
    gravityOverride {z=-980}
    physicsBodyDefinitions [{
      boundBone {boneName=dyn_Hair_BackR_03}
      boxExtents=<25 25 5>
      constraintSetup {angularLimitsMax=<0 2 0>, angularLimitsMin=<0 -20 0>, angularTarget=<1 0 0>}
      localJointOffset=<2 0 0>
     }]
    relativeSpaceBone {boneName=spine_03}
   }
   n37 = Anim Dynamics - Bone: dyn_Hair_BackR_04!(SimulationSpace:AnimPhysSimSpaceType Enum=World, ComponentPose=@n36, Alpha=1, GravityOverride:Vector="0,0,-980") -> (Pose) {
    angularDampingOverride=1
    angularSpringConstant=8000
    bAngularSpring=true
    bGravityOverrideInSimSpace=true
    bOverrideAngularDamping=true
    simulationSpace=World
    alphaBoolBlend {blendOption=Linear}
    alphaScaleBiasClamp {bInterpResult=true}
    boundBone {boneName=dyn_Hair_BackR_04}
    componentAppliedLinearAccClamp=<2000 2000 2000>
    componentLinearAccScale=<0.5 0.5 0.5>
    componentLinearVelScale=<0.5 0.5 0.5>
    gravityOverride {z=-980}
    physicsBodyDefinitions [{
      boundBone {boneName=dyn_Hair_BackR_04}
      boxExtents=<25 23 23>
      constraintSetup {angularLimitsMax=<0 0 20>, angularLimitsMin=<0 -70 -20>, angularTarget=<1 0 0>}
      localJointOffset=<2 0 0>
     }]
    relativeSpaceBone {boneName=root}
   }
   n38 = Anim Dynamics - Bone: dyn_Hair_BackR_05!(SimulationSpace:AnimPhysSimSpaceType Enum=World, ComponentPose=@n37, Alpha=1, GravityOverride:Vector="0,0,-980") -> (Pose) {
    angularDampingOverride=1
    angularSpringConstant=5000
    bAngularSpring=true
    bGravityOverrideInSimSpace=true
    bOverrideAngularDamping=true
    simulationSpace=World
    alphaBoolBlend {blendOption=Linear}
    alphaScaleBiasClamp {bInterpResult=true}
    boundBone {boneName=dyn_Hair_BackR_05}
    componentAppliedLinearAccClamp=<2000 2000 2000>
    componentLinearAccScale=<0.5 0.5 0.5>
    componentLinearVelScale=<0.5 0.5 0.5>
    gravityOverride {z=-980}
    physicsBodyDefinitions [{
      boundBone {boneName=dyn_Hair_BackR_05}
      boxExtents=<25 23 23>
      constraintSetup {angularLimitsMax=<0 0 30>, angularLimitsMin=<0 -70 -30>, angularTarget=<1 0 0>}
      localJointOffset=<2 0 0>
     }]
    relativeSpaceBone {boneName=root}
   }
   Save cached pose 'CRCache_HairBack_AD'(Pose=Component To Local(ComponentPose=@n38)) {cacheName=CRCache_HairBack_AD}
   n39 = Anim Dynamics - Bone: dyn_Hair_a(ComponentPose=Local To Component(LocalPose=Use cached pose 'CRCache_Backpack'()), Alpha=1, ExternalForce="0,0,0") {
    angularDampingOverride=1
    angularSpringConstant=20000
    bAngularSpring=true
    bOverrideAngularDamping=true
    bUseGravityOverride=true
    gravityScale=0.35
    simulationSpace=BoneRelative
    alphaBoolBlend {blendOption=Linear}
    boundBone {boneName=dyn_Hair_a}
    componentAppliedLinearAccClamp=<3000 3000 3000>
    componentLinearAccScale=<0.3 0.3 0.3>
    componentLinearVelScale=<0.3 0.3 0.3>
    gravityOverride {z=-980}
    physicsBodyDefinitions [{
      boundBone {boneName=dyn_Hair_a}
      boxExtents=<20 30 30>
      constraintSetup {angularLimitsMax=<5 15 25>, angularLimitsMin=<-5 -15 -25>, angularTarget=<1 0 0>}
      localJointOffset=<1 0 0>
     }]
    relativeSpaceBone {boneName=spine_03}
   }
   n40 = Anim Dynamics - Bone: dyn_Hair_b01(ComponentPose=@n39, Alpha=1, ExternalForce="0,0,0") {
    angularDampingOverride=1
    angularSpringConstant=20000
    bAngularSpring=true
    bOverrideAngularDamping=true
    bUseGravityOverride=true
    gravityScale=0.85
    simulationSpace=BoneRelative
    alphaBoolBlend {blendOption=Linear}
    boundBone {boneName=dyn_Hair_b01}
    componentAppliedLinearAccClamp=<1000 1000 1000>
    componentLinearAccScale=<0.5 0.5 0.5>
    componentLinearVelScale=<0.5 0.5 0.5>
    gravityOverride {z=-980}
    physicsBodyDefinitions [{
      boundBone {boneName=dyn_Hair_b01}
      boxExtents=<22 22 22>
      constraintSetup {angularLimitsMax=<5 5 25>, angularLimitsMin=<-5 -15 -25>, angularTarget=<1 0 0>}
      localJointOffset=<1.5 0 0>
     }]
    relativeSpaceBone {boneName=spine_04}
   }
   n41 = Anim Dynamics - Bone: dyn_Hair_c(ComponentPose=@n40, Alpha=1, ExternalForce="0,0,0") {
    angularDampingOverride=1
    angularSpringConstant=20000
    bAngularSpring=true
    bOverrideAngularDamping=true
    bUseGravityOverride=true
    gravityScale=0.85
    simulationSpace=BoneRelative
    alphaBoolBlend {blendOption=Linear}
    boundBone {boneName=dyn_Hair_c}
    componentAppliedLinearAccClamp=<1000 1000 1000>
    componentLinearAccScale=<0.5 0.5 0.5>
    componentLinearVelScale=<0.5 0.5 0.5>
    gravityOverride {z=-980}
    physicsBodyDefinitions [{
      boundBone {boneName=dyn_Hair_c}
      boxExtents=<24 24 24>
      constraintSetup {angularLimitsMax=<5 5 25>, angularLimitsMin=<-5 -15 -25>, angularTarget=<1 0 0>}
      localJointOffset=<1.5 0 0>
     }]
    relativeSpaceBone {boneName=spine_05}
   }
   n42 = Anim Dynamics - Bone: dyn_Hair_d01(ComponentPose=@n41, Alpha=1, ExternalForce="0,0,0") {
    angularDampingOverride=1
    angularSpringConstant=20000
    bAngularSpring=true
    bOverrideAngularDamping=true
    bUseGravityOverride=true
    gravityScale=0.85
    simulationSpace=BoneRelative
    alphaBoolBlend {blendOption=Linear}
    boundBone {boneName=dyn_Hair_d01}
    componentAppliedLinearAccClamp=<1000 1000 1000>
    componentLinearAccScale=<0.5 0.5 0.5>
    componentLinearVelScale=<0.5 0.5 0.5>
    gravityOverride {z=-980}
    physicsBodyDefinitions [{
      boundBone {boneName=dyn_Hair_d01}
      boxExtents=<23 23 23>
      constraintSetup {angularLimitsMax=<5 5 25>, angularLimitsMin=<-5 -15 -25>, angularTarget=<1 0 0>}
      localJointOffset=<1.5 0 0>
     }]
    relativeSpaceBone {boneName=spine_04}
   }
   n43 = Anim Dynamics - Bone: dyn_Hair_e01(ComponentPose=@n42, Alpha=1, ExternalForce="0,0,0") {
    angularDampingOverride=1
    angularSpringConstant=20000
    bAngularSpring=true
    bOverrideAngularDamping=true
    bUseGravityOverride=true
    gravityScale=0.85
    simulationSpace=BoneRelative
    alphaBoolBlend {blendOption=Linear}
    boundBone {boneName=dyn_Hair_e01}
    componentAppliedLinearAccClamp=<1000 1000 1000>
    componentLinearAccScale=<0.5 0.5 0.5>
    componentLinearVelScale=<0.5 0.5 0.5>
    gravityOverride {z=-980}
    physicsBodyDefinitions [{
      boundBone {boneName=dyn_Hair_e01}
      boxExtents=<25 25 25>
      constraintSetup {angularLimitsMax=<5 5 15>, angularLimitsMin=<-5 -15 -15>, angularTarget=<1 0 0>}
      localJointOffset=<1.5 0 0>
     }]
    relativeSpaceBone {boneName=spine_05}
   }
   Save cached pose 'CRCache_HairBangsDyn'(Pose=Component To Local(ComponentPose=@n43)) {cacheName=CRCache_HairBangsDyn}
   n44 = "Get Gravity Override "()
   n45 = RigidBody!(ComponentPose=Local To Component(LocalPose=Use cached pose 'CRCache_DynBlend'()), Alpha, OverrideWorldGravity=@n44) -> (Pose) {
    bClampLinearTranslationLimitToRefPose=true
    bOverrideWorldGravity=true
    overridePhysicsAsset=&/F_MED_WaywardRebel_FaceAcc_FrontHair_Physics
    simulationSpace=BaseBoneSpace
    alphaBoolBlend {blendOption=Linear}
    baseBoneRef {boneName=spine_03}
    componentAppliedLinearAccClamp=<3500 3500 3500>
    componentLinearAccScale=<0.5 0.8 0.5>
    componentLinearVelScale=<0.8 0.8 0.8>
    overrideWorldGravity {z=-980}
    simSpaceSettings {
     maxAngularAcceleration=1000
     maxAngularVelocity=0.3
     maxLinearAcceleration=700
     maxLinearVelocity=500
     velocityScaleZ=0
     worldAlpha=0.8
     externalLinearDragV=<0.5 0.5>
    }
   }
   n46 = RigidBody(ComponentPose=Local To Component(LocalPose=Use cached pose 'CRCache_DynBlend'()), OverrideWorldGravity=@n44) {
    bClampLinearTranslationLimitToRefPose=true
    bOverrideWorldGravity=true
    overridePhysicsAsset=&/F_MED_WaywardRebel_FaceAcc_Physics
    simulationSpace=BaseBoneSpace
    alphaBoolBlend {blendOption=Linear}
    baseBoneRef {boneName=spine_03}
    componentAppliedLinearAccClamp=<3500 3500 3500>
    componentLinearAccScale=<0.5 0.8 0.5>
    componentLinearVelScale=<0.8 0.8 0.8>
    overrideWorldGravity {z=-980}
    simSpaceSettings {
     maxAngularAcceleration=1000
     maxAngularVelocity=0.3
     maxLinearAcceleration=700
     maxLinearVelocity=500
     velocityScaleZ=0
     worldAlpha=0.8
     externalLinearDragV=<0.5 0.5>
    }
   }
   n47 = "Transform (Modify) Bone - Bone: dyn_Hair_r_base"!(ComponentPose=@n46, bAlphaBoolEnabled:bool=True, Translation="0,2.5,0", Rotation="0,5,-10") -> (Pose) {
    alphaInputType=Bool
    rotationMode=BMM_Additive
    rotationSpace=BCS_BoneSpace
    scaleSpace=BCS_BoneSpace
    translationMode=BMM_Additive
    translationSpace=BCS_BoneSpace
    alphaBoolBlend {blendOption=Linear}
    boneToModify {boneName=dyn_Hair_r_base}
    rotation {roll=-10, yaw=5}
    translation {y=2.5}
   }
   n48 = "Transform (Modify) Bone - Bone: dyn_Hair_r_base"(ComponentPose=@n47, bAlphaBoolEnabled=True, Translation="0,1.5,0", Rotation="0.3,1.43,-11.35") {
    alphaInputType=Bool
    rotationMode=BMM_Additive
    rotationSpace=BCS_BoneSpace
    scaleSpace=BCS_BoneSpace
    translationMode=BMM_Additive
    translationSpace=BCS_BoneSpace
    alphaBoolBlend {blendOption=Linear}
    boneToModify {boneName=dyn_Hair_r_base}
    rotation=<rot 0.3 1.43 -11.35>
    translation {y=1.5}
   }
   n49 = "Transform (Modify) Bone - Bone: dyn_Hair_r_02"(ComponentPose=@n48, bAlphaBoolEnabled=True, Translation="0,0,0", Rotation="-14.207,5.127,2.502") {
    alphaInputType=Bool
    rotationMode=BMM_Additive
    rotationSpace=BCS_BoneSpace
    scaleSpace=BCS_BoneSpace
    translationMode=BMM_Additive
    alphaBoolBlend {blendOption=Linear}
    boneToModify {boneName=dyn_Hair_r_02}
    rotation=<rot -14.207 5.127 2.502>
   }
   n50 = "Transform (Modify) Bone - Bone: dyn_Hair_l_base"!(ComponentPose=@n49, bAlphaBoolEnabled:bool=True, Translation="0,0,0", Rotation="25,6,5") -> (Pose) {
    alphaInputType=Bool
    rotationMode=BMM_Additive
    rotationSpace=BCS_BoneSpace
    scaleSpace=BCS_BoneSpace
    translationMode=BMM_Additive
    translationSpace=BCS_BoneSpace
    alphaBoolBlend {blendOption=Linear}
    boneToModify {boneName=dyn_Hair_l_base}
    rotation=<rot 25 6 5>
   }
   Save cached pose 'CRCache_HairFrontDyn'(Pose=Blend Poses by bool(
     BlendPose_0=Component To Local(ComponentPose=@n45)
     BlendPose_1=Component To Local(ComponentPose=@n50)
     bActiveValue=Get FrontHairAlpha() -> (FrontHairAlpha:bool)
    ) {blendType=Linear}) {cacheName=CRCache_HairFrontDyn}
   n51 = Get Targeting() -> (Targeting:double)
   n52 = "Transform (Modify) Bone - Bone: dyn_Hair_BackR_03"!(ComponentPose=Local To Component(LocalPose=Use cached pose 'CRCache_FrontFix'()), Alpha=@n51, Translation="0,0,1") -> (Pose) {
    rotationMode=BMM_Additive
    rotationSpace=BCS_BoneSpace
    translationMode=BMM_Additive
    translationSpace=BCS_BoneSpace
    alphaBoolBlend {blendOption=Linear}
    alphaScaleBiasClamp {bInterpResult=true}
    boneToModify {boneName=dyn_Hair_BackR_03}
    translation {z=1}
   }
   n53 = "Transform (Modify) Bone - Bone: dyn_Hair_BackL_03"!(ComponentPose=@n52, Alpha=@n51, Translation="0,0,1") -> (Pose) {
    rotationMode=BMM_Additive
    rotationSpace=BCS_BoneSpace
    translationMode=BMM_Additive
    translationSpace=BCS_BoneSpace
    alphaBoolBlend {blendOption=Linear}
    alphaScaleBiasClamp {bInterpResult=true}
    boneToModify {boneName=dyn_Hair_BackL_03}
    translation {z=1}
   }
   n54 = Get CrouchADS() -> (CrouchADS:double)
   n55 = "Transform (Modify) Bone - Bone: dyn_Hair_Back_03"!(ComponentPose=@n53, Alpha=@n54, Rotation="5,0,0") -> (Pose) {
    rotationMode=BMM_Additive
    rotationSpace=BCS_BoneSpace
    translationMode=BMM_Additive
    translationSpace=BCS_BoneSpace
    alphaBoolBlend {blendOption=Linear}
    alphaScaleBiasClamp {bInterpResult=true}
    boneToModify {boneName=dyn_Hair_Back_03}
    rotation {pitch=5}
   }
   n56 = "Transform (Modify) Bone - Bone: dyn_Hair_BackR_03"!(ComponentPose=@n55, Alpha=@n54, Rotation:Rotator="5,0,0") -> (Pose) {
    rotationMode=BMM_Additive
    rotationSpace=BCS_BoneSpace
    translationMode=BMM_Additive
    translationSpace=BCS_BoneSpace
    alphaBoolBlend {blendOption=Linear}
    alphaScaleBiasClamp {bInterpResult=true}
    boneToModify {boneName=dyn_Hair_BackR_03}
    rotation {pitch=5}
    translation {y=2}
   }
   n57 = "Transform (Modify) Bone - Bone: dyn_Hair_BackL_03"!(ComponentPose=@n56, Alpha=@n54, Rotation:Rotator="5,0,0") -> (Pose) {
    rotationMode=BMM_Additive
    rotationSpace=BCS_BoneSpace
    translationMode=BMM_Additive
    translationSpace=BCS_BoneSpace
    alphaBoolBlend {blendOption=Linear}
    alphaScaleBiasClamp {bInterpResult=true}
    boneToModify {boneName=dyn_Hair_BackL_03}
    rotation {pitch=5}
    translation {y=-2}
   }
   n58 = Get JumpMelee() -> (JumpMelee:double)
   n59 = "Transform (Modify) Bone - Bone: dyn_Hair_Back_03"!(ComponentPose=@n57, Alpha=@n58, Translation:Vector="-0.5,0,1") -> (Pose) {
    rotationMode=BMM_Additive
    rotationSpace=BCS_BoneSpace
    translationMode=BMM_Additive
    translationSpace=BCS_BoneSpace
    alphaBoolBlend {blendOption=Linear}
    alphaScaleBiasClamp {bInterpResult=true}
    boneToModify {boneName=dyn_Hair_Back_03}
    translation {x=-0.5, z=1}
   }
   n60 = "Transform (Modify) Bone - Bone: dyn_Hair_BackR_03"!(ComponentPose=@n59, Alpha=@n58, Translation:Vector="-0.5,0,1") -> (Pose) {
    rotationMode=BMM_Additive
    rotationSpace=BCS_BoneSpace
    translationMode=BMM_Additive
    translationSpace=BCS_BoneSpace
    alphaBoolBlend {blendOption=Linear}
    alphaScaleBiasClamp {bInterpResult=true}
    boneToModify {boneName=dyn_Hair_BackR_03}
    translation {x=-0.5, z=1}
   }
   n61 = "Transform (Modify) Bone - Bone: dyn_Hair_BackL_03"!(ComponentPose=@n60, Alpha=@n58, Translation:Vector="-0.5,0,1") -> (Pose) {
    rotationMode=BMM_Additive
    rotationSpace=BCS_BoneSpace
    translationMode=BMM_Additive
    translationSpace=BCS_BoneSpace
    alphaBoolBlend {blendOption=Linear}
    alphaScaleBiasClamp {bInterpResult=true}
    boneToModify {boneName=dyn_Hair_BackL_03}
    translation {x=-0.5, z=1}
   }
   n62 = Get Parachuting() -> (Parachuting:double)
   n63 = "Transform (Modify) Bone - Bone: dyn_Hair_Back_02"!(ComponentPose=@n61, Alpha=@n62, Translation="-2.75,0,4") -> (Pose) {
    rotationMode=BMM_Additive
    rotationSpace=BCS_BoneSpace
    translationMode=BMM_Additive
    translationSpace=BCS_BoneSpace
    alphaBoolBlend {blendOption=Linear}
    alphaScaleBiasClamp {bInterpResult=true}
    boneToModify {boneName=dyn_Hair_Back_02}
    rotation {pitch=2.5}
    translation {x=-2.75, z=4}
   }
   n64 = "Transform (Modify) Bone - Bone: dyn_Hair_BackR_02"!(ComponentPose=@n63, Alpha=@n62, Translation="-2.75,0,1.5") -> (Pose) {
    rotationMode=BMM_Additive
    rotationSpace=BCS_BoneSpace
    translationMode=BMM_Additive
    translationSpace=BCS_BoneSpace
    alphaBoolBlend {blendOption=Linear}
    alphaScaleBiasClamp {bInterpResult=true}
    boneToModify {boneName=dyn_Hair_BackR_02}
    translation {x=-2.75, z=1.5}
   }
   n65 = "Transform (Modify) Bone - Bone: dyn_Hair_BackL_02"!(ComponentPose=@n64, Alpha=@n62, Translation="-2.5,0,1.5") -> (Pose) {
    rotationMode=BMM_Additive
    rotationSpace=BCS_BoneSpace
    translationMode=BMM_Additive
    translationSpace=BCS_BoneSpace
    alphaBoolBlend {blendOption=Linear}
    alphaScaleBiasClamp {bInterpResult=true}
    boneToModify {boneName=dyn_Hair_BackL_02}
    translation {x=-2.5, z=1.5}
   }
   n66 = Get DBNO_Moving() -> (DBNO_Moving:double)
   n67 = "Transform (Modify) Bone - Bone: dyn_Hair_BackR_03"!(ComponentPose=@n65, Alpha=@n66, Translation="-1,1,0.5", Rotation:Rotator="2,0,10") -> (Pose) {
    rotationMode=BMM_Additive
    rotationSpace=BCS_BoneSpace
    translationMode=BMM_Additive
    translationSpace=BCS_BoneSpace
    alphaBoolBlend {blendOption=Linear}
    alphaScaleBiasClamp {bInterpResult=true}
    boneToModify {boneName=dyn_Hair_BackR_03}
    rotation {pitch=2, roll=10}
    translation=<-1 1 0.5>
   }
   n68 = "Transform (Modify) Bone - Bone: dyn_Hair_BackL_03"!(ComponentPose=@n67, Alpha=@n66, Translation="-1,-1,0.5", Rotation:Rotator="2,0,-10") -> (Pose) {
    rotationMode=BMM_Additive
    rotationSpace=BCS_BoneSpace
    translationMode=BMM_Additive
    translationSpace=BCS_BoneSpace
    alphaBoolBlend {blendOption=Linear}
    alphaScaleBiasClamp {bInterpResult=true}
    boneToModify {boneName=dyn_Hair_BackL_03}
    rotation {pitch=2, roll=-10}
    translation=<-1 -1 0.5>
   }
   n69 = Get SkydiveMode() -> (SkydiveMode:double)
   n70 = "Transform (Modify) Bone - Bone: dyn_Hair_Back_01"!(ComponentPose=@n68, Alpha=@n69, Rotation="-5,0,0") -> (Pose) {
    rotationMode=BMM_Additive
    rotationSpace=BCS_BoneSpace
    translationMode=BMM_Additive
    translationSpace=BCS_BoneSpace
    alphaBoolBlend {blendOption=Linear}
    alphaScaleBiasClamp {bInterpResult=true}
    boneToModify {boneName=dyn_Hair_Back_01}
    rotation {pitch=-5}
   }
   n71 = "Transform (Modify) Bone - Bone: dyn_Hair_BackR_01"!(ComponentPose=@n70, Alpha=@n69, Rotation="-10,0,0") -> (Pose) {
    rotationMode=BMM_Additive
    rotationSpace=BCS_BoneSpace
    translationMode=BMM_Additive
    translationSpace=BCS_BoneSpace
    alphaBoolBlend {blendOption=Linear}
    alphaScaleBiasClamp {bInterpResult=true}
    boneToModify {boneName=dyn_Hair_BackR_01}
    rotation {pitch=-10}
   }
   n72 = "Transform (Modify) Bone - Bone: dyn_Hair_BackL_01"!(ComponentPose=@n71, Alpha=@n69, Rotation="-10,0,0") -> (Pose) {
    rotationMode=BMM_Additive
    rotationSpace=BCS_BoneSpace
    translationMode=BMM_Additive
    translationSpace=BCS_BoneSpace
    alphaBoolBlend {blendOption=Linear}
    alphaScaleBiasClamp {bInterpResult=true}
    boneToModify {boneName=dyn_Hair_BackL_01}
    rotation {pitch=-10}
   }
   n73 = "Transform (Modify) Bone - Bone: dyn_Hair_Back_02"!(ComponentPose=@n72, Alpha=1, Rotation:Rotator="45,0,0") -> (Pose) {
    rotationMode=BMM_Additive
    rotationSpace=BCS_BoneSpace
    translationMode=BMM_Additive
    translationSpace=BCS_BoneSpace
    alphaBoolBlend {blendOption=Linear}
    alphaScaleBiasClamp {bInterpResult=true, interpSpeedDecreasing=5, interpSpeedIncreasing=5}
    boneToModify {boneName=dyn_Hair_Back_02}
    rotation {pitch=45}
    translation {x=-2}
   }
   n74 = "Transform (Modify) Bone - Bone: dyn_Hair_BackR_02"!(ComponentPose=@n73, Alpha=1, Rotation:Rotator="52,0,0") -> (Pose) {
    rotationMode=BMM_Additive
    rotationSpace=BCS_BoneSpace
    translationMode=BMM_Additive
    translationSpace=BCS_BoneSpace
    alphaBoolBlend {blendOption=Linear}
    alphaScaleBiasClamp {bInterpResult=true, interpSpeedDecreasing=5, interpSpeedIncreasing=5}
    boneToModify {boneName=dyn_Hair_BackR_02}
    rotation {pitch=52}
    translation {x=-2.5}
   }
   n75 = "Transform (Modify) Bone - Bone: dyn_Hair_BackL_02"!(ComponentPose=@n74, Alpha=1, Rotation:Rotator="52,0,0") -> (Pose) {
    rotationMode=BMM_Additive
    rotationSpace=BCS_BoneSpace
    translationMode=BMM_Additive
    translationSpace=BCS_BoneSpace
    alphaBoolBlend {blendOption=Linear}
    alphaScaleBiasClamp {bInterpResult=true, interpSpeedDecreasing=5, interpSpeedIncreasing=5}
    boneToModify {boneName=dyn_Hair_BackL_02}
    rotation {pitch=52}
    translation {x=-2.75}
   }
   n76 = "Transform (Modify) Bone - Bone: dyn_Hair_Back_03"!(ComponentPose=@n75, Alpha=1, Translation="0,0,2", Rotation:Rotator="-75,0,0") -> (Pose) {
    rotationMode=BMM_Additive
    rotationSpace=BCS_BoneSpace
    translationMode=BMM_Additive
    translationSpace=BCS_BoneSpace
    alphaBoolBlend {blendOption=Linear}
    alphaScaleBiasClamp {bInterpResult=true, interpSpeedDecreasing=5, interpSpeedIncreasing=5}
    boneToModify {boneName=dyn_Hair_Back_03}
    rotation {pitch=-75}
    translation {z=2}
   }
   n77 = "Transform (Modify) Bone - Bone: dyn_Hair_BackR_03"(ComponentPose=@n76, Alpha=1, Translation="0,0,4.5", Rotation="-60,0,0") {
    rotationMode=BMM_Additive
    rotationSpace=BCS_BoneSpace
    translationMode=BMM_Additive
    translationSpace=BCS_BoneSpace
    alphaBoolBlend {blendOption=Linear}
    alphaScaleBiasClamp {bInterpResult=true, interpSpeedDecreasing=5, interpSpeedIncreasing=5}
    boneToModify {boneName=dyn_Hair_BackR_03}
    rotation {pitch=-60}
    translation {z=4.5}
   }
   n78 = "Transform (Modify) Bone - Bone: dyn_Hair_BackL_03"(ComponentPose=@n77, Alpha=1, Translation="0,0,4.5", Rotation="-60,0,0") {
    rotationMode=BMM_Additive
    rotationSpace=BCS_BoneSpace
    translationMode=BMM_Additive
    translationSpace=BCS_BoneSpace
    alphaBoolBlend {blendOption=Linear}
    alphaScaleBiasClamp {bInterpResult=true, interpSpeedDecreasing=5, interpSpeedIncreasing=5}
    boneToModify {boneName=dyn_Hair_BackL_03}
    rotation {pitch=-60}
    translation {z=4.5}
   }
   n79 = "Transform (Modify) Bone - Bone: dyn_Hair_Back_05"!(ComponentPose=@n78, Alpha=1, Rotation="60,0,0") -> (Pose) {
    rotationMode=BMM_Additive
    rotationSpace=BCS_BoneSpace
    translationMode=BMM_Additive
    translationSpace=BCS_BoneSpace
    alphaBoolBlend {blendOption=Linear}
    alphaScaleBiasClamp {bInterpResult=true, interpSpeedDecreasing=5, interpSpeedIncreasing=5}
    boneToModify {boneName=dyn_Hair_Back_05}
    rotation {pitch=60}
   }
   n80 = "Transform (Modify) Bone - Bone: dyn_Hair_BackR_05"!(ComponentPose=@n79, Alpha=1, Rotation="30,0,0") -> (Pose) {
    rotationMode=BMM_Additive
    rotationSpace=BCS_BoneSpace
    translationMode=BMM_Additive
    translationSpace=BCS_BoneSpace
    alphaBoolBlend {blendOption=Linear}
    alphaScaleBiasClamp {bInterpResult=true, interpSpeedDecreasing=5, interpSpeedIncreasing=5}
    boneToModify {boneName=dyn_Hair_BackR_05}
    rotation {pitch=30}
   }
   n81 = "Transform (Modify) Bone - Bone: dyn_Hair_BackL_05"!(ComponentPose=@n80, Alpha=1, Rotation="30,0,0") -> (Pose) {
    rotationMode=BMM_Additive
    rotationSpace=BCS_BoneSpace
    translationMode=BMM_Additive
    translationSpace=BCS_BoneSpace
    alphaBoolBlend {blendOption=Linear}
    alphaScaleBiasClamp {bInterpResult=true, interpSpeedDecreasing=5, interpSpeedIncreasing=5}
    boneToModify {boneName=dyn_Hair_BackL_05}
    rotation {pitch=30}
   }
   Save cached pose 'CRCache_BackFix'(Pose=Component To Local(ComponentPose=@n81)) {cacheName=CRCache_BackFix}
  ]}
 EventGraph {nodes [
   Event OnCharacterCustomizationLoaded()
   > n1 = Cast To FortPlayerPawn(Object=TryGetFortPawn())
   n2 = CheckForBackbling()
   > Set is_WearingCape(is_WearingCape:bool=@n2.bIsBackblingLong) -> (Output_Get:bool)
   > Set isWearingPackBool(isWearingPackBool:bool=OR Boolean!(A=@n2, B=@n2.bIsBackblingACape, C:bool=@n2.bIsBackblingLong) -> (ReturnValue)) -> (Output_Get:bool)
   > n3 = WatchCosmeticStat!(
    execute:Exec
    Pawn:Fort Player Pawn Object Reference=@n1
    StatTag:Fort Stat Manager Tag Structure=(TagName="StatManager.AthenaKills")
    DefaultValue:int=0
    bBroadcastInitialValue:bool=false
    InitialValueBroadcastDelay:float=-1
    ChangeValueBroadcastDelay:float=-1
    bConsumeWaitingNotifications:bool=true
   ) -> (then:Exec, OnStatChanged:Exec, StatName:Name, StatValue:int) {is=title K2Node_AsyncAction}
   Set KillCount(execute=@n3.OnStatChanged, KillCount:double="To Float (Integer)"(InInt=@n3.StatValue)) -> (Output_Get:double)
   > Set Active_Kill(Active_Kill:bool=true) -> (Output_Get:bool)
   > Delay(Duration=1.5)
   > Set Active_Kill(Active_Kill=false)
   n4 = Event BlueprintProcessAnimInputEvent()
   > n5 = Branch(Condition=IsNamed(self=@n4.AnimInputEvent, InEventName=aie.Player.Zipline.Entered))
   > Set OnZipline(OnZipline:bool=true) -> (Output_Get:bool)
   Branch(execute=@n5.else, Condition=IsNamed(self=@n4.AnimInputEvent, InEventName=aie.Player.Zipline.Exited))
   > Set OnZipline(OnZipline=false)
   Event BlueprintInitializeAnimation()
   > Set Skel_Mesh(Skel_Mesh=GetOwnerSkeletalMeshComponent())
   Event BlueprintUpdateAnimation()
   > n6 = Sequence!(execute) -> (then_0, then_1, then_2:Exec, then_3:Exec, then_4:Exec, then_5:Exec, then_6:Exec, then_7:Exec)
   Get!(execute:Exec=@n6.then_3) -> (then:Exec, else:Exec, Skel_Mesh:Skeletal Mesh Component Object Reference)
   > Sequence!(execute) -> (then_0, then_1)
   > n7 = Branch(Condition=Get Active_Kill() -> (Active_Kill:bool))
   > "Set FInterp(Alpha, (10: 10))"("FInterp(Alpha, (10: 10))":float="Get FInterp(Alpha, (10: 10))_0"() -> ("FInterp(Alpha, (10: 10))_0":float)) -> (Output_Get:float)
   > Delay(Duration=1.5)
   > Set Active_Kill(Active_Kill=false)
   "Set FInterp(Alpha, (10: 10))"(execute=@n7.else, "FInterp(Alpha, (10: 10))"=0)
   Set Kill_Happened(execute=@n6.then_0, Kill_Happened:double="To Float (Boolean)"(InBool=Get Active_Kill())) -> (Output_Get:double)
   > Set FE_Alpha(FE_Alpha:bool=OR Boolean!(A=Get bIsSkyDiving(), B=Get bIsInFrontEnd()) -> (ReturnValue)) -> (Output_Get:bool)
   n8 = Get bIsDBNO()
   n9 = Get isWearingPackBool()
   > Set WearPackAlpha(WearPackAlpha:double=SelectFloat(A=1, B=0.2, bPickA=OR Boolean(A=@n9, B=@n8))) -> (Output_Get:double)
   > Set DiveCapeAlpha(DiveCapeAlpha:bool=OR Boolean!(A=Get bIsSkydiveDiveMode(), B=Get bIsSkydiveFloating(), C:bool=Get isWearingPackBool(), D:bool=Get bIsDBNO()) -> (ReturnValue)) -> (Output_Get:bool)
   n10 = OR Boolean(A=Get bIsSkydiveDiveMode(), B=Get bIsSkyDiving(), C=Get bIsSkydiveFloating(), D=Get bIsParachuteOpened())
   > Set isInAir(isInAir:double="To Float (Boolean)"(InBool=@n10)) -> (Output_Get:double)
   n11 = NOT Boolean(A=@n10)
   n12 = NOT Boolean(A=Get TacticalJump() -> (TacticalJump:bool))
   > Set AirFloatingAlpha(AirFloatingAlpha:double="To Float (Boolean)"(InBool=OR Boolean!(A=@n11, B=@n12) -> (ReturnValue))) -> (Output_Get:double)
   n13 = Get bIsMoving2D()
   n14 = Get bIsDBNO()
   > Set DBNO_Moving(DBNO_Moving:double="To Float (Boolean)"(InBool=AND Boolean(A=@n13, B=@n14))) -> (Output_Get:double)
   n15 = GetGameTimeInSeconds()
   n16 = Set Sphere_Rotation(execute=@n6.then_4, Sphere_Rotation:double=float * float(A=float * float(A=1, B=@n15), B=300)) -> (Output_Get:double)
   > Set Spinning_Rotator(Spinning_Rotator:Rotator=MakeRotator(Roll=@n16)) -> (Output_Get:Rotator)
   > Set isNotFE(isNotFE:bool=NOT Boolean(A=Get bIsInFrontEnd())) -> (Output_Get:bool)
   n17 = GetGameTimeInSeconds()
   n18 = float * float(A=6, B=@n17)
   n19 = MultiplyByPi!(Value:double="Sin (Radians)"(A=float + float(A=@n18, B=-0.25))) -> (ReturnValue:double)
   > Set sineRotate1(sineRotate1:Rotator=MakeRotator(Pitch=float * float(A=@n19, B=1.5))) -> (Output_Get:Rotator)
   n20 = MultiplyByPi(Value="Sin (Radians)"(A=float + float(A=@n18, B=0.25)))
   > Set sineRotate2(sineRotate2:Rotator=MakeRotator(Pitch=float * float(A=@n20, B=1))) -> (Output_Get:Rotator)
   n21 = MultiplyByPi(Value="Sin (Radians)"(A=float + float(A=@n18, B=0.5)))
   > Set sineRotate3(sineRotate3:Rotator=MakeRotator(Pitch=float * float(A=@n21, B=1))) -> (Output_Get:Rotator)
   n22 = float * float(A=6, B=@n17)
   n23 = MultiplyByPi(Value="Sin (Radians)"(A=float + float(A=@n22, B=-0.25)))
   > Set sineRotate4(sineRotate4:Rotator=MakeRotator(Pitch=float * float(A=@n23, B=0.5))) -> (Output_Get:Rotator)
   n24 = MultiplyByPi(Value="Sin (Radians)"(A=float + float(A=@n22, B=0.25)))
   > Set sineRotate5(sineRotate5:Rotator=MakeRotator(Pitch=float * float(A=@n24, B=0.75))) -> (Output_Get:Rotator)
   n25 = MultiplyByPi(Value="Sin (Radians)"(A=float + float(A=@n22, B=0.5)))
   > Set sineRotate6(sineRotate6:Rotator=MakeRotator(Pitch=float * float(A=@n25, B=0.75))) -> (Output_Get:Rotator)
   n26 = Gravity Override State Function DEPRECATED(
    execute=@n6.then_5
    CharacterPartAnimInstance=Self-Reference()
    GameTimeInSeconds=GetGameTimeInSeconds()
    GravityOverrideParams=Get Gravity Override Params()
   )
   > "Set Gravity Override "("Gravity Override ":Vector=@n26) -> (Output_Get:Vector)
   > n27 = Branch(Condition=OR Boolean(A=Get isWearingPackBool(), B=Get bIsDBNO()))
   > Set BackADS_Fix(BackADS_Fix:double=0) -> (Output_Get:double)
   Set BackADS_Fix(execute=@n27.else, BackADS_Fix=1)
   n28 = AND Boolean(
    A=OR Boolean!(A=Get bIsSkydiveDiveMode(), B=Get bIsSkyDiving(), C:bool=Get bIsSkydiveFloating()) -> (ReturnValue)
    B=NOT Boolean(A=Get bIsParachuteOpened())
   )
   n29 = Get TacticalJump()
   Set HairSkydivePose(execute=@n6.then_2, HairSkydivePose:double="To Float (Boolean)"(InBool=OR Boolean!(A=@n29, B=@n28) -> (ReturnValue))) -> (Output_Get:double)
   > n30 = Branch(Condition=OR Boolean!(A=Get bIsSkydiveDiveMode(), B=Get bIsSkyDiving(), C:bool=Get bIsSkydiveFloating(), D:bool=Get bIsParachuteOpened()) -> (ReturnValue))
   > Set BackHairDynAlpha(BackHairDynAlpha:double=0) -> (Output_Get:double)
   n31 = Branch(execute=@n30.else, Condition=OR Boolean!(A=Get bIsTargeting(), B=Get bIsCrouching()) -> (ReturnValue))
   > Set BackHairDynAlpha(BackHairDynAlpha=0.3)
   > Set UpperHairDyn(UpperHairDyn:double=0.1) -> (Output_Get:double)
   Set BackHairDynAlpha(execute=@n31.else, BackHairDynAlpha=1)
   > Set UpperHairDyn(UpperHairDyn=0.15)
   > n32 = Branch(Condition=Get bIsInFrontEnd())
   > Set BackHairDynAlpha(BackHairDynAlpha=0.45)
   Set BackHairDynAlpha(execute=@n32.else, BackHairDynAlpha=1)
   n33 = Branch(execute=@n6.then_7, Condition=Get OnZipline() -> (OnZipline:bool))
   > Set Simulation Space Dyncamis Hair(Simulation Space Dyncamis Hair:AnimPhysSimSpaceType Enum=Component) -> (Output_Get:AnimPhysSimSpaceType Enum)
   Set Simulation Space Dyncamis Hair(execute=@n33.else, Simulation Space Dyncamis Hair=World)
   Set Parachuting(execute=@n6.then_1, Parachuting:double="To Float (Boolean)"(InBool=Get bIsParachuteOpened())) -> (Output_Get:double)
   n34 = OR Boolean!(A=Get bIsSkydiveDiveMode(), B=Get bIsSkyDiving(), C:bool=Get bIsSkydiveFloating()) -> (ReturnValue)
   n35 = Get TacticalJump()
   > Set FakeWindAlpha(FakeWindAlpha:double="To Float (Boolean)"(InBool=OR Boolean!(A=@n35, B=@n34) -> (ReturnValue))) -> (Output_Get:double)
   n36 = OR Boolean!(A=Get bIsJumping(), B=Get bIsFalling(), C:bool=Get bIsPlayingMeleeAnim()) -> (ReturnValue)
   n37 = Get bRigidBodyEnabled()
   > Set JumpMelee(JumpMelee:double="To Float (Boolean)"(InBool=AND Boolean(A=@n37, B=@n36))) -> (Output_Get:double)
   n38 = Get bIsJumping()
   n39 = Get bIsFalling()
   > Set TacticalJump(TacticalJump:bool=AND Boolean(A=Get bIsTacticalSprinting(), B=OR Boolean!(A=@n38, B=@n39) -> (ReturnValue))) -> (Output_Get:bool)
   n40 = NOT Boolean(A=Get bIsMoving2D())
   n41 = Get bIsTargeting()
   > Set Targeting(Targeting:double="To Float (Boolean)"(InBool=AND Boolean(A=@n40, B=@n41))) -> (Output_Get:double)
   > n42 = Branch(Condition=OR Boolean!(A=Get bIsJumping(), B=Get bIsFalling(), C:bool=Get bIsInFrontEnd(), D:bool=Get bIsDBNO()) -> (ReturnValue))
   > Set JumpFix(JumpFix:double=0) -> (Output_Get:double)
   Set JumpFix(execute=@n42.else, JumpFix=1)
   n43 = NOT Boolean(A=Get isWearingPackBool())
   n44 = NOT Boolean(A=Get bIsParachuteOpened())
   n45 = OR Boolean!(A=Get bIsSkydiveDiveMode(), B=Get bIsSkyDiving(), C=Get bIsSkydiveFloating()) -> (ReturnValue)
   Set SkydiveMode(execute=@n6.then_6, SkydiveMode:double="To Float (Boolean)"(InBool=AND Boolean!(A=@n45, B=@n44, C:bool=@n43) -> (ReturnValue))) -> (Output_Get:double)
   > Set DBNO(DBNO:double="To Float (Boolean)"(InBool=Get bIsDBNO())) -> (Output_Get:double)
   n46 = Get bIsCrouching()
   n47 = Get bIsTargeting()
   > Set CrouchADS(CrouchADS:double="To Float (Boolean)"(InBool=AND Boolean!(A=@n47, B=@n46) -> (ReturnValue))) -> (Output_Get:double)
   > Set FrontHairAlpha(FrontHairAlpha:bool=OR Boolean!(A=Get bIsOnGrindRail(), B=Get bIsParachuteOpened()) -> (ReturnValue)) -> (Output_Get:bool)
   > Set FE_RBAN(FE_RBAN:bool=AND Boolean(A=Get bRigidBodyEnabled(), B=Get bIsInFrontEnd())) -> (Output_Get:bool)
   Get bIsMoving2D()
   n48 = OR Boolean!(A=Get bIsMoving2D(), B=Get bIsSkydiveDiveMode(), C:bool=Get bIsSkyDiving(), D:bool=Get bIsSkydiveFloating(), E:bool=Get bIsParachuteOpened()) -> (ReturnValue)
   n49 = Get bRigidBodyEnabled()
   Set PushBackAplha(PushBackAplha:double="To Float (Boolean)"(InBool=AND Boolean(A=@n49, B=@n48))) -> (Output_Get:double)
  ]}
]
variables [
 Active_Kill {type=boolean}
 AirFloatingAlpha {type=number}
 BackADS_Fix {type=number}
 BackHairDynAlpha {type=number}
 Body_Mesh {type=SkeletalMeshComponent}
 CrouchADS {type=number}
 DBNO {type=number}
 DBNO_Moving {type=number}
 DiveCapeAlpha {type=boolean}
 FakeWindAlpha {type=number}
 FE_Alpha {type=boolean}
 FE_RBAN {type=boolean}
 "FInterp(Alpha, (10: 10))" {type=number}
 "FInterp(Alpha, (10: 10))_0" {type=number}
 float {type=Vector}
 FrontHairAlpha {type=boolean}
 "Gravity Override " {type=Vector}
 Gravity Override Params {type=GravityOverrideParamsStruct}
 HairSkydivePose {type=number}
 is_WearingCape {type=boolean}
 isInAir {type=number}
 isNotFE {type=boolean}
 isWearingPackBool {type=boolean}
 JumpFix {type=number}
 JumpMelee {type=number}
 Kill_Happened {type=number}
 KillCount {type=number}
 OnZipline {type=boolean}
 Parachuting {type=number}
 Pawn {type=FortPlayerPawn}
 PlayerPawn {type=FortPlayerPawnAthena}
 PushBackAplha {type=number}
 ShoulderUpFix {type=number}
 Simulation Space Dyncamis Hair {type=AnimPhysSimSpaceType}
 Sine Rotation {type=number}
 sineRotate1 {type=Rotator}
 sineRotate2 {type=Rotator}
 sineRotate3 {type=Rotator}
 sineRotate4 {type=Rotator}
 sineRotate5 {type=Rotator}
 sineRotate6 {type=Rotator}
 Skel_Mesh {type=SkeletalMeshComponent}
 SkydiveMode {type=number}
 Sphere_Rotation {type=number}
 Spinning_Rotator {type=Rotator}
 TacticalJump {type=boolean}
 Targeting {type=number}
 UpperArmRot_l {type=boolean}
 UpperHairDyn {type=number}
 WearPackAlpha {type=number}
]
